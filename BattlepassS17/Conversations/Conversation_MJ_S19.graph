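# NPC conversation for MJ's Season 19 quest: a Daily Bugle interview about the Seven, the Imagined Order and the Loopers.
# The entry offers Interview (auto-selected while the quest is active and its objective incomplete), Hire (same objective check), and a DEBUG quest-grant option.
# Interview branches end in hiring MJ or returning to start.
# Only team-up offer answers complete the objective, which references AthenaQuest_InteractableCoconuts, whether the player hires or declines.
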ConversationDatabase /BattlepassS17/Conversations/Conversation_MJ_S19
entry_points [{tag=Conversation.Entry.Season19.MJ.Quest, destinations [#1]}]
nodes [
 ConversationEntryPointNode_7 {
  id=#1
  class=ConversationEntryPointNode
  outputs [#2, #3, #4]
  properties {entryTag {tagName=Conversation.Entry.Season19.MJ.Quest}}
 }
 Speech_C_5 {
  id=#4
  class=Speech_C
  outputs [#5, #6]
  properties {softTaskIcon=&T_UI_NPC_MainQuest, generalConfig {defaultMessage="Get To The Thief!"}}
  sub_nodes [ConversationChoiceNode_2 {
    class=ConversationChoiceNode
    properties {defaultChoiceDisplayText=DEBUG, choiceTags {gameplayTags [{tagName=Conversation.Choice.Category.MainQuestline}]}}
   }]
 }
 Back_C_1 {
  id=#6
  class=Back_C
  sub_nodes [ConversationChoiceNode_8 {class=ConversationChoiceNode, properties {defaultChoiceDisplayText=BACK}}]
 }
 GrantStaticQuest_C_2 {
  id=#5
  class=GrantStaticQuest_C
  outputs []
  sub_nodes [ConversationChoiceNode_1 {class=ConversationChoiceNode, properties {defaultChoiceDisplayText=ACCEPT}}]
 }
 Speech_C_42 {
  id=#3
  class=Speech_C
  outputs [#7]
  properties {generalConfig {defaultMessage=Hope you don't mind if I take notes.}}
  sub_nodes [
   ConversationChoiceNode_33 {class=ConversationChoiceNode, properties {defaultChoiceDisplayText=Hire}}
   FortConversationRequirement_ControllerRequirement_3 {
    class=FortConversationRequirement_ControllerRequirement
    properties {
     nodeName=Require: Has Completed Quest Objective
     requirement {
      class=FortControllerRequirement_HasCompletedQuestObjective
      properties {
       bInverseCondition=true
       objectiveBackendName=quest_interactablecoconuts_1
       questDefinition=/Game/Athena/Items/Quests/DiscoverableQuests/AthenaQuest_InteractableCoconuts
      }
     }
    }
   }
  ]
 }
 HireNPC_C_3 {id=#7, class=HireNPC_C, outputs []}
 Speech_C_0 {
  id=#2
  class=Speech_C
  outputs [#8, #9, #10, #11]
  properties {
   softTaskIcon=&T_UI_NPC_MainQuest
   generalConfig {defaultMessage="What's up, dork? \r\nI've heard you know <b>stuff</> about the Island. <enemy>Secret stuff.</>"}
  }
  sub_nodes [
   ConversationChoiceNode_4 {
    class=ConversationChoiceNode
    properties {
     defaultChoiceDisplayText=Interview
     choiceTags {gameplayTags [{tagName=Conversation.Choice.Category.MainQuestline}, {tagName=Conversation.Choice.AutoSelectIfValid}]}
    }
   }
   FortConversationRequirement_ControllerRequirement_10 {
    class=FortConversationRequirement_ControllerRequirement
    properties {
     nodeName=Require: Is Quest Active
     requirement {
      class=FortControllerRequirement_IsQuestActive
      properties {questDefinition=&AthenaQuest_InteractableCoconuts}
     }
    }
   }
   FortConversationRequirement_ControllerRequirement_0 {
    class=FortConversationRequirement_ControllerRequirement
    properties {
     nodeName=Require: Has Completed Quest Objective
     requirement {
      class=FortControllerRequirement_HasCompletedQuestObjective
      properties {
       bInverseCondition=true
       objectiveBackendName=quest_interactablecoconuts_1
       questDefinition=&AthenaQuest_InteractableCoconuts
      }
     }
    }
   }
  ]
 }
 Speech_C_30 {
  id=#11
  class=Speech_C
  outputs [#12, #13, #14]
  properties {generalConfig {defaultMessage="So you're an agent for <b>The Seven</> then? \r\nThought so."}}
  sub_nodes [ConversationChoiceNode_5 {class=ConversationChoiceNode, properties {defaultChoiceDisplayText="Sorry, I'm not at liberty to say."}}]
 }
 Speech_C_33 {
  id=#14
  class=Speech_C
  outputs [#15]
  properties {generalConfig {defaultMessage="Nothing screams \"cool secret agent\" like being paralyzed by indecision."}}
  sub_nodes [ConversationChoiceNode_11 {class=ConversationChoiceNode, properties {defaultChoiceDisplayText="I haven't decided yet. "}}]
 }
 Speech_C_40 {
  id=#15
  class=Speech_C
  outputs [#16]
  properties {generalConfig {defaultMessage=Let me know where you land when you make up your mind.}}
 }
 Speech_C_41 {
  id=#16
  class=Speech_C
  outputs [#17]
  properties {generalConfig {defaultMessage=Every good junior reporter needs good informants. And you seem... fine.}}
 }
 ReturnToStart_C_10 {id=#17, class=ReturnToStart_C}
 Speech_C_32 {
  id=#13
  class=Speech_C
  outputs [#18]
  properties {generalConfig {defaultMessage="You're saving the world, one humblebrag at a time. "}}
  sub_nodes [ConversationChoiceNode_21 {class=ConversationChoiceNode, properties {defaultChoiceDisplayText=" I'm just trying to save the world. No biggie."}}]
 }
 Speech_C_36 {
  id=#18
  class=Speech_C
  outputs [#19]
  properties {generalConfig {defaultMessage="Well <b>Captain Superawesome,</> mind if I trail you for my <b>Bugle</> article?"}}
 }
 Speech_C_37 {
  id=#19
  class=Speech_C
  outputs [#20, #21]
  properties {generalConfig {defaultMessage=Don't worry about me. I've worked with superdorks before.}}
 }
 FortConversationTaskNode_CompleteQuestObjectiveByBackendName_9 {
  id=#21
  class=FortConversationTaskNode_CompleteQuestObjectiveByBackendName
  outputs [#22]
  properties {
   objectiveBackendName=quest_interactablecoconuts_1
   questDefinition=&AthenaQuest_InteractableCoconuts
  }
  sub_nodes [ConversationChoiceNode_10 {class=ConversationChoiceNode, properties {defaultChoiceDisplayText="No thanks, I work alone. "}}]
 }
 Speech_C_39 {
  id=#22
  class=Speech_C
  outputs [#23]
  properties {generalConfig {defaultMessage="Meh. Whatever.\r\nI'll get my intel somewhere else."}}
 }
 ReturnToStart_C_9 {id=#23, class=ReturnToStart_C}
 FortConversationTaskNode_CompleteQuestObjectiveByBackendName_8 {
  id=#20
  class=FortConversationTaskNode_CompleteQuestObjectiveByBackendName
  outputs [#24]
  properties {
   objectiveBackendName=quest_interactablecoconuts_1
   questDefinition=&AthenaQuest_InteractableCoconuts
  }
  sub_nodes [ConversationChoiceNode_17 {class=ConversationChoiceNode, properties {defaultChoiceDisplayText=Hire}}]
 }
 Speech_C_38 {
  id=#24
  class=Speech_C
  outputs [#25]
  properties {generalConfig {defaultMessage="Lead the way, oh mighty hero. "}}
 }
 HireNPC_C_2 {id=#25, class=HireNPC_C, outputs []}
 Speech_C_31 {
  id=#12
  class=Speech_C
  outputs [#26]
  properties {generalConfig {defaultMessage=Convincing.}}
  sub_nodes [ConversationChoiceNode_20 {class=ConversationChoiceNode, properties {defaultChoiceDisplayText="(Lie) Pffft. No way. The Seven? Who's that?!"}}]
 }
 Speech_C_34 {
  id=#26
  class=Speech_C
  outputs [#27]
  properties {generalConfig {defaultMessage="That's fine. \r\nI've got other sources for my <b>Bugle</> article."}}
 }
 Speech_C_35 {
  id=#27
  class=Speech_C
  outputs [#28]
  properties {generalConfig {defaultMessage="If you hear anything about <b>the Seven,</> \r\nyou know where to find me."}}
 }
 ReturnToStart_C_8 {id=#28, class=ReturnToStart_C}
 Speech_C_19 {
  id=#10
  class=Speech_C
  outputs [#29, #30, #31]
  properties {generalConfig {defaultMessage="Is that a pun? <b>Yikes. </>\r\nI guess that explains why the Loopers love you."}}
  sub_nodes [ConversationChoiceNode_26 {class=ConversationChoiceNode, properties {defaultChoiceDisplayText="I did... Until the world turned upside down. "}}]
 }
 Speech_C_22 {
  id=#31
  class=Speech_C
  outputs [#32]
  properties {generalConfig {defaultMessage="I can respect loyalty. "}}
  sub_nodes [ConversationChoiceNode_29 {class=ConversationChoiceNode, properties {defaultChoiceDisplayText="That, and I don't go spilling their secrets to strangers."}}]
 }
 Speech_C_28 {
  id=#32
  class=Speech_C
  outputs [#33]
  properties {generalConfig {defaultMessage="I'm working on an article for <b>the Bugle.</>"}}
 }
 Speech_C_29 {
  id=#33
  class=Speech_C
  outputs [#34]
  properties {generalConfig {defaultMessage="If you change your mind and want to talk, I'm around."}}
 }
 ReturnToStart_C_7 {id=#34, class=ReturnToStart_C}
 Speech_C_21 {
  id=#30
  class=Speech_C
  outputs [#35, #36]
  properties {generalConfig {defaultMessage="It'd help my story for <b>the Bugle</> if I had a local show me around the Island. "}}
  sub_nodes [ConversationChoiceNode_28 {class=ConversationChoiceNode, properties {defaultChoiceDisplayText="One of the many reasons, yes."}}]
 }
 FortConversationTaskNode_CompleteQuestObjectiveByBackendName_7 {
  id=#36
  class=FortConversationTaskNode_CompleteQuestObjectiveByBackendName
  outputs [#37]
  properties {
   objectiveBackendName=quest_interactablecoconuts_1
   questDefinition=&AthenaQuest_InteractableCoconuts
  }
  sub_nodes [ConversationChoiceNode_31 {class=ConversationChoiceNode, properties {defaultChoiceDisplayText=Maybe later.}}]
 }
 Speech_C_27 {
  id=#37
  class=Speech_C
  outputs [#38]
  properties {generalConfig {defaultMessage=See ya around.}}
 }
 ReturnToStart_C_6 {id=#38, class=ReturnToStart_C}
 FortConversationTaskNode_CompleteQuestObjectiveByBackendName_6 {
  id=#35
  class=FortConversationTaskNode_CompleteQuestObjectiveByBackendName
  outputs [#39]
  properties {
   objectiveBackendName=quest_interactablecoconuts_1
   questDefinition=&AthenaQuest_InteractableCoconuts
  }
  sub_nodes [ConversationChoiceNode_32 {class=ConversationChoiceNode, properties {defaultChoiceDisplayText=Hire}}]
 }
 Speech_C_26 {
  id=#39
  class=Speech_C
  outputs [#40]
  properties {generalConfig {defaultMessage=Hope you don't mind if I take notes.}}
 }
 HireNPC_C_1 {id=#40, class=HireNPC_C, outputs []}
 Speech_C_20 {
  id=#29
  class=Speech_C
  outputs [#41]
  properties {generalConfig {defaultMessage="Yeah. \r\nEspecially the creepy <b>pink teddy bear.</>"}}
  sub_nodes [ConversationChoiceNode_27 {class=ConversationChoiceNode, properties {defaultChoiceDisplayText=The Loopers love me?}}]
 }
 Speech_C_23 {
  id=#41
  class=Speech_C
  outputs [#42, #43]
  properties {generalConfig {defaultMessage=Any comment for the Bugle on behalf of the Loopers?}}
 }
 Speech_C_25 {
  id=#43
  class=Speech_C
  outputs [#44]
  properties {generalConfig {defaultMessage="Got it. Thanks. \r\nIf you hear of any new scoops... You know where I am. "}}
  sub_nodes [ConversationChoiceNode_14 {class=ConversationChoiceNode, properties {defaultChoiceDisplayText="This Island is ours. And we will defend it. "}}]
 }
 ReturnToStart_C_5 {id=#44, class=ReturnToStart_C}
 Speech_C_24 {
  id=#42
  class=Speech_C
  outputs [#44]
  properties {generalConfig {defaultMessage=You know where I am.}}
  sub_nodes [ConversationChoiceNode_23 {class=ConversationChoiceNode, properties {defaultChoiceDisplayText="Maybe some other time. "}}]
 }
 Speech_C_2 {
  id=#9
  class=Speech_C
  outputs [#45, #46, #47]
  properties {generalConfig {defaultMessage="I'm working on a piece for <b>the Bugle </>\r\nabout your buddies, the <enemy>Imagined Order. </>"}}
  sub_nodes [ConversationChoiceNode_18 {class=ConversationChoiceNode, properties {defaultChoiceDisplayText="I might know a few things. "}}]
 }
 Speech_C_13 {
  id=#47
  class=Speech_C
  outputs [#48]
  properties {generalConfig {defaultMessage="This is... <b>Helpful.</>"}}
  sub_nodes [ConversationChoiceNode_19 {class=ConversationChoiceNode, properties {defaultChoiceDisplayText=I'll share my info if you share yours.}}]
 }
 Speech_C_14 {
  id=#48
  class=Speech_C
  outputs [#49]
  properties {generalConfig {defaultMessage="I can see why the <enemy>Imagined Order</> liked you. \r\nYou're easy to trick."}}
 }
 Speech_C_15 {
  id=#49
  class=Speech_C
  outputs [#50, #51]
  properties {generalConfig {defaultMessage="Any interest in <b>teaming up</> for a round? \r\nI could use a <b>guide.</>"}}
 }
 FortConversationTaskNode_CompleteQuestObjectiveByBackendName_5 {
  id=#51
  class=FortConversationTaskNode_CompleteQuestObjectiveByBackendName
  outputs [#52]
  properties {
   objectiveBackendName=quest_interactablecoconuts_1
   questDefinition=&AthenaQuest_InteractableCoconuts
  }
  sub_nodes [ConversationChoiceNode_25 {class=ConversationChoiceNode, properties {defaultChoiceDisplayText=Maybe some other time.}}]
 }
 Speech_C_18 {
  id=#52
  class=Speech_C
  outputs [#53]
  properties {generalConfig {defaultMessage="Thanks for the intel anyway. \r\nSee ya."}}
 }
 ReturnToStart_C_4 {id=#53, class=ReturnToStart_C}
 FortConversationTaskNode_CompleteQuestObjectiveByBackendName_0 {
  id=#50
  class=FortConversationTaskNode_CompleteQuestObjectiveByBackendName
  outputs [#54]
  properties {
   objectiveBackendName=quest_interactablecoconuts_1
   questDefinition=&AthenaQuest_InteractableCoconuts
  }
  sub_nodes [ConversationChoiceNode_24 {class=ConversationChoiceNode, properties {defaultChoiceDisplayText=Hire}}]
 }
 Speech_C_17 {
  id=#54
  class=Speech_C
  outputs [#55]
  properties {generalConfig {defaultMessage=Let's get going.}}
 }
 HireNPC_C_0 {id=#55, class=HireNPC_C, outputs []}
 Speech_C_6 {
  id=#46
  class=Speech_C
  outputs [#56, #57]
  properties {generalConfig {defaultMessage="No judgement... \r\nOkay, maybe a <b>little</> judgement. "}}
  sub_nodes [ConversationChoiceNode_7 {class=ConversationChoiceNode, properties {defaultChoiceDisplayText=We're former colleagues. Emphasis on FORMER.}}]
 }
 Speech_C_11 {
  id=#57
  class=Speech_C
  outputs [#58]
  properties {generalConfig {defaultMessage="The <enemy>Imagined Order</> is making a comeback. \r\nAnd it'll be <b>soon.</>"}}
  sub_nodes [ConversationChoiceNode_22 {class=ConversationChoiceNode, properties {defaultChoiceDisplayText=Very funny. What's the piece about?}}]
 }
 Speech_C_12 {
  id=#58
  class=Speech_C
  outputs [#59]
  properties {generalConfig {defaultMessage="So if you've got tips... \r\nCome <b>find me.</>"}}
 }
 ReturnToStart_C_3 {id=#59, class=ReturnToStart_C}
 Speech_C_8 {
  id=#56
  class=Speech_C
  outputs [#60]
  properties {generalConfig {defaultMessage="I know a guy who works with aliens.\r\nThey're nice... <b>Most of the time.</>"}}
  sub_nodes [ConversationChoiceNode_13 {class=ConversationChoiceNode, properties {defaultChoiceDisplayText="There was a literal alien invasion going on!"}}]
 }
 Speech_C_9 {
  id=#60
  class=Speech_C
  outputs [#61]
  properties {generalConfig {defaultMessage="Point is, I hear the <enemy>Imagined Order</> is making a comeback."}}
 }
 Speech_C_10 {
  id=#61
  class=Speech_C
  outputs [#62]
  properties {generalConfig {defaultMessage="If you hear anything... \r\n<b>Call me.</>"}}
 }
 ReturnToStart_C_2 {id=#62, class=ReturnToStart_C}
 Speech_C_4 {
  id=#45
  class=Speech_C
  outputs [#63]
  properties {generalConfig {defaultMessage="Whatever. \r\nIf you're hiding something, I'll figure it out."}}
  sub_nodes [ConversationChoiceNode_6 {class=ConversationChoiceNode, properties {defaultChoiceDisplayText=No comment.}}]
 }
 ReturnToStart_C_1 {id=#63, class=ReturnToStart_C}
 Back_C_0 {
  id=#8
  class=Back_C
  sub_nodes [ConversationChoiceNode_9 {class=ConversationChoiceNode, properties {defaultChoiceDisplayText=BACK}}]
 }
]
