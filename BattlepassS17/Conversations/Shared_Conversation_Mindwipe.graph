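# Shared Season 17 NPC conversation for the Mindwipe quest (quest_s17_w08_legendary_q4), branching for AlienSummer, SwampStalker and BunkerJonesy.
# Auto-selected greetings require the NPC's tag and the active, uncompleted quest.
# Each branch applies a mindwipe GameplayEffect, plays two reaction lines, then completes that NPC's objective unless the quest awaits turn-in.
# All three objective-not-completed checks name quest_s17_w05_legendary_q5_obj0, unlike the objectives completed.
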ConversationDatabase /BattlepassS17/Conversations/Shared_Conversation_Mindwipe
entry_points [{tag=Conversation.Entry.Season17.Narrative.Mindwipe, destinations [#1]}]
nodes [
 ConversationEntryPointNode_7 {
  id=#1
  class=ConversationEntryPointNode
  outputs [#2, #3, #4]
  properties {entryTag {tagName=Conversation.Entry.Season17.Narrative.Mindwipe}}
 }
 Speech_C_2 {
  id=#4
  class=Speech_C
  outputs [#5]
  properties {
   softTaskIcon=&T_UI_NPC_MainQuest
   generalConfig {defaultMessage="Let me guess. \r\nYou're here to make me forget."}
  }
  sub_nodes [
   ConversationChoiceNode_26 {
    class=ConversationChoiceNode
    properties {
     defaultChoiceDisplayText=Initial Greeting
     choiceTags {gameplayTags [{tagName=Conversation.Choice.Category.MainQuestline}, {tagName=Conversation.Choice.AutoSelectIfValid}]}
    }
   }
   FortConversationRequirement_ControllerRequirement_19 {
    class=FortConversationRequirement_ControllerRequirement
    properties {
     nodeName=Require: Has Completed Quest
     requirement {
      class=FortControllerRequirement_HasCompletedQuest
      properties {bInverseCondition=true, questDefinition=/BattlepassS17/Items/QuestItems/Legendary/quest_s17_w08_legendary_q4}
     }
    }
   }
   FortConversationRequirement_ControllerRequirement_20 {
    class=FortConversationRequirement_ControllerRequirement
    properties {
     nodeName=Require: Is Quest Active
     requirement {
      class=FortControllerRequirement_IsQuestActive
      properties {questDefinition=&quest_s17_w08_legendary_q4}
     }
    }
   }
   FortConversationRequirement_ControllerRequirement_21 {
    class=FortConversationRequirement_ControllerRequirement
    properties {
     nodeName=Require: Has Completed Quest Objective
     requirement {
      class=FortControllerRequirement_HasCompletedQuestObjective
      properties {
       bInverseCondition=true
       objectiveBackendName=quest_s17_w05_legendary_q5_obj0
       questDefinition=&quest_s17_w08_legendary_q4
      }
     }
    }
   }
   FortConversationRequirement_ControllerRequirement_22 {
    class=FortConversationRequirement_ControllerRequirement
    properties {
     nodeName=Require: NPC Has Conversation Tag
     participantId {tagName=Conversation.Participant.NPC}
     requirement {class=FortControllerRequirement_NPCHasConversationEntryTag, properties {queryTag {tagName=Conversation.Entry.BunkerJonesy}}}
    }
   }
  ]
 }
 FortConversationTaskNode_ApplyDataDrivenEffect_2 {
  id=#5
  class=FortConversationTaskNode_ApplyDataDrivenEffect
  outputs [#6]
  properties {
   nodeName=Flash Mindwipe
   effectsToApply [{
     class=FortControllerEffect_ApplyGameplayEffect
     properties {gameplayEffect=/BattlepassS17/Items/QuestAbilities/Qest_Mindwipe/GE_Quest_Mindwipe.GE_Quest_Mindwipe_C}
    }]
  }
 }
 Speech_C_7 {
  id=#6
  class=Speech_C
  outputs [#7]
  properties {generalConfig {defaultMessage="I WILL remember! \r\nThe people will hear the TRUTH!"}}
  sub_nodes [ConversationChoiceNode_9 {class=ConversationChoiceNode, properties {defaultChoiceDisplayText=Continue}}]
 }
 Speech_C_3 {
  id=#7
  class=Speech_C
  outputs [#8]
  properties {generalConfig {defaultMessage="...Where am I? Who are you? "}}
  sub_nodes [ConversationChoiceNode_25 {class=ConversationChoiceNode, properties {defaultChoiceDisplayText=Continue}}]
 }
 FortConversationTaskNode_CompleteQuestObjectiveByBackendName_1 {
  id=#8
  class=FortConversationTaskNode_CompleteQuestObjectiveByBackendName
  properties {objectiveBackendName=quest_s17_w08_legendary_q4_obj2, questDefinition=&quest_s17_w08_legendary_q4}
  sub_nodes [FortConversationRequirement_ControllerRequirement_18 {
    class=FortConversationRequirement_ControllerRequirement
    properties {
     nodeName=Require: Has Quest To Turn In
     requirement {
      class=FortControllerRequirement_HasQuestToTurnIn
      properties {bInverseCondition=true, softQuestToTurnIn=&quest_s17_w08_legendary_q4}
     }
    }
   }]
 }
 Speech_C_0 {
  id=#3
  class=Speech_C
  outputs [#9]
  properties {softTaskIcon=&T_UI_NPC_MainQuest, generalConfig {defaultMessage=You look worried. Is something wrong?}}
  sub_nodes [
   ConversationChoiceNode_5 {
    class=ConversationChoiceNode
    properties {
     defaultChoiceDisplayText=Initial Greeting
     choiceTags {gameplayTags [{tagName=Conversation.Choice.Category.MainQuestline}, {tagName=Conversation.Choice.AutoSelectIfValid}]}
    }
   }
   FortConversationRequirement_ControllerRequirement_5 {
    class=FortConversationRequirement_ControllerRequirement
    properties {
     nodeName=Require: Has Completed Quest
     requirement {
      class=FortControllerRequirement_HasCompletedQuest
      properties {bInverseCondition=true, questDefinition=&quest_s17_w08_legendary_q4}
     }
    }
   }
   FortConversationRequirement_ControllerRequirement_6 {
    class=FortConversationRequirement_ControllerRequirement
    properties {
     nodeName=Require: Is Quest Active
     requirement {
      class=FortControllerRequirement_IsQuestActive
      properties {questDefinition=&quest_s17_w08_legendary_q4}
     }
    }
   }
   FortConversationRequirement_ControllerRequirement_7 {
    class=FortConversationRequirement_ControllerRequirement
    properties {
     nodeName=Require: Has Completed Quest Objective
     requirement {
      class=FortControllerRequirement_HasCompletedQuestObjective
      properties {
       bInverseCondition=true
       objectiveBackendName=quest_s17_w05_legendary_q5_obj0
       questDefinition=&quest_s17_w08_legendary_q4
      }
     }
    }
   }
   FortConversationRequirement_ControllerRequirement_8 {
    class=FortConversationRequirement_ControllerRequirement
    properties {
     nodeName=Require: NPC Has Conversation Tag
     participantId {tagName=Conversation.Participant.NPC}
     requirement {class=FortControllerRequirement_NPCHasConversationEntryTag, properties {queryTag {tagName=Conversation.Entry.SwampStalker}}}
    }
   }
  ]
 }
 FortConversationTaskNode_ApplyDataDrivenEffect_1 {
  id=#9
  class=FortConversationTaskNode_ApplyDataDrivenEffect
  outputs [#10]
  properties {
   nodeName=Flash Mindwipe
   effectsToApply [{
     class=FortControllerEffect_ApplyGameplayEffect
     properties {gameplayEffect=&GE_Quest_Mindwipe_C}
    }]
  }
 }
 Speech_C_6 {
  id=#10
  class=Speech_C
  outputs [#11]
  properties {generalConfig {defaultMessage="What is that?\r\nSome sort of frog finding device?"}}
  sub_nodes [ConversationChoiceNode_8 {class=ConversationChoiceNode, properties {defaultChoiceDisplayText=Continue}}]
 }
 Speech_C_1 {
  id=#11
  class=Speech_C
  outputs [#12]
  properties {generalConfig {defaultMessage=Funny. I don't remember putting my swamp boots on.}}
  sub_nodes [ConversationChoiceNode_6 {class=ConversationChoiceNode, properties {defaultChoiceDisplayText=Continue}}]
 }
 FortConversationTaskNode_CompleteQuestObjectiveByBackendName_0 {
  id=#12
  class=FortConversationTaskNode_CompleteQuestObjectiveByBackendName
  properties {objectiveBackendName=quest_s17_w08_legendary_q4_obj1, questDefinition=&quest_s17_w08_legendary_q4}
  sub_nodes [FortConversationRequirement_ControllerRequirement_9 {
    class=FortConversationRequirement_ControllerRequirement
    properties {
     nodeName=Require: Has Quest To Turn In
     requirement {
      class=FortControllerRequirement_HasQuestToTurnIn
      properties {bInverseCondition=true, softQuestToTurnIn=&quest_s17_w08_legendary_q4}
     }
    }
   }]
 }
 Speech_C_20 {
  id=#2
  class=Speech_C
  outputs [#13]
  properties {
   softTaskIcon=&T_UI_NPC_MainQuest
   generalConfig {defaultMessage="You wouldn't mindwipe a very innocent, VERY wealthy tourist!"}
  }
  sub_nodes [
   ConversationChoiceNode_3 {
    class=ConversationChoiceNode
    properties {
     defaultChoiceDisplayText=Initial Greeting
     choiceTags {gameplayTags [{tagName=Conversation.Choice.Category.MainQuestline}, {tagName=Conversation.Choice.AutoSelectIfValid}]}
    }
   }
   FortConversationRequirement_ControllerRequirement_0 {
    class=FortConversationRequirement_ControllerRequirement
    properties {
     nodeName=Require: Has Completed Quest
     requirement {
      class=FortControllerRequirement_HasCompletedQuest
      properties {bInverseCondition=true, questDefinition=&quest_s17_w08_legendary_q4}
     }
    }
   }
   FortConversationRequirement_ControllerRequirement_1 {
    class=FortConversationRequirement_ControllerRequirement
    properties {
     nodeName=Require: Is Quest Active
     requirement {
      class=FortControllerRequirement_IsQuestActive
      properties {questDefinition=&quest_s17_w08_legendary_q4}
     }
    }
   }
   FortConversationRequirement_ControllerRequirement_2 {
    class=FortConversationRequirement_ControllerRequirement
    properties {
     nodeName=Require: Has Completed Quest Objective
     requirement {
      class=FortControllerRequirement_HasCompletedQuestObjective
      properties {
       bInverseCondition=true
       objectiveBackendName=quest_s17_w05_legendary_q5_obj0
       questDefinition=&quest_s17_w08_legendary_q4
      }
     }
    }
   }
   FortConversationRequirement_ControllerRequirement_3 {
    class=FortConversationRequirement_ControllerRequirement
    properties {
     nodeName=Require: NPC Has Conversation Tag
     participantId {tagName=Conversation.Participant.NPC}
     requirement {class=FortControllerRequirement_NPCHasConversationEntryTag, properties {queryTag {tagName=Conversation.Entry.AlienSummer}}}
    }
   }
  ]
 }
 FortConversationTaskNode_ApplyDataDrivenEffect_0 {
  id=#13
  class=FortConversationTaskNode_ApplyDataDrivenEffect
  outputs [#14]
  properties {
   nodeName=Flash Mindwipe
   effectsToApply [{
     class=FortControllerEffect_ApplyGameplayEffect
     properties {gameplayEffect=&GE_Quest_Mindwipe_C}
    }]
  }
 }
 Speech_C_4 {
  id=#14
  class=Speech_C
  outputs [#15]
  properties {generalConfig {defaultMessage="MY FATHER WILL BE HEARING ABOUT THIS!"}}
  sub_nodes [ConversationChoiceNode_7 {class=ConversationChoiceNode, properties {defaultChoiceDisplayText=Continue}}]
 }
 Speech_C_19 {
  id=#15
  class=Speech_C
  outputs [#16]
  properties {generalConfig {defaultMessage="My father... who is, um... who is my father?"}}
  sub_nodes [ConversationChoiceNode_4 {class=ConversationChoiceNode, properties {defaultChoiceDisplayText=Continue}}]
 }
 FortConversationTaskNode_CompleteQuestObjectiveByBackendName_2 {
  id=#16
  class=FortConversationTaskNode_CompleteQuestObjectiveByBackendName
  properties {objectiveBackendName=quest_s17_w08_legendary_q4_obj0, questDefinition=&quest_s17_w08_legendary_q4}
  sub_nodes [FortConversationRequirement_ControllerRequirement_4 {
    class=FortConversationRequirement_ControllerRequirement
    properties {
     nodeName=Require: Has Quest To Turn In
     requirement {
      class=FortControllerRequirement_HasQuestToTurnIn
      properties {bInverseCondition=true, softQuestToTurnIn=&quest_s17_w08_legendary_q4}
     }
    }
   }]
 }
]
