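# Season 17 payphone conversation database in which Sloan briefs players on weekly Legendary quests against the alien invasion.
# One entry tag fans out to nine auto-selected "NEW MISSION" branches, gated on quest active and quest and objective incomplete.
# Each branch completes that quest's first objective, passes a message number to the payphone actor and plays linear speech nodes.
# The message 11 branch checks week 11 is active but otherwise references week 12.
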ConversationDatabase /BattlepassS17/Conversations/Shared_Conversation_Sloan_Payphones
entry_points [{tag=Conversation.Entry.Season17.Narrative.Payphones, destinations [#1]}]
nodes [
 ConversationEntryPointNode_0 {
  id=#1
  class=ConversationEntryPointNode
  outputs [
   #2
   #3
   #4
   #5
   #6
   #7
   #8
   #9
   #10
  ]
  properties {entryTag {tagName=Conversation.Entry.Season17.Narrative.Payphones}}
 }
 FortConversationTaskNode_CompleteQuestObjectiveByBackendName_8 {
  id=#10
  class=FortConversationTaskNode_CompleteQuestObjectiveByBackendName
  outputs [#11]
  properties {objectiveBackendName=quest_s17_w14_legendary_q1_obj0, questDefinition=/BattlepassS17/Items/QuestItems/Legendary/quest_s17_w14_legendary_q1}
  sub_nodes [
   FortConversationRequirement_ControllerRequirement_21 {
    class=FortConversationRequirement_ControllerRequirement
    properties {
     nodeName=Require: Is Quest Active
     requirement {
      class=FortControllerRequirement_IsQuestActive
      properties {questDefinition=&quest_s17_w14_legendary_q1}
     }
    }
   }
   FortConversationRequirement_ControllerRequirement_22 {
    class=FortConversationRequirement_ControllerRequirement
    properties {
     nodeName=Require: Has Completed Quest
     requirement {
      class=FortControllerRequirement_HasCompletedQuest
      properties {bInverseCondition=true, questDefinition=&quest_s17_w14_legendary_q1}
     }
    }
   }
   FortConversationRequirement_ControllerRequirement_23 {
    class=FortConversationRequirement_ControllerRequirement
    properties {
     nodeName=Require: Has Completed Quest Objective
     requirement {
      class=FortControllerRequirement_HasCompletedQuestObjective
      properties {
       bInverseCondition=true
       objectiveBackendName=quest_s17_w14_legendary_q1_obj0
       questDefinition=&quest_s17_w14_legendary_q1
      }
     }
    }
   }
   ConversationChoiceNode_8 {
    class=ConversationChoiceNode
    properties {
     defaultChoiceDisplayText=NEW MISSION
     choiceTags {gameplayTags [{tagName=Conversation.Choice.Category.MainQuestline}, {tagName=Conversation.Choice.AutoSelectIfValid}]}
    }
   }
   ConvoSideEffect_PayphoneActor_Message_C_8 {class=ConvoSideEffect_PayphoneActor_Message_C, properties {message Number=13}}
  ]
 }
 Speech_C_33 {
  id=#11
  class=Speech_C
  outputs [#12]
  properties {
   softTaskIcon=&T_UI_NPC_MainQuest
   generalConfig {defaultMessage="If you're hearing this recorded message, that means we're <b>getting close</> to executing our plan."}
  }
 }
 Speech_C_37 {
  id=#12
  class=Speech_C
  outputs [#13]
  properties {generalConfig {defaultMessage="If everything works like I think it will... that  <enemy>Mothership's</> days of ominous hovering are numbered."}}
 }
 Speech_C_38 {
  id=#13
  class=Speech_C
  outputs [#14]
  properties {generalConfig {defaultMessage="<b>Unfortunately</>, The area where it's coming down isn't as unpopulated as I'd like."}}
 }
 Speech_C_39 {
  id=#14
  class=Speech_C
  outputs [#15]
  properties {
   softTaskIcon=&T_UI_NPC_MainQuest
   generalConfig {defaultMessage="I need you to <b>warn the people</> there ahead of time and get them to leave."}
  }
 }
 Speech_C_40 {
  id=#15
  class=Speech_C
  outputs []
  properties {generalConfig {defaultMessage="<enemy>Can't risk</>\r\n any of them doing something to <b>interfere with the plan.</>"}}
 }
 FortConversationTaskNode_CompleteQuestObjectiveByBackendName_6 {
  id=#9
  class=FortConversationTaskNode_CompleteQuestObjectiveByBackendName
  outputs [#16]
  properties {objectiveBackendName=quest_s17_w12_legendary_q1_obj0, questDefinition=/BattlepassS17/Items/QuestItems/Legendary/quest_s17_w12_legendary_q1}
  sub_nodes [
   FortConversationRequirement_ControllerRequirement_24 {
    class=FortConversationRequirement_ControllerRequirement
    properties {
     nodeName=Require: Is Quest Active
     requirement {
      class=FortControllerRequirement_IsQuestActive
      properties {questDefinition=&quest_s17_w12_legendary_q1}
     }
    }
   }
   FortConversationRequirement_ControllerRequirement_25 {
    class=FortConversationRequirement_ControllerRequirement
    properties {
     nodeName=Require: Has Completed Quest
     requirement {
      class=FortControllerRequirement_HasCompletedQuest
      properties {bInverseCondition=true, questDefinition=&quest_s17_w12_legendary_q1}
     }
    }
   }
   FortConversationRequirement_ControllerRequirement_26 {
    class=FortConversationRequirement_ControllerRequirement
    properties {
     nodeName=Require: Has Completed Quest Objective
     requirement {
      class=FortControllerRequirement_HasCompletedQuestObjective
      properties {
       bInverseCondition=true
       objectiveBackendName=quest_s17_w12_legendary_q1_obj0
       questDefinition=&quest_s17_w12_legendary_q1
      }
     }
    }
   }
   ConversationChoiceNode_9 {
    class=ConversationChoiceNode
    properties {
     defaultChoiceDisplayText=NEW MISSION
     choiceTags {gameplayTags [{tagName=Conversation.Choice.Category.MainQuestline}, {tagName=Conversation.Choice.AutoSelectIfValid}]}
    }
   }
   ConvoSideEffect_PayphoneActor_Message_C_9 {class=ConvoSideEffect_PayphoneActor_Message_C, properties {message Number=12}}
  ]
 }
 Speech_C_27 {
  id=#16
  class=Speech_C
  outputs [#17]
  properties {
   softTaskIcon=&T_UI_NPC_MainQuest
   generalConfig {defaultMessage="Peaceful solutions are still in play, but don't hold your breath."}
  }
 }
 Speech_C_28 {
  id=#17
  class=Speech_C
  outputs [#18]
  properties {generalConfig {defaultMessage="We need to prepare for the worst. And I do mean <b>the worst</>. "}}
 }
 Speech_C_29 {
  id=#18
  class=Speech_C
  outputs [#19]
  properties {generalConfig {defaultMessage="The IO does have a counter attack plan in place, but I believe it's been compromised."}}
 }
 Speech_C_30 {
  id=#19
  class=Speech_C
  outputs [#20]
  properties {softTaskIcon=&T_UI_NPC_MainQuest, generalConfig {defaultMessage="You and me are creating a <b>backup plan</>."}}
 }
 Speech_C_31 {
  id=#20
  class=Speech_C
  outputs [#21]
  properties {generalConfig {defaultMessage="Start with <enemy>weapons testing</>."}}
 }
 Speech_C_32 {
  id=#21
  class=Speech_C
  outputs []
  properties {generalConfig {defaultMessage=I'll handle the behind-the-scenes work.}}
 }
 FortConversationTaskNode_CompleteQuestObjectiveByBackendName_7 {
  id=#8
  class=FortConversationTaskNode_CompleteQuestObjectiveByBackendName
  outputs [#22]
  properties {objectiveBackendName=quest_s17_w12_legendary_q1_obj0, questDefinition=&quest_s17_w12_legendary_q1}
  sub_nodes [
   FortConversationRequirement_ControllerRequirement_12 {
    class=FortConversationRequirement_ControllerRequirement
    properties {
     nodeName=Require: Is Quest Active
     requirement {
      class=FortControllerRequirement_IsQuestActive
      properties {questDefinition=/BattlepassS17/Items/QuestItems/Legendary/quest_s17_w11_legendary_q1}
     }
    }
   }
   FortConversationRequirement_ControllerRequirement_13 {
    class=FortConversationRequirement_ControllerRequirement
    properties {
     nodeName=Require: Has Completed Quest
     requirement {
      class=FortControllerRequirement_HasCompletedQuest
      properties {bInverseCondition=true, questDefinition=&quest_s17_w12_legendary_q1}
     }
    }
   }
   FortConversationRequirement_ControllerRequirement_17 {
    class=FortConversationRequirement_ControllerRequirement
    properties {
     nodeName=Require: Has Completed Quest Objective
     requirement {
      class=FortControllerRequirement_HasCompletedQuestObjective
      properties {
       bInverseCondition=true
       objectiveBackendName=quest_s17_w12_legendary_q1_obj0
       questDefinition=&quest_s17_w12_legendary_q1
      }
     }
    }
   }
   ConversationChoiceNode_6 {
    class=ConversationChoiceNode
    properties {
     defaultChoiceDisplayText=NEW MISSION
     choiceTags {gameplayTags [{tagName=Conversation.Choice.Category.MainQuestline}, {tagName=Conversation.Choice.AutoSelectIfValid}]}
    }
   }
   ConvoSideEffect_PayphoneActor_Message_C_6 {class=ConvoSideEffect_PayphoneActor_Message_C, properties {message Number=11}}
  ]
 }
 Speech_C_36 {
  id=#22
  class=Speech_C
  outputs [#23]
  properties {
   softTaskIcon=&T_UI_NPC_MainQuest
   generalConfig {defaultMessage=I'm hitting a dead end. We need an asset on the inside.}
  }
 }
 Speech_C_35 {
  id=#23
  class=Speech_C
  outputs [#24]
  properties {generalConfig {defaultMessage="There's an alien in quote-unquote \"<b>deep cover</>.\" "}}
 }
 Speech_C_34 {
  id=#24
  class=Speech_C
  outputs []
  properties {generalConfig {defaultMessage="Try to squeeze any useful intel from them. If you can recruit them, even better."}}
 }
 FortConversationTaskNode_CompleteQuestObjectiveByBackendName_5 {
  id=#7
  class=FortConversationTaskNode_CompleteQuestObjectiveByBackendName
  outputs [#25]
  properties {objectiveBackendName=quest_s17_w10_legendary_q1_obj0, questDefinition=/BattlepassS17/Items/QuestItems/Legendary/quest_s17_w10_legendary_q1}
  sub_nodes [
   FortConversationRequirement_ControllerRequirement_14 {
    class=FortConversationRequirement_ControllerRequirement
    properties {
     nodeName=Require: Is Quest Active
     requirement {
      class=FortControllerRequirement_IsQuestActive
      properties {questDefinition=&quest_s17_w10_legendary_q1}
     }
    }
   }
   FortConversationRequirement_ControllerRequirement_15 {
    class=FortConversationRequirement_ControllerRequirement
    properties {
     nodeName=Require: Has Completed Quest
     requirement {
      class=FortControllerRequirement_HasCompletedQuest
      properties {bInverseCondition=true, questDefinition=&quest_s17_w10_legendary_q1}
     }
    }
   }
   FortConversationRequirement_ControllerRequirement_16 {
    class=FortConversationRequirement_ControllerRequirement
    properties {
     nodeName=Require: Has Completed Quest Objective
     requirement {
      class=FortControllerRequirement_HasCompletedQuestObjective
      properties {
       bInverseCondition=true
       objectiveBackendName=quest_s17_w10_legendary_q1_obj0
       questDefinition=&quest_s17_w10_legendary_q1
      }
     }
    }
   }
   ConversationChoiceNode_5 {
    class=ConversationChoiceNode
    properties {
     defaultChoiceDisplayText=NEW MISSION
     choiceTags {gameplayTags [{tagName=Conversation.Choice.Category.MainQuestline}, {tagName=Conversation.Choice.AutoSelectIfValid}]}
    }
   }
   ConvoSideEffect_PayphoneActor_Message_C_5 {class=ConvoSideEffect_PayphoneActor_Message_C, properties {message Number=10}}
  ]
 }
 Speech_C_20 {
  id=#25
  class=Speech_C
  outputs [#26]
  properties {
   nodeName=ShowWeek5Quest
   softTaskIcon=&T_UI_NPC_MainQuest
   generalConfig {defaultMessage="The rate of alien abductions is <b>far above</> expectations. "}
  }
 }
 Speech_C_21 {
  id=#26
  class=Speech_C
  outputs [#27]
  properties {generalConfig {defaultMessage=This isn't just for study.}}
 }
 Speech_C_22 {
  id=#27
  class=Speech_C
  outputs [#28]
  properties {generalConfig {defaultMessage="It's like they want to abduct <b>everyone</b> at least once."}}
 }
 Speech_C_23 {
  id=#28
  class=Speech_C
  outputs [#29]
  properties {generalConfig {defaultMessage=We need to know why.}}
 }
 Speech_C_24 {
  id=#29
  class=Speech_C
  outputs [#30]
  properties {generalConfig {defaultMessage=Learn everything you can about the abductions}}
 }
 Speech_C_25 {
  id=#30
  class=Speech_C
  outputs [#31]
  properties {generalConfig {defaultMessage=The why. The how. The who. And try not to get abducted too many times...}}
 }
 Speech_C_26 {
  id=#31
  class=Speech_C
  outputs []
  properties {generalConfig {defaultMessage=It's probably not great for your brain.}}
 }
 FortConversationTaskNode_CompleteQuestObjectiveByBackendName_3 {
  id=#6
  class=FortConversationTaskNode_CompleteQuestObjectiveByBackendName
  outputs [#32]
  properties {objectiveBackendName=quest_s17_w09_legendary_q1_obj0, questDefinition=/BattlepassS17/Items/QuestItems/Legendary/quest_s17_w09_legendary_q1}
  sub_nodes [
   FortConversationRequirement_ControllerRequirement_18 {
    class=FortConversationRequirement_ControllerRequirement
    properties {
     nodeName=Require: Is Quest Active
     requirement {
      class=FortControllerRequirement_IsQuestActive
      properties {questDefinition=&quest_s17_w09_legendary_q1}
     }
    }
   }
   FortConversationRequirement_ControllerRequirement_19 {
    class=FortConversationRequirement_ControllerRequirement
    properties {
     nodeName=Require: Has Completed Quest
     requirement {
      class=FortControllerRequirement_HasCompletedQuest
      properties {bInverseCondition=true, questDefinition=&quest_s17_w09_legendary_q1}
     }
    }
   }
   FortConversationRequirement_ControllerRequirement_20 {
    class=FortConversationRequirement_ControllerRequirement
    properties {
     nodeName=Require: Has Completed Quest Objective
     requirement {
      class=FortControllerRequirement_HasCompletedQuestObjective
      properties {
       bInverseCondition=true
       objectiveBackendName=quest_s17_w09_legendary_q1_obj0
       questDefinition=&quest_s17_w09_legendary_q1
      }
     }
    }
   }
   ConversationChoiceNode_7 {
    class=ConversationChoiceNode
    properties {
     defaultChoiceDisplayText=NEW MISSION
     choiceTags {gameplayTags [{tagName=Conversation.Choice.Category.MainQuestline}, {tagName=Conversation.Choice.AutoSelectIfValid}]}
    }
   }
   ConvoSideEffect_PayphoneActor_Message_C_7 {class=ConvoSideEffect_PayphoneActor_Message_C, properties {message Number=9}}
  ]
 }
 Speech_C_15 {
  id=#32
  class=Speech_C
  outputs [#33]
  properties {
   nodeName=ShowWeek5Quest
   softTaskIcon=&T_UI_NPC_MainQuest
   generalConfig {defaultMessage="We still have a <b>significant</> lack of knowledge about the alien invasion force."}
  }
 }
 Speech_C_16 {
  id=#33
  class=Speech_C
  outputs [#34]
  properties {generalConfig {defaultMessage="Knowledge isn't just power here... knowledge is <b>survival</>."}}
 }
 Speech_C_17 {
  id=#34
  class=Speech_C
  outputs [#35]
  properties {generalConfig {defaultMessage=We'll start by monitoring their activities on the Island.}}
 }
 Speech_C_18 {
  id=#35
  class=Speech_C
  outputs [#36]
  properties {generalConfig {defaultMessage="Then we turn our attention to the <b>mothership</>."}}
 }
 Speech_C_19 {
  id=#36
  class=Speech_C
  outputs []
  properties {generalConfig {defaultMessage="Uh-- Why don't you start by looking into those <enemy>strange billboards</>?"}}
 }
 FortConversationTaskNode_CompleteQuestObjectiveByBackendName_0 {
  id=#5
  class=FortConversationTaskNode_CompleteQuestObjectiveByBackendName
  outputs [#37]
  properties {objectiveBackendName=quest_s17_w08_legendary_q1_obj0, questDefinition=/BattlepassS17/Items/QuestItems/Legendary/quest_s17_w08_legendary_q1}
  sub_nodes [
   FortConversationRequirement_ControllerRequirement_4 {
    class=FortConversationRequirement_ControllerRequirement
    properties {
     nodeName=Require: Is Quest Active
     requirement {
      class=FortControllerRequirement_IsQuestActive
      properties {questDefinition=&quest_s17_w08_legendary_q1}
     }
    }
   }
   FortConversationRequirement_ControllerRequirement_6 {
    class=FortConversationRequirement_ControllerRequirement
    properties {
     nodeName=Require: Has Completed Quest
     requirement {
      class=FortControllerRequirement_HasCompletedQuest
      properties {bInverseCondition=true, questDefinition=&quest_s17_w08_legendary_q1}
     }
    }
   }
   FortConversationRequirement_ControllerRequirement_7 {
    class=FortConversationRequirement_ControllerRequirement
    properties {
     nodeName=Require: Has Completed Quest Objective
     requirement {
      class=FortControllerRequirement_HasCompletedQuestObjective
      properties {
       bInverseCondition=true
       objectiveBackendName=quest_s17_w08_legendary_q1_obj0
       questDefinition=&quest_s17_w08_legendary_q1
      }
     }
    }
   }
   ConversationChoiceNode_4 {
    class=ConversationChoiceNode
    properties {
     defaultChoiceDisplayText=NEW MISSION
     choiceTags {gameplayTags [{tagName=Conversation.Choice.Category.MainQuestline}, {tagName=Conversation.Choice.AutoSelectIfValid}]}
    }
   }
   ConvoSideEffect_PayphoneActor_Message_C_4 {class=ConvoSideEffect_PayphoneActor_Message_C, properties {message Number=8}}
  ]
 }
 Speech_C_10 {
  id=#37
  class=Speech_C
  outputs [#38]
  properties {
   nodeName=ShowWeek5Quest
   softTaskIcon=&T_UI_NPC_MainQuest
   generalConfig {defaultMessage="We just had a <b>major</> IO security breach. "}
  }
 }
 Speech_C_11 {
  id=#38
  class=Speech_C
  outputs [#39]
  properties {generalConfig {defaultMessage="I need to investigate, but I <b>can't let anyone know</> I'm aware of what happened. "}}
 }
 Speech_C_12 {
  id=#39
  class=Speech_C
  outputs [#40]
  properties {generalConfig {defaultMessage="Find me a <enemy>Field Team access card. </>"}}
 }
 Speech_C_13 {
  id=#40
  class=Speech_C
  outputs [#41]
  properties {generalConfig {defaultMessage="Do <b>whatever</> needs to be done. "}}
 }
 Speech_C_14 {
  id=#41
  class=Speech_C
  outputs []
  properties {generalConfig {defaultMessage="All of <b>our lives</> depend on you finding that <enemy>card. </>"}}
 }
 FortConversationTaskNode_CompleteQuestObjectiveByBackendName_2 {
  id=#4
  class=FortConversationTaskNode_CompleteQuestObjectiveByBackendName
  outputs [#42]
  properties {objectiveBackendName=quest_s17_w06_legendary_q1_obj0, questDefinition=/BattlepassS17/Items/QuestItems/Legendary/quest_s17_w06_legendary_q1}
  sub_nodes [
   FortConversationRequirement_ControllerRequirement_1 {
    class=FortConversationRequirement_ControllerRequirement
    properties {
     nodeName=Require: Is Quest Active
     requirement {
      class=FortControllerRequirement_IsQuestActive
      properties {questDefinition=&quest_s17_w06_legendary_q1}
     }
    }
   }
   FortConversationRequirement_ControllerRequirement_2 {
    class=FortConversationRequirement_ControllerRequirement
    properties {
     nodeName=Require: Has Completed Quest
     requirement {
      class=FortControllerRequirement_HasCompletedQuest
      properties {bInverseCondition=true, questDefinition=&quest_s17_w06_legendary_q1}
     }
    }
   }
   FortConversationRequirement_ControllerRequirement_3 {
    class=FortConversationRequirement_ControllerRequirement
    properties {
     nodeName=Require: Has Completed Quest Objective
     requirement {
      class=FortControllerRequirement_HasCompletedQuestObjective
      properties {
       bInverseCondition=true
       objectiveBackendName=quest_s17_w06_legendary_q1_obj0
       questDefinition=&quest_s17_w06_legendary_q1
      }
     }
    }
   }
   ConversationChoiceNode_3 {
    class=ConversationChoiceNode
    properties {
     defaultChoiceDisplayText=NEW MISSION
     choiceTags {gameplayTags [{tagName=Conversation.Choice.Category.MainQuestline}, {tagName=Conversation.Choice.AutoSelectIfValid}]}
    }
   }
   ConvoSideEffect_PayphoneActor_Message_C_1 {class=ConvoSideEffect_PayphoneActor_Message_C, properties {message Number=6}}
  ]
 }
 Speech_C_2 {
  id=#42
  class=Speech_C
  outputs [#43]
  properties {
   nodeName=ShowWeek5Quest
   softTaskIcon=&T_UI_NPC_MainQuest
   generalConfig {defaultMessage="<enemy>The abduction situation</> is hitting a <b>crisis point. </> We need to get a handle on it yesterday."}
  }
 }
 Speech_C_8 {
  id=#43
  class=Speech_C
  outputs [#44]
  properties {generalConfig {defaultMessage="We need a boots-on-the-ground presence on <b>Hayseed's farm.</>  \r\nWe need him <b>alive and on our side. </>"}}
 }
 Speech_C_9 {
  id=#44
  class=Speech_C
  outputs []
  properties {generalConfig {defaultMessage="Maybe you could \r\n<b>bribe him with some supplies.</>\r\nThe guy's a bit of a hoarder."}}
 }
 FortConversationTaskNode_CompleteQuestObjectiveByBackendName_1 {
  id=#3
  class=FortConversationTaskNode_CompleteQuestObjectiveByBackendName
  outputs [#45]
  properties {objectiveBackendName=quest_s17_w05_legendary_q1_obj0, questDefinition=/BattlepassS17/Items/QuestItems/Legendary/quest_s17_w05_legendary_q1}
  sub_nodes [
   FortConversationRequirement_ControllerRequirement_9 {
    class=FortConversationRequirement_ControllerRequirement
    properties {
     nodeName=Require: Is Quest Active
     requirement {
      class=FortControllerRequirement_IsQuestActive
      properties {questDefinition=&quest_s17_w05_legendary_q1}
     }
    }
   }
   ConvoSideEffect_PayphoneActor_Message_C_3 {class=ConvoSideEffect_PayphoneActor_Message_C, properties {message Number=5}}
   ConversationChoiceNode_1 {
    class=ConversationChoiceNode
    properties {
     defaultChoiceDisplayText=NEW MISSION
     choiceTags {gameplayTags [{tagName=Conversation.Choice.Category.MainQuestline}, {tagName=Conversation.Choice.AutoSelectIfValid}]}
    }
   }
   FortConversationRequirement_ControllerRequirement_10 {
    class=FortConversationRequirement_ControllerRequirement
    properties {
     nodeName=Require: Has Completed Quest Objective
     requirement {
      class=FortControllerRequirement_HasCompletedQuestObjective
      properties {
       bInverseCondition=true
       objectiveBackendName=quest_s17_w05_legendary_q1_obj0
       questDefinition=&quest_s17_w05_legendary_q1
      }
     }
    }
   }
   FortConversationRequirement_ControllerRequirement_11 {
    class=FortConversationRequirement_ControllerRequirement
    properties {
     nodeName=Require: Has Completed Quest
     requirement {
      class=FortControllerRequirement_HasCompletedQuest
      properties {bInverseCondition=true, questDefinition=&quest_s17_w05_legendary_q1}
     }
    }
   }
  ]
 }
 Speech_C_0 {
  id=#45
  class=Speech_C
  outputs [#46]
  properties {
   nodeName=ShowWeek5Quest
   softTaskIcon=&T_UI_NPC_MainQuest
   generalConfig {defaultMessage="Welcome back, agent. I need you to focus on <enemy>Sunny.</>"}
  }
 }
 Speech_C_4 {
  id=#46
  class=Speech_C
  outputs [#47]
  properties {generalConfig {defaultMessage="I think she may be behind the uptick in <b>anti-human sentiment</> on the Island."}}
 }
 Speech_C_6 {
  id=#47
  class=Speech_C
  outputs [#48]
  properties {generalConfig {defaultMessage="All I know about her is that she's a big fan of that radio show, uh...\r\n<b> \"Hot Saucers.\"</>"}}
 }
 Speech_C_7 {
  id=#48
  class=Speech_C
  outputs []
  properties {generalConfig {defaultMessage="Catch up on the show. Learn the lingo. \r\n<b>Earn her trust.</>"}}
 }
 FortConversationTaskNode_CompleteQuestObjectiveByBackendName_4 {
  id=#2
  class=FortConversationTaskNode_CompleteQuestObjectiveByBackendName
  outputs [#49]
  properties {objectiveBackendName=quest_s17_w03_legendary_q1_obj0, questDefinition=/BattlepassS17/Items/QuestItems/Legendary/quest_s17_w03_legendary_q1}
  sub_nodes [
   FortConversationRequirement_ControllerRequirement_0 {
    class=FortConversationRequirement_ControllerRequirement
    properties {
     nodeName=Require: Is Quest Active
     requirement {
      class=FortControllerRequirement_IsQuestActive
      properties {questDefinition=&quest_s17_w03_legendary_q1}
     }
    }
   }
   FortConversationRequirement_ControllerRequirement_5 {
    class=FortConversationRequirement_ControllerRequirement
    properties {
     nodeName=Require: Has Completed Quest
     requirement {
      class=FortControllerRequirement_HasCompletedQuest
      properties {bInverseCondition=true, questDefinition=&quest_s17_w03_legendary_q1}
     }
    }
   }
   FortConversationRequirement_ControllerRequirement_8 {
    class=FortConversationRequirement_ControllerRequirement
    properties {
     nodeName=Require: Has Completed Quest Objective
     requirement {
      class=FortControllerRequirement_HasCompletedQuestObjective
      properties {
       bInverseCondition=true
       objectiveBackendName=quest_s17_w03_legendary_q1_obj0
       questDefinition=&quest_s17_w03_legendary_q1
      }
     }
    }
   }
   ConversationChoiceNode_0 {
    class=ConversationChoiceNode
    properties {
     defaultChoiceDisplayText=NEW MISSION
     choiceTags {gameplayTags [{tagName=Conversation.Choice.Category.MainQuestline}, {tagName=Conversation.Choice.AutoSelectIfValid}]}
    }
   }
   ConvoSideEffect_PayphoneActor_Message_C_2 {class=ConvoSideEffect_PayphoneActor_Message_C, properties {message Number=3}}
  ]
 }
 Speech_C_1 {
  id=#49
  class=Speech_C
  outputs [#50]
  properties {
   nodeName=ShowWeek3Quest
   softTaskIcon=&T_UI_NPC_MainQuest
   generalConfig {defaultMessage="Listen up. Your mission is to infiltrate <enemy>Believer Beach.</>\r\nFind the ringleader and <b>win them over</>"}
  }
 }
 Speech_C_3 {
  id=#50
  class=Speech_C
  outputs []
  properties {generalConfig {defaultMessage="<b>Earn his trust</> and learn \r\neverything you can.\r\nThis level of alien support <enemy>isn't natural</>"}}
 }
]
