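# Shared NPC conversation database for the Season 17 Week 1 legendary quest, giving five NPCs (Believer, Abstrakt, Dreamflower, Riot, Bushranger) their "Aliens?!?" dialogue.
# One entry point fans out to five branches, each gated on the active, uncompleted quest, that NPC's unfinished objective and its conversation entry tag.
# Auto-selected choices start short speech chains, each completing its objective (obj0 to obj4) when the quest is not ready to turn in.
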ConversationDatabase /BattlepassS17/Conversations/Shared_Conversation_Week1
entry_points [{tag=Conversation.Entry.Season17.Week1, destinations [#1]}]
nodes [
 ConversationEntryPointNode_7 {
  id=#1
  class=ConversationEntryPointNode
  outputs [
   #2
   #3
   #4
   #5
   #6
  ]
  properties {entryTag {tagName=Conversation.Entry.Season17.Week1}}
 }
 Speech_C_58 {
  id=#6
  class=Speech_C
  outputs [#7]
  properties {
   softTaskIcon=&T_UI_NPC_MainQuest
   generalConfig {defaultMessage="<b>ShinyShiny</> rock! \r\nAliens make them SOOOO shiny. "}
  }
  sub_nodes [
   ConversationChoiceNode_58 {
    class=ConversationChoiceNode
    properties {
     defaultChoiceDisplayText="Aliens?!?"
     choiceTags {gameplayTags [{tagName=Conversation.Choice.Category.MainQuestline}, {tagName=Conversation.Choice.AutoSelectIfValid}]}
    }
   }
   FortConversationRequirement_ControllerRequirement_72 {
    class=FortConversationRequirement_ControllerRequirement
    properties {
     nodeName=Require: Has Completed Quest
     requirement {
      class=FortControllerRequirement_HasCompletedQuest
      properties {bInverseCondition=true, questDefinition=/BattlepassS17/Items/QuestItems/Legendary/quest_s17_w01_legendary_q1}
     }
    }
   }
   FortConversationRequirement_ControllerRequirement_73 {
    class=FortConversationRequirement_ControllerRequirement
    properties {
     nodeName=Require: Is Quest Active
     requirement {
      class=FortControllerRequirement_IsQuestActive
      properties {questDefinition=&quest_s17_w01_legendary_q1}
     }
    }
   }
   FortConversationRequirement_ControllerRequirement_74 {
    class=FortConversationRequirement_ControllerRequirement
    properties {
     nodeName=Require: Has Completed Quest Objective
     requirement {
      class=FortControllerRequirement_HasCompletedQuestObjective
      properties {
       bInverseCondition=true
       objectiveBackendName=quest_s17_w01_legendary_q1_obj4
       questDefinition=&quest_s17_w01_legendary_q1
      }
     }
    }
   }
   FortConversationRequirement_ControllerRequirement_75 {
    class=FortConversationRequirement_ControllerRequirement
    properties {
     nodeName=Require: NPC Has Conversation Tag
     participantId {tagName=Conversation.Participant.NPC}
     requirement {class=FortControllerRequirement_NPCHasConversationEntryTag, properties {queryTag {tagName=Conversation.Entry.Bushranger}}}
    }
   }
  ]
 }
 Speech_C_59 {
  id=#7
  class=Speech_C
  outputs [#8]
  properties {generalConfig {defaultMessage="Yaay imagination! "}}
  sub_nodes [ConversationChoiceNode_59 {class=ConversationChoiceNode, properties {defaultChoiceDisplayText=Continue}}]
 }
 FortConversationTaskNode_CompleteQuestObjectiveByBackendName_12 {
  id=#8
  class=FortConversationTaskNode_CompleteQuestObjectiveByBackendName
  properties {objectiveBackendName=quest_s17_w01_legendary_q1_obj4, questDefinition=&quest_s17_w01_legendary_q1}
  sub_nodes [FortConversationRequirement_ControllerRequirement_76 {
    class=FortConversationRequirement_ControllerRequirement
    properties {
     nodeName=Require: Has Quest To Turn In
     requirement {
      class=FortControllerRequirement_HasQuestToTurnIn
      properties {bInverseCondition=true, softQuestToTurnIn=&quest_s17_w01_legendary_q1}
     }
    }
   }]
 }
 Speech_C_56 {
  id=#5
  class=Speech_C
  outputs [#9]
  properties {
   softTaskIcon=&T_UI_NPC_MainQuest
   generalConfig {defaultMessage="Aliens blast <enemy>the Spire</> into <b>shiny rocks</> and then bang! \r\nThere's the IO"}
  }
  sub_nodes [
   ConversationChoiceNode_60 {
    class=ConversationChoiceNode
    properties {
     defaultChoiceDisplayText="Aliens?!?"
     choiceTags {gameplayTags [{tagName=Conversation.Choice.Category.MainQuestline}, {tagName=Conversation.Choice.AutoSelectIfValid}]}
    }
   }
   FortConversationRequirement_ControllerRequirement_77 {
    class=FortConversationRequirement_ControllerRequirement
    properties {
     nodeName=Require: Has Completed Quest
     requirement {
      class=FortControllerRequirement_HasCompletedQuest
      properties {bInverseCondition=true, questDefinition=&quest_s17_w01_legendary_q1}
     }
    }
   }
   FortConversationRequirement_ControllerRequirement_78 {
    class=FortConversationRequirement_ControllerRequirement
    properties {
     nodeName=Require: Is Quest Active
     requirement {
      class=FortControllerRequirement_IsQuestActive
      properties {questDefinition=&quest_s17_w01_legendary_q1}
     }
    }
   }
   FortConversationRequirement_ControllerRequirement_79 {
    class=FortConversationRequirement_ControllerRequirement
    properties {
     nodeName=Require: Has Completed Quest Objective
     requirement {
      class=FortControllerRequirement_HasCompletedQuestObjective
      properties {
       bInverseCondition=true
       objectiveBackendName=quest_s17_w01_legendary_q1_obj3
       questDefinition=&quest_s17_w01_legendary_q1
      }
     }
    }
   }
   FortConversationRequirement_ControllerRequirement_80 {
    class=FortConversationRequirement_ControllerRequirement
    properties {
     nodeName=Require: NPC Has Conversation Tag
     participantId {tagName=Conversation.Participant.NPC}
     requirement {class=FortControllerRequirement_NPCHasConversationEntryTag, properties {queryTag {tagName=Conversation.Entry.Riot}}}
    }
   }
  ]
 }
 Speech_C_57 {
  id=#9
  class=Speech_C
  outputs [#10]
  properties {generalConfig {defaultMessage="Walkin around with weapons that act just like <b>that ship's laser. </>"}}
  sub_nodes [ConversationChoiceNode_61 {class=ConversationChoiceNode, properties {defaultChoiceDisplayText=Continue}}]
 }
 FortConversationTaskNode_CompleteQuestObjectiveByBackendName_11 {
  id=#10
  class=FortConversationTaskNode_CompleteQuestObjectiveByBackendName
  properties {objectiveBackendName=quest_s17_w01_legendary_q1_obj3, questDefinition=&quest_s17_w01_legendary_q1}
  sub_nodes [FortConversationRequirement_ControllerRequirement_81 {
    class=FortConversationRequirement_ControllerRequirement
    properties {
     nodeName=Require: Has Quest To Turn In
     requirement {
      class=FortControllerRequirement_HasQuestToTurnIn
      properties {bInverseCondition=true, softQuestToTurnIn=&quest_s17_w01_legendary_q1}
     }
    }
   }]
 }
 Speech_C_42 {
  id=#4
  class=Speech_C
  outputs [#11]
  properties {
   softTaskIcon=&T_UI_NPC_MainQuest
   generalConfig {defaultMessage="I've <b>seen aliens all my life</>, but now it's like I'm seeing them again for the first time?"}
  }
  sub_nodes [
   ConversationChoiceNode_7 {
    class=ConversationChoiceNode
    properties {
     defaultChoiceDisplayText="Aliens?!?"
     choiceTags {gameplayTags [{tagName=Conversation.Choice.Category.MainQuestline}, {tagName=Conversation.Choice.AutoSelectIfValid}]}
    }
   }
   FortConversationRequirement_ControllerRequirement_8 {
    class=FortConversationRequirement_ControllerRequirement
    properties {
     nodeName=Require: Has Completed Quest
     requirement {
      class=FortControllerRequirement_HasCompletedQuest
      properties {bInverseCondition=true, questDefinition=&quest_s17_w01_legendary_q1}
     }
    }
   }
   FortConversationRequirement_ControllerRequirement_9 {
    class=FortConversationRequirement_ControllerRequirement
    properties {
     nodeName=Require: Is Quest Active
     requirement {
      class=FortControllerRequirement_IsQuestActive
      properties {questDefinition=&quest_s17_w01_legendary_q1}
     }
    }
   }
   FortConversationRequirement_ControllerRequirement_10 {
    class=FortConversationRequirement_ControllerRequirement
    properties {
     nodeName=Require: Has Completed Quest Objective
     requirement {
      class=FortControllerRequirement_HasCompletedQuestObjective
      properties {
       bInverseCondition=true
       objectiveBackendName=quest_s17_w01_legendary_q1_obj2
       questDefinition=&quest_s17_w01_legendary_q1
      }
     }
    }
   }
   FortConversationRequirement_ControllerRequirement_52 {
    class=FortConversationRequirement_ControllerRequirement
    properties {
     nodeName=Require: NPC Has Conversation Tag
     participantId {tagName=Conversation.Participant.NPC}
     requirement {class=FortControllerRequirement_NPCHasConversationEntryTag, properties {queryTag {tagName=Conversation.Entry.Dreamflower}}}
    }
   }
  ]
 }
 Speech_C_55 {
  id=#11
  class=Speech_C
  outputs [#12]
  properties {generalConfig {defaultMessage="It's like <b>falling in love all over again, </>but also like seeing the doctor,"}}
  sub_nodes [ConversationChoiceNode_5 {class=ConversationChoiceNode, properties {defaultChoiceDisplayText=Continue}}]
 }
 Speech_C_3 {
  id=#12
  class=Speech_C
  outputs [#13]
  properties {generalConfig {defaultMessage="Because they keep taking <b>blood samples?</>"}}
  sub_nodes [ConversationChoiceNode_24 {class=ConversationChoiceNode, properties {defaultChoiceDisplayText=Continue}}]
 }
 FortConversationTaskNode_CompleteQuestObjectiveByBackendName_10 {
  id=#13
  class=FortConversationTaskNode_CompleteQuestObjectiveByBackendName
  properties {objectiveBackendName=quest_s17_w01_legendary_q1_obj2, questDefinition=&quest_s17_w01_legendary_q1}
  sub_nodes [FortConversationRequirement_ControllerRequirement_0 {
    class=FortConversationRequirement_ControllerRequirement
    properties {
     nodeName=Require: Has Quest To Turn In
     requirement {
      class=FortControllerRequirement_HasQuestToTurnIn
      properties {bInverseCondition=true, softQuestToTurnIn=&quest_s17_w01_legendary_q1}
     }
    }
   }]
 }
 Speech_C_11 {
  id=#3
  class=Speech_C
  outputs [#14]
  properties {
   softTaskIcon=&T_UI_NPC_MainQuest
   generalConfig {defaultMessage="I had some next level designs in mind for that <b>spire...</>"}
  }
  sub_nodes [
   ConversationChoiceNode_44 {
    class=ConversationChoiceNode
    properties {
     defaultChoiceDisplayText="Aliens?!?"
     choiceTags {gameplayTags [{tagName=Conversation.Choice.Category.MainQuestline}, {tagName=Conversation.Choice.AutoSelectIfValid}]}
    }
   }
   FortConversationRequirement_ControllerRequirement_67 {
    class=FortConversationRequirement_ControllerRequirement
    properties {
     nodeName=Require: Has Completed Quest
     requirement {
      class=FortControllerRequirement_HasCompletedQuest
      properties {bInverseCondition=true, questDefinition=&quest_s17_w01_legendary_q1}
     }
    }
   }
   FortConversationRequirement_ControllerRequirement_68 {
    class=FortConversationRequirement_ControllerRequirement
    properties {
     nodeName=Require: Is Quest Active
     requirement {
      class=FortControllerRequirement_IsQuestActive
      properties {questDefinition=&quest_s17_w01_legendary_q1}
     }
    }
   }
   FortConversationRequirement_ControllerRequirement_69 {
    class=FortConversationRequirement_ControllerRequirement
    properties {
     nodeName=Require: Has Completed Quest Objective
     requirement {
      class=FortControllerRequirement_HasCompletedQuestObjective
      properties {
       bInverseCondition=true
       objectiveBackendName=quest_s17_w01_legendary_q1_obj1
       questDefinition=&quest_s17_w01_legendary_q1
      }
     }
    }
   }
   FortConversationRequirement_ControllerRequirement_70 {
    class=FortConversationRequirement_ControllerRequirement
    properties {
     nodeName=Require: NPC Has Conversation Tag
     participantId {tagName=Conversation.Participant.NPC}
     requirement {class=FortControllerRequirement_NPCHasConversationEntryTag, properties {queryTag {tagName=Conversation.Entry.Abstrakt}}}
    }
   }
  ]
 }
 Speech_C_41 {
  id=#14
  class=Speech_C
  outputs [#15]
  properties {generalConfig {defaultMessage="Now it's nothing but piles of <b>shiny rubble.</>"}}
  sub_nodes [ConversationChoiceNode_4 {class=ConversationChoiceNode, properties {defaultChoiceDisplayText=Continue}}]
 }
 Speech_C_2 {
  id=#15
  class=Speech_C
  outputs [#16]
  properties {generalConfig {defaultMessage="Paint won't even stick to it, \r\nso there goes the Rubble Artist side hustle. "}}
  sub_nodes [ConversationChoiceNode_57 {class=ConversationChoiceNode, properties {defaultChoiceDisplayText=Continue}}]
 }
 FortConversationTaskNode_CompleteQuestObjectiveByBackendName_6 {
  id=#16
  class=FortConversationTaskNode_CompleteQuestObjectiveByBackendName
  properties {objectiveBackendName=quest_s17_w01_legendary_q1_obj1, questDefinition=&quest_s17_w01_legendary_q1}
  sub_nodes [FortConversationRequirement_ControllerRequirement_71 {
    class=FortConversationRequirement_ControllerRequirement
    properties {
     nodeName=Require: Has Quest To Turn In
     requirement {
      class=FortControllerRequirement_HasQuestToTurnIn
      properties {bInverseCondition=true, softQuestToTurnIn=&quest_s17_w01_legendary_q1}
     }
    }
   }]
 }
 Speech_C_1 {
  id=#2
  class=Speech_C
  outputs [#17]
  properties {
   softTaskIcon=&T_UI_NPC_MainQuest
   generalConfig {defaultMessage="Oh <b>good.</> \r\nThe bunker guy knows I exist... \r\nThat's <b>exactly</> what I wanted. "}
  }
  sub_nodes [
   ConversationChoiceNode_20 {
    class=ConversationChoiceNode
    properties {
     defaultChoiceDisplayText="Aliens?!?"
     choiceTags {gameplayTags [{tagName=Conversation.Choice.Category.MainQuestline}, {tagName=Conversation.Choice.AutoSelectIfValid}]}
    }
   }
   FortConversationRequirement_ControllerRequirement_40 {
    class=FortConversationRequirement_ControllerRequirement
    properties {
     nodeName=Require: Has Completed Quest
     requirement {
      class=FortControllerRequirement_HasCompletedQuest
      properties {bInverseCondition=true, questDefinition=&quest_s17_w01_legendary_q1}
     }
    }
   }
   FortConversationRequirement_ControllerRequirement_41 {
    class=FortConversationRequirement_ControllerRequirement
    properties {
     nodeName=Require: Is Quest Active
     requirement {
      class=FortControllerRequirement_IsQuestActive
      properties {questDefinition=&quest_s17_w01_legendary_q1}
     }
    }
   }
   FortConversationRequirement_ControllerRequirement_42 {
    class=FortConversationRequirement_ControllerRequirement
    properties {
     nodeName=Require: Has Completed Quest Objective
     requirement {
      class=FortControllerRequirement_HasCompletedQuestObjective
      properties {
       bInverseCondition=true
       objectiveBackendName=quest_s17_w01_legendary_q1_obj0
       questDefinition=&quest_s17_w01_legendary_q1
      }
     }
    }
   }
   FortConversationRequirement_ControllerRequirement_58 {
    class=FortConversationRequirement_ControllerRequirement
    properties {
     nodeName=Require: NPC Has Conversation Tag
     participantId {tagName=Conversation.Participant.NPC}
     requirement {class=FortControllerRequirement_NPCHasConversationEntryTag, properties {queryTag {tagName=Conversation.Entry.Believer}}}
    }
   }
  ]
 }
 Speech_C_7 {
  id=#17
  class=Speech_C
  outputs [#18]
  properties {generalConfig {defaultMessage="Yeah. I've seens some things.\r\nBut I'm not telling you <b>anything. </>"}}
  sub_nodes [ConversationChoiceNode_3 {class=ConversationChoiceNode, properties {defaultChoiceDisplayText=Continue}}]
 }
 Speech_C_0 {
  id=#18
  class=Speech_C
  outputs [#19]
  properties {generalConfig {defaultMessage="I don't work with <b>people. </>"}}
  sub_nodes [ConversationChoiceNode_23 {class=ConversationChoiceNode, properties {defaultChoiceDisplayText=Continue}}]
 }
 FortConversationTaskNode_CompleteQuestObjectiveByBackendName_0 {
  id=#19
  class=FortConversationTaskNode_CompleteQuestObjectiveByBackendName
  properties {objectiveBackendName=quest_s17_w01_legendary_q1_obj0, questDefinition=&quest_s17_w01_legendary_q1}
  sub_nodes [FortConversationRequirement_ControllerRequirement_66 {
    class=FortConversationRequirement_ControllerRequirement
    properties {
     nodeName=Require: Has Quest To Turn In
     requirement {
      class=FortControllerRequirement_HasQuestToTurnIn
      properties {bInverseCondition=true, softQuestToTurnIn=&quest_s17_w01_legendary_q1}
     }
    }
   }]
 }
]
