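# Season 17 Week 11 shared conversation in which the player gives Faux a cat food gift basket for legendary quest 5.
# The "Give gift basket" choice auto-selects if valid, gated on the quest being active, quest and objective incomplete, and the NPC carrying the Faux entry tag.
# A linear chain of speech nodes with Continue choices reveals the plan to annihilate humans, then a task completes the quest objective by backend name.
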ConversationDatabase /BattlepassS17/Conversations/Shared_Conversation_Week11
entry_points [{tag=Conversation.Entry.Season17.Week11, destinations [#1]}]
nodes [
 ConversationEntryPointNode_7 {
  id=#1
  class=ConversationEntryPointNode
  outputs [#2]
  properties {entryTag {tagName=Conversation.Entry.Season17.Week11}}
 }
 Speech_C_2 {
  id=#2
  class=Speech_C
  outputs [#3]
  properties {
   softTaskIcon=&T_UI_NPC_MainQuest
   generalConfig {defaultMessage="Liver and cheese flavored cat food? For me?! That's... so kind."}
  }
  sub_nodes [
   ConversationChoiceNode_26 {
    class=ConversationChoiceNode
    properties {
     defaultChoiceDisplayText=Give gift basket
     choiceTags {gameplayTags [{tagName=Conversation.Choice.Category.MainQuestline}, {tagName=Conversation.Choice.AutoSelectIfValid}]}
    }
   }
   FortConversationRequirement_ControllerRequirement_19 {
    class=FortConversationRequirement_ControllerRequirement
    properties {
     nodeName=Require: Has Completed Quest
     requirement {
      class=FortControllerRequirement_HasCompletedQuest
      properties {bInverseCondition=true, questDefinition=/BattlepassS17/Items/QuestItems/Legendary/quest_s17_w11_legendary_q5}
     }
    }
   }
   FortConversationRequirement_ControllerRequirement_20 {
    class=FortConversationRequirement_ControllerRequirement
    properties {
     nodeName=Require: Is Quest Active
     requirement {
      class=FortControllerRequirement_IsQuestActive
      properties {questDefinition=&quest_s17_w11_legendary_q5}
     }
    }
   }
   FortConversationRequirement_ControllerRequirement_21 {
    class=FortConversationRequirement_ControllerRequirement
    properties {
     nodeName=Require: Has Completed Quest Objective
     requirement {
      class=FortControllerRequirement_HasCompletedQuestObjective
      properties {
       bInverseCondition=true
       objectiveBackendName=quest_s17_w11_legendary_q5_obj0
       questDefinition=&quest_s17_w11_legendary_q5
      }
     }
    }
   }
   FortConversationRequirement_ControllerRequirement_22 {
    class=FortConversationRequirement_ControllerRequirement
    properties {
     nodeName=Require: NPC Has Conversation Tag
     participantId {tagName=Conversation.Participant.NPC}
     requirement {class=FortControllerRequirement_NPCHasConversationEntryTag, properties {queryTag {tagName=Conversation.Entry.Faux}}}
    }
   }
  ]
 }
 Speech_C_7 {
  id=#3
  class=Speech_C
  outputs [#4]
  properties {generalConfig {defaultMessage="Okay, I'll level with you: The plan is to annihilate you all. "}}
  sub_nodes [ConversationChoiceNode_4 {class=ConversationChoiceNode, properties {defaultChoiceDisplayText=Continue}}]
 }
 Speech_C_6 {
  id=#4
  class=Speech_C
  outputs [#5]
  properties {generalConfig {defaultMessage="Believe me, I wish it wasn't true. I think you humans are fascinating. "}}
  sub_nodes [ConversationChoiceNode_6 {class=ConversationChoiceNode, properties {defaultChoiceDisplayText=Continue}}]
 }
 Speech_C_4 {
  id=#5
  class=Speech_C
  outputs [#6]
  properties {generalConfig {defaultMessage="You're so squishy and weird looking. And your \"cat food\" delicacy is delicious. "}}
  sub_nodes [ConversationChoiceNode_5 {class=ConversationChoiceNode, properties {defaultChoiceDisplayText=Continue}}]
 }
 Speech_C_0 {
  id=#6
  class=Speech_C
  outputs [#7]
  properties {generalConfig {defaultMessage="I'll do what I can to nudge my people in a different direction. But no promises. "}}
  sub_nodes [ConversationChoiceNode_7 {class=ConversationChoiceNode, properties {defaultChoiceDisplayText=Continue}}]
 }
 FortConversationTaskNode_CompleteQuestObjectiveByBackendName_1 {
  id=#7
  class=FortConversationTaskNode_CompleteQuestObjectiveByBackendName
  properties {objectiveBackendName=quest_s17_w11_legendary_q5_obj0, questDefinition=&quest_s17_w11_legendary_q5}
 }
]
