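# Shared Season 17 Week 14 conversation database that adds main-questline NPC dialogue for legendary quests, mostly the "Impending Doom" warnings for q1b.
# One entry point fans out to per-NPC speech branches, each gated on active quest, incomplete objective and NPC conversation tag, and tagged to auto-select if valid.
# Branches play one or more lines, then complete their objective by backend name.
# Maven's "Mole" branch (q5) runs a ChallengeNPC task before completion.
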ConversationDatabase /BattlepassS17/Conversations/Shared_Conversation_Week14
entry_points [{tag=Conversation.Entry.Season17.Week14, destinations [#1]}]
nodes [
 ConversationEntryPointNode_7 {
  id=#1
  class=ConversationEntryPointNode
  outputs [
   #2
   #3
   #4
   #5
   #6
   #7
   #8
   #9
   #10
   #11
   #12
   #13
   #14
   #15
   #16
   #17
   #18
   #19
   #20
   #21
   #22
  ]
  properties {entryTag {tagName=Conversation.Entry.Season17.Week14}}
 }
 Speech_C_0 {
  id=#22
  class=Speech_C
  outputs [#23]
  properties {
   softTaskIcon=&T_UI_NPC_MainQuest
   generalConfig {defaultMessage="If it isn't  <enemy>Slone's newest lapdog.</> That  <enemy>warmongering charlatan</> figured me out, huh?"}
  }
  sub_nodes [
   ConversationChoiceNode_102 {
    class=ConversationChoiceNode
    properties {
     defaultChoiceDisplayText=Mole
     choiceTags {gameplayTags [{tagName=Conversation.Choice.Category.MainQuestline}, {tagName=Conversation.Choice.AutoSelectIfValid}]}
    }
   }
   FortConversationRequirement_ControllerRequirement_77 {
    class=FortConversationRequirement_ControllerRequirement
    properties {
     nodeName=Require: Has Completed Quest
     requirement {
      class=FortControllerRequirement_HasCompletedQuest
      properties {bInverseCondition=true, questDefinition=/BattlepassS17/Items/QuestItems/Legendary/quest_s17_w14_legendary_q5}
     }
    }
   }
   FortConversationRequirement_ControllerRequirement_78 {
    class=FortConversationRequirement_ControllerRequirement
    properties {
     nodeName=Require: Is Quest Active
     requirement {
      class=FortControllerRequirement_IsQuestActive
      properties {questDefinition=&quest_s17_w14_legendary_q5}
     }
    }
   }
   FortConversationRequirement_ControllerRequirement_79 {
    class=FortConversationRequirement_ControllerRequirement
    properties {
     nodeName=Require: Has Completed Quest Objective
     requirement {
      class=FortControllerRequirement_HasCompletedQuestObjective
      properties {
       bInverseCondition=true
       objectiveBackendName=quest_s17_w14_legendary_q5_obj0
       questDefinition=&quest_s17_w14_legendary_q5
      }
     }
    }
   }
   FortConversationRequirement_ControllerRequirement_80 {
    class=FortConversationRequirement_ControllerRequirement
    properties {
     nodeName=Require: NPC Has Conversation Tag
     participantId {tagName=Conversation.Participant.NPC}
     requirement {class=FortControllerRequirement_NPCHasConversationEntryTag, properties {queryTag {tagName=Conversation.Entry.Maven}}}
    }
   }
  ]
 }
 Speech_C_3 {
  id=#23
  class=Speech_C
  outputs [#24]
  properties {generalConfig {defaultMessage="Well <enemy>I'm done</> designing inelegant tools of war. There's a whole world out there to study."}}
  sub_nodes [ConversationChoiceNode_7 {class=ConversationChoiceNode, properties {defaultChoiceDisplayText=Continue}}]
 }
 Speech_C_4 {
  id=#24
  class=Speech_C
  outputs [#25]
  properties {generalConfig {defaultMessage="The Aliens care about science, and math, and... well, yeah,  <enemy>destruction.</>"}}
  sub_nodes [ConversationChoiceNode_8 {class=ConversationChoiceNode, properties {defaultChoiceDisplayText=Continue}}]
 }
 Speech_C_6 {
  id=#25
  class=Speech_C
  outputs [#26]
  properties {generalConfig {defaultMessage="But <b>I can convince them</> to save this world. Continue my meaningful work."}}
  sub_nodes [ConversationChoiceNode_9 {class=ConversationChoiceNode, properties {defaultChoiceDisplayText=Continue}}]
 }
 Speech_C_7 {
  id=#26
  class=Speech_C
  outputs [#27]
  properties {generalConfig {defaultMessage="I just need to  <enemy>get you out of my way.</>"}}
  sub_nodes [ConversationChoiceNode_4 {class=ConversationChoiceNode, properties {defaultChoiceDisplayText=Continue}}]
 }
 ChallengeNPC_C_0 {id=#27, class=ChallengeNPC_C, outputs [#28]}
 FortConversationTaskNode_CompleteQuestObjectiveByBackendName_4 {
  id=#28
  class=FortConversationTaskNode_CompleteQuestObjectiveByBackendName
  properties {objectiveBackendName=quest_s17_w14_legendary_q5_obj0, questDefinition=&quest_s17_w14_legendary_q5}
 }
 Speech_C_130 {
  id=#21
  class=Speech_C
  outputs [#29]
  properties {
   softTaskIcon=&T_UI_NPC_MainQuest
   generalConfig {defaultMessage="So I should run away? You think I'm a coward, mate?!"}
  }
  sub_nodes [
   ConversationChoiceNode_97 {
    class=ConversationChoiceNode
    properties {
     defaultChoiceDisplayText=Impending Doom
     choiceTags {gameplayTags [{tagName=Conversation.Choice.Category.MainQuestline}, {tagName=Conversation.Choice.AutoSelectIfValid}]}
    }
   }
   FortConversationRequirement_ControllerRequirement_73 {
    class=FortConversationRequirement_ControllerRequirement
    properties {
     nodeName=Require: Has Completed Quest
     requirement {
      class=FortControllerRequirement_HasCompletedQuest
      properties {bInverseCondition=true, questDefinition=/BattlepassS17/Items/QuestItems/Legendary/quest_s17_w14_legendary_q1b}
     }
    }
   }
   FortConversationRequirement_ControllerRequirement_74 {
    class=FortConversationRequirement_ControllerRequirement
    properties {
     nodeName=Require: Is Quest Active
     requirement {
      class=FortControllerRequirement_IsQuestActive
      properties {questDefinition=&quest_s17_w14_legendary_q1b}
     }
    }
   }
   FortConversationRequirement_ControllerRequirement_75 {
    class=FortConversationRequirement_ControllerRequirement
    properties {
     nodeName=Require: Has Completed Quest Objective
     requirement {
      class=FortControllerRequirement_HasCompletedQuestObjective
      properties {
       bInverseCondition=true
       objectiveBackendName=quest_s17_w14_legendary_q1b_obj19
       questDefinition=&quest_s17_w14_legendary_q1b
      }
     }
    }
   }
   FortConversationRequirement_ControllerRequirement_76 {
    class=FortConversationRequirement_ControllerRequirement
    properties {
     nodeName=Require: NPC Has Conversation Tag
     participantId {tagName=Conversation.Participant.NPC}
     requirement {class=FortControllerRequirement_NPCHasConversationEntryTag, properties {queryTag {tagName=Conversation.Entry.Riot}}}
    }
   }
  ]
 }
 Speech_C_132 {
  id=#29
  class=Speech_C
  outputs [#30]
  properties {generalConfig {defaultMessage="Where's your fightin' spirit?!"}}
  sub_nodes [ConversationChoiceNode_99 {class=ConversationChoiceNode, properties {defaultChoiceDisplayText=Continue}}]
 }
 FortConversationTaskNode_CompleteQuestObjectiveByBackendName_25 {
  id=#30
  class=FortConversationTaskNode_CompleteQuestObjectiveByBackendName
  properties {
   objectiveBackendName=quest_s17_w14_legendary_q1b_obj19
   questDefinition=&quest_s17_w14_legendary_q1b
  }
 }
 Speech_C_110 {
  id=#20
  class=Speech_C
  outputs [#31]
  properties {softTaskIcon=&T_UI_NPC_MainQuest, generalConfig {defaultMessage="I'll collect fresh samples right away. "}}
  sub_nodes [
   ConversationChoiceNode_122 {
    class=ConversationChoiceNode
    properties {
     defaultChoiceDisplayText=Impending Doom
     choiceTags {gameplayTags [{tagName=Conversation.Choice.Category.MainQuestline}, {tagName=Conversation.Choice.AutoSelectIfValid}]}
    }
   }
   FortConversationRequirement_ControllerRequirement_93 {
    class=FortConversationRequirement_ControllerRequirement
    properties {
     nodeName=Require: Has Completed Quest
     requirement {
      class=FortControllerRequirement_HasCompletedQuest
      properties {bInverseCondition=true, questDefinition=&quest_s17_w14_legendary_q1b}
     }
    }
   }
   FortConversationRequirement_ControllerRequirement_94 {
    class=FortConversationRequirement_ControllerRequirement
    properties {
     nodeName=Require: Is Quest Active
     requirement {
      class=FortControllerRequirement_IsQuestActive
      properties {questDefinition=&quest_s17_w14_legendary_q1b}
     }
    }
   }
   FortConversationRequirement_ControllerRequirement_95 {
    class=FortConversationRequirement_ControllerRequirement
    properties {
     nodeName=Require: Has Completed Quest Objective
     requirement {
      class=FortControllerRequirement_HasCompletedQuestObjective
      properties {
       bInverseCondition=true
       objectiveBackendName=quest_s17_w14_legendary_q1b_obj18
       questDefinition=&quest_s17_w14_legendary_q1b
      }
     }
    }
   }
   FortConversationRequirement_ControllerRequirement_96 {
    class=FortConversationRequirement_ControllerRequirement
    properties {
     nodeName=Require: NPC Has Conversation Tag
     participantId {tagName=Conversation.Participant.NPC}
     requirement {class=FortControllerRequirement_NPCHasConversationEntryTag, properties {queryTag {tagName=Conversation.Entry.SwampStalker}}}
    }
   }
  ]
 }
 Speech_C_112 {
  id=#31
  class=Speech_C
  outputs [#32]
  properties {generalConfig {defaultMessage="We must protect the biodiversity of the swamp. "}}
  sub_nodes [ConversationChoiceNode_124 {class=ConversationChoiceNode, properties {defaultChoiceDisplayText=Continue}}]
 }
 FortConversationTaskNode_CompleteQuestObjectiveByBackendName_21 {
  id=#32
  class=FortConversationTaskNode_CompleteQuestObjectiveByBackendName
  properties {
   objectiveBackendName=quest_s17_w14_legendary_q1b_obj18
   questDefinition=&quest_s17_w14_legendary_q1b
  }
 }
 Speech_C_125 {
  id=#19
  class=Speech_C
  outputs [#33]
  properties {softTaskIcon=&T_UI_NPC_MainQuest, generalConfig {defaultMessage="I'm all for bootin' those invaders, but... "}}
  sub_nodes [
   ConversationChoiceNode_117 {
    class=ConversationChoiceNode
    properties {
     defaultChoiceDisplayText=Impending Doom
     choiceTags {gameplayTags [{tagName=Conversation.Choice.Category.MainQuestline}, {tagName=Conversation.Choice.AutoSelectIfValid}]}
    }
   }
   FortConversationRequirement_ControllerRequirement_89 {
    class=FortConversationRequirement_ControllerRequirement
    properties {
     nodeName=Require: Has Completed Quest
     requirement {
      class=FortControllerRequirement_HasCompletedQuest
      properties {bInverseCondition=true, questDefinition=&quest_s17_w14_legendary_q1b}
     }
    }
   }
   FortConversationRequirement_ControllerRequirement_90 {
    class=FortConversationRequirement_ControllerRequirement
    properties {
     nodeName=Require: Is Quest Active
     requirement {
      class=FortControllerRequirement_IsQuestActive
      properties {questDefinition=&quest_s17_w14_legendary_q1b}
     }
    }
   }
   FortConversationRequirement_ControllerRequirement_91 {
    class=FortConversationRequirement_ControllerRequirement
    properties {
     nodeName=Require: Has Completed Quest Objective
     requirement {
      class=FortControllerRequirement_HasCompletedQuestObjective
      properties {
       bInverseCondition=true
       objectiveBackendName=quest_s17_w14_legendary_q1b_obj17
       questDefinition=&quest_s17_w14_legendary_q1b
      }
     }
    }
   }
   FortConversationRequirement_ControllerRequirement_92 {
    class=FortConversationRequirement_ControllerRequirement
    properties {
     nodeName=Require: NPC Has Conversation Tag
     participantId {tagName=Conversation.Participant.NPC}
     requirement {class=FortControllerRequirement_NPCHasConversationEntryTag, properties {queryTag {tagName=Conversation.Entry.Hayseed}}}
    }
   }
  ]
 }
 Speech_C_127 {
  id=#33
  class=Speech_C
  outputs [#34]
  properties {generalConfig {defaultMessage="My corn! My precious corn!!"}}
  sub_nodes [ConversationChoiceNode_119 {class=ConversationChoiceNode, properties {defaultChoiceDisplayText=Continue}}]
 }
 FortConversationTaskNode_CompleteQuestObjectiveByBackendName_24 {
  id=#34
  class=FortConversationTaskNode_CompleteQuestObjectiveByBackendName
  properties {
   objectiveBackendName=quest_s17_w14_legendary_q1b_obj17
   questDefinition=&quest_s17_w14_legendary_q1b
  }
 }
 Speech_C_120 {
  id=#18
  class=Speech_C
  outputs [#35]
  properties {softTaskIcon=&T_UI_NPC_MainQuest, generalConfig {defaultMessage="Roger. Mission is a go. "}}
  sub_nodes [
   ConversationChoiceNode_112 {
    class=ConversationChoiceNode
    properties {
     defaultChoiceDisplayText=Impending Doom
     choiceTags {gameplayTags [{tagName=Conversation.Choice.Category.MainQuestline}, {tagName=Conversation.Choice.AutoSelectIfValid}]}
    }
   }
   FortConversationRequirement_ControllerRequirement_85 {
    class=FortConversationRequirement_ControllerRequirement
    properties {
     nodeName=Require: Has Completed Quest
     requirement {
      class=FortControllerRequirement_HasCompletedQuest
      properties {bInverseCondition=true, questDefinition=&quest_s17_w14_legendary_q1b}
     }
    }
   }
   FortConversationRequirement_ControllerRequirement_86 {
    class=FortConversationRequirement_ControllerRequirement
    properties {
     nodeName=Require: Is Quest Active
     requirement {
      class=FortControllerRequirement_IsQuestActive
      properties {questDefinition=&quest_s17_w14_legendary_q1b}
     }
    }
   }
   FortConversationRequirement_ControllerRequirement_87 {
    class=FortConversationRequirement_ControllerRequirement
    properties {
     nodeName=Require: Has Completed Quest Objective
     requirement {
      class=FortControllerRequirement_HasCompletedQuestObjective
      properties {
       bInverseCondition=true
       objectiveBackendName=quest_s17_w14_legendary_q1b_obj16
       questDefinition=&quest_s17_w14_legendary_q1b
      }
     }
    }
   }
   FortConversationRequirement_ControllerRequirement_88 {
    class=FortConversationRequirement_ControllerRequirement
    properties {
     nodeName=Require: NPC Has Conversation Tag
     participantId {tagName=Conversation.Participant.NPC}
     requirement {class=FortControllerRequirement_NPCHasConversationEntryTag, properties {queryTag {tagName=Conversation.Entry.Rook}}}
    }
   }
  ]
 }
 FortConversationTaskNode_CompleteQuestObjectiveByBackendName_23 {
  id=#35
  class=FortConversationTaskNode_CompleteQuestObjectiveByBackendName
  properties {
   objectiveBackendName=quest_s17_w14_legendary_q1b_obj16
   questDefinition=&quest_s17_w14_legendary_q1b
  }
 }
 Speech_C_115 {
  id=#17
  class=Speech_C
  outputs [#36]
  properties {
   softTaskIcon=&T_UI_NPC_MainQuest
   generalConfig {defaultMessage="Roger that. When the time comes, I'll proceed to the rendezvous point."}
  }
  sub_nodes [
   ConversationChoiceNode_92 {
    class=ConversationChoiceNode
    properties {
     defaultChoiceDisplayText=Impending Doom
     choiceTags {gameplayTags [{tagName=Conversation.Choice.Category.MainQuestline}, {tagName=Conversation.Choice.AutoSelectIfValid}]}
    }
   }
   FortConversationRequirement_ControllerRequirement_69 {
    class=FortConversationRequirement_ControllerRequirement
    properties {
     nodeName=Require: Has Completed Quest
     requirement {
      class=FortControllerRequirement_HasCompletedQuest
      properties {bInverseCondition=true, questDefinition=&quest_s17_w14_legendary_q1b}
     }
    }
   }
   FortConversationRequirement_ControllerRequirement_70 {
    class=FortConversationRequirement_ControllerRequirement
    properties {
     nodeName=Require: Is Quest Active
     requirement {
      class=FortControllerRequirement_IsQuestActive
      properties {questDefinition=&quest_s17_w14_legendary_q1b}
     }
    }
   }
   FortConversationRequirement_ControllerRequirement_71 {
    class=FortConversationRequirement_ControllerRequirement
    properties {
     nodeName=Require: Has Completed Quest Objective
     requirement {
      class=FortControllerRequirement_HasCompletedQuestObjective
      properties {
       bInverseCondition=true
       objectiveBackendName=quest_s17_w14_legendary_q1b_obj15
       questDefinition=&quest_s17_w14_legendary_q1b
      }
     }
    }
   }
   FortConversationRequirement_ControllerRequirement_72 {
    class=FortConversationRequirement_ControllerRequirement
    properties {
     nodeName=Require: NPC Has Conversation Tag
     participantId {tagName=Conversation.Participant.NPC}
     requirement {class=FortControllerRequirement_NPCHasConversationEntryTag, properties {queryTag {tagName=Conversation.Entry.SpecialForces}}}
    }
   }
  ]
 }
 FortConversationTaskNode_CompleteQuestObjectiveByBackendName_22 {
  id=#36
  class=FortConversationTaskNode_CompleteQuestObjectiveByBackendName
  properties {
   objectiveBackendName=quest_s17_w14_legendary_q1b_obj15
   questDefinition=&quest_s17_w14_legendary_q1b
  }
 }
 Speech_C_105 {
  id=#16
  class=Speech_C
  outputs [#37]
  properties {softTaskIcon=&T_UI_NPC_MainQuest, generalConfig {defaultMessage="Yeah, yeah. Whatever, kid. "}}
  sub_nodes [
   ConversationChoiceNode_107 {
    class=ConversationChoiceNode
    properties {
     defaultChoiceDisplayText=Impending Doom
     choiceTags {gameplayTags [{tagName=Conversation.Choice.Category.MainQuestline}, {tagName=Conversation.Choice.AutoSelectIfValid}]}
    }
   }
   FortConversationRequirement_ControllerRequirement_81 {
    class=FortConversationRequirement_ControllerRequirement
    properties {
     nodeName=Require: Has Completed Quest
     requirement {
      class=FortControllerRequirement_HasCompletedQuest
      properties {bInverseCondition=true, questDefinition=&quest_s17_w14_legendary_q1b}
     }
    }
   }
   FortConversationRequirement_ControllerRequirement_82 {
    class=FortConversationRequirement_ControllerRequirement
    properties {
     nodeName=Require: Is Quest Active
     requirement {
      class=FortControllerRequirement_IsQuestActive
      properties {questDefinition=&quest_s17_w14_legendary_q1b}
     }
    }
   }
   FortConversationRequirement_ControllerRequirement_83 {
    class=FortConversationRequirement_ControllerRequirement
    properties {
     nodeName=Require: Has Completed Quest Objective
     requirement {
      class=FortControllerRequirement_HasCompletedQuestObjective
      properties {
       bInverseCondition=true
       objectiveBackendName=quest_s17_w14_legendary_q1b_obj14
       questDefinition=&quest_s17_w14_legendary_q1b
      }
     }
    }
   }
   FortConversationRequirement_ControllerRequirement_84 {
    class=FortConversationRequirement_ControllerRequirement
    properties {
     nodeName=Require: NPC Has Conversation Tag
     participantId {tagName=Conversation.Participant.NPC}
     requirement {class=FortControllerRequirement_NPCHasConversationEntryTag, properties {queryTag {tagName=Conversation.Entry.Rickus}}}
    }
   }
  ]
 }
 FortConversationTaskNode_CompleteQuestObjectiveByBackendName_20 {
  id=#37
  class=FortConversationTaskNode_CompleteQuestObjectiveByBackendName
  properties {
   objectiveBackendName=quest_s17_w14_legendary_q1b_obj14
   questDefinition=&quest_s17_w14_legendary_q1b
  }
 }
 Speech_C_100 {
  id=#15
  class=Speech_C
  outputs [#38]
  properties {
   softTaskIcon=&T_UI_NPC_MainQuest
   generalConfig {defaultMessage=No... She didn't... Excuse me. I have to make a call.}
  }
  sub_nodes [
   ConversationChoiceNode_5 {
    class=ConversationChoiceNode
    properties {
     defaultChoiceDisplayText=Impending Doom
     choiceTags {gameplayTags [{tagName=Conversation.Choice.Category.MainQuestline}, {tagName=Conversation.Choice.AutoSelectIfValid}]}
    }
   }
   FortConversationRequirement_ControllerRequirement_23 {
    class=FortConversationRequirement_ControllerRequirement
    properties {
     nodeName=Require: Has Completed Quest
     requirement {
      class=FortControllerRequirement_HasCompletedQuest
      properties {bInverseCondition=true, questDefinition=&quest_s17_w14_legendary_q1b}
     }
    }
   }
   FortConversationRequirement_ControllerRequirement_24 {
    class=FortConversationRequirement_ControllerRequirement
    properties {
     nodeName=Require: Is Quest Active
     requirement {
      class=FortControllerRequirement_IsQuestActive
      properties {questDefinition=&quest_s17_w14_legendary_q1b}
     }
    }
   }
   FortConversationRequirement_ControllerRequirement_25 {
    class=FortConversationRequirement_ControllerRequirement
    properties {
     nodeName=Require: Has Completed Quest Objective
     requirement {
      class=FortControllerRequirement_HasCompletedQuestObjective
      properties {
       bInverseCondition=true
       objectiveBackendName=quest_s17_w14_legendary_q1b_obj13
       questDefinition=&quest_s17_w14_legendary_q1b
      }
     }
    }
   }
   FortConversationRequirement_ControllerRequirement_26 {
    class=FortConversationRequirement_ControllerRequirement
    properties {
     nodeName=Require: NPC Has Conversation Tag
     participantId {tagName=Conversation.Participant.NPC}
     requirement {class=FortControllerRequirement_NPCHasConversationEntryTag, properties {queryTag {tagName=Conversation.Entry.Maven}}}
    }
   }
  ]
 }
 FortConversationTaskNode_CompleteQuestObjectiveByBackendName_19 {
  id=#38
  class=FortConversationTaskNode_CompleteQuestObjectiveByBackendName
  properties {
   objectiveBackendName=quest_s17_w14_legendary_q1b_obj13
   questDefinition=&quest_s17_w14_legendary_q1b
  }
 }
 Speech_C_95 {
  id=#14
  class=Speech_C
  outputs [#39]
  properties {
   softTaskIcon=&T_UI_NPC_MainQuest
   generalConfig {defaultMessage=Don't worry about me. I've got a getaway driver.}
  }
  sub_nodes [
   ConversationChoiceNode_87 {
    class=ConversationChoiceNode
    properties {
     defaultChoiceDisplayText=Impending Doom
     choiceTags {gameplayTags [{tagName=Conversation.Choice.Category.MainQuestline}, {tagName=Conversation.Choice.AutoSelectIfValid}]}
    }
   }
   FortConversationRequirement_ControllerRequirement_65 {
    class=FortConversationRequirement_ControllerRequirement
    properties {
     nodeName=Require: Has Completed Quest
     requirement {
      class=FortControllerRequirement_HasCompletedQuest
      properties {bInverseCondition=true, questDefinition=&quest_s17_w14_legendary_q1b}
     }
    }
   }
   FortConversationRequirement_ControllerRequirement_66 {
    class=FortConversationRequirement_ControllerRequirement
    properties {
     nodeName=Require: Is Quest Active
     requirement {
      class=FortControllerRequirement_IsQuestActive
      properties {questDefinition=&quest_s17_w14_legendary_q1b}
     }
    }
   }
   FortConversationRequirement_ControllerRequirement_67 {
    class=FortConversationRequirement_ControllerRequirement
    properties {
     nodeName=Require: Has Completed Quest Objective
     requirement {
      class=FortControllerRequirement_HasCompletedQuestObjective
      properties {
       bInverseCondition=true
       objectiveBackendName=quest_s17_w14_legendary_q1b_obj12
       questDefinition=&quest_s17_w14_legendary_q1b
      }
     }
    }
   }
   FortConversationRequirement_ControllerRequirement_68 {
    class=FortConversationRequirement_ControllerRequirement
    properties {
     nodeName=Require: NPC Has Conversation Tag
     participantId {tagName=Conversation.Participant.NPC}
     requirement {class=FortControllerRequirement_NPCHasConversationEntryTag, properties {queryTag {tagName=Conversation.Entry.Marigold}}}
    }
   }
  ]
 }
 FortConversationTaskNode_CompleteQuestObjectiveByBackendName_18 {
  id=#39
  class=FortConversationTaskNode_CompleteQuestObjectiveByBackendName
  properties {
   objectiveBackendName=quest_s17_w14_legendary_q1b_obj12
   questDefinition=&quest_s17_w14_legendary_q1b
  }
 }
 Speech_C_90 {
  id=#13
  class=Speech_C
  outputs [#40]
  properties {
   softTaskIcon=&T_UI_NPC_MainQuest
   generalConfig {defaultMessage="Nice try. I've got three days left on this trip. "}
  }
  sub_nodes [
   ConversationChoiceNode_82 {
    class=ConversationChoiceNode
    properties {
     defaultChoiceDisplayText=Impending Doom
     choiceTags {gameplayTags [{tagName=Conversation.Choice.Category.MainQuestline}, {tagName=Conversation.Choice.AutoSelectIfValid}]}
    }
   }
   FortConversationRequirement_ControllerRequirement_61 {
    class=FortConversationRequirement_ControllerRequirement
    properties {
     nodeName=Require: Has Completed Quest
     requirement {
      class=FortControllerRequirement_HasCompletedQuest
      properties {bInverseCondition=true, questDefinition=&quest_s17_w14_legendary_q1b}
     }
    }
   }
   FortConversationRequirement_ControllerRequirement_62 {
    class=FortConversationRequirement_ControllerRequirement
    properties {
     nodeName=Require: Is Quest Active
     requirement {
      class=FortControllerRequirement_IsQuestActive
      properties {questDefinition=&quest_s17_w14_legendary_q1b}
     }
    }
   }
   FortConversationRequirement_ControllerRequirement_63 {
    class=FortConversationRequirement_ControllerRequirement
    properties {
     nodeName=Require: Has Completed Quest Objective
     requirement {
      class=FortControllerRequirement_HasCompletedQuestObjective
      properties {
       bInverseCondition=true
       objectiveBackendName=quest_s17_w14_legendary_q1b_obj11
       questDefinition=&quest_s17_w14_legendary_q1b
      }
     }
    }
   }
   FortConversationRequirement_ControllerRequirement_64 {
    class=FortConversationRequirement_ControllerRequirement
    properties {
     nodeName=Require: NPC Has Conversation Tag
     participantId {tagName=Conversation.Participant.NPC}
     requirement {class=FortControllerRequirement_NPCHasConversationEntryTag, properties {queryTag {tagName=Conversation.Entry.HenchmanSummer}}}
    }
   }
  ]
 }
 Speech_C_92 {
  id=#40
  class=Speech_C
  outputs [#41]
  properties {generalConfig {defaultMessage="What part of \"out of office\" don't you understand?!"}}
  sub_nodes [ConversationChoiceNode_84 {class=ConversationChoiceNode, properties {defaultChoiceDisplayText=Continue}}]
 }
 FortConversationTaskNode_CompleteQuestObjectiveByBackendName_17 {
  id=#41
  class=FortConversationTaskNode_CompleteQuestObjectiveByBackendName
  properties {
   objectiveBackendName=quest_s17_w14_legendary_q1b_obj11
   questDefinition=&quest_s17_w14_legendary_q1b
  }
 }
 Speech_C_85 {
  id=#12
  class=Speech_C
  outputs [#42]
  properties {softTaskIcon=&T_UI_NPC_MainQuest, generalConfig {defaultMessage="I've got to warn everyone. Quickly. "}}
  sub_nodes [
   ConversationChoiceNode_77 {
    class=ConversationChoiceNode
    properties {
     defaultChoiceDisplayText=Impending Doom
     choiceTags {gameplayTags [{tagName=Conversation.Choice.Category.MainQuestline}, {tagName=Conversation.Choice.AutoSelectIfValid}]}
    }
   }
   FortConversationRequirement_ControllerRequirement_57 {
    class=FortConversationRequirement_ControllerRequirement
    properties {
     nodeName=Require: Has Completed Quest
     requirement {
      class=FortControllerRequirement_HasCompletedQuest
      properties {bInverseCondition=true, questDefinition=&quest_s17_w14_legendary_q1b}
     }
    }
   }
   FortConversationRequirement_ControllerRequirement_58 {
    class=FortConversationRequirement_ControllerRequirement
    properties {
     nodeName=Require: Is Quest Active
     requirement {
      class=FortControllerRequirement_IsQuestActive
      properties {questDefinition=&quest_s17_w14_legendary_q1b}
     }
    }
   }
   FortConversationRequirement_ControllerRequirement_59 {
    class=FortConversationRequirement_ControllerRequirement
    properties {
     nodeName=Require: Has Completed Quest Objective
     requirement {
      class=FortControllerRequirement_HasCompletedQuestObjective
      properties {
       bInverseCondition=true
       objectiveBackendName=quest_s17_w14_legendary_q1b_obj10
       questDefinition=&quest_s17_w14_legendary_q1b
      }
     }
    }
   }
   FortConversationRequirement_ControllerRequirement_60 {
    class=FortConversationRequirement_ControllerRequirement
    properties {
     nodeName=Require: NPC Has Conversation Tag
     participantId {tagName=Conversation.Participant.NPC}
     requirement {class=FortControllerRequirement_NPCHasConversationEntryTag, properties {queryTag {tagName=Conversation.Entry.Emperor}}}
    }
   }
  ]
 }
 FortConversationTaskNode_CompleteQuestObjectiveByBackendName_16 {
  id=#42
  class=FortConversationTaskNode_CompleteQuestObjectiveByBackendName
  properties {
   objectiveBackendName=quest_s17_w14_legendary_q1b_obj10
   questDefinition=&quest_s17_w14_legendary_q1b
  }
 }
 Speech_C_80 {
  id=#11
  class=Speech_C
  outputs [#43]
  properties {
   softTaskIcon=&T_UI_NPC_MainQuest
   generalConfig {defaultMessage=So that's why Slone was asking me about my explosives designs.}
  }
  sub_nodes [
   ConversationChoiceNode_72 {
    class=ConversationChoiceNode
    properties {
     defaultChoiceDisplayText=Impending Doom
     choiceTags {gameplayTags [{tagName=Conversation.Choice.Category.MainQuestline}, {tagName=Conversation.Choice.AutoSelectIfValid}]}
    }
   }
   FortConversationRequirement_ControllerRequirement_53 {
    class=FortConversationRequirement_ControllerRequirement
    properties {
     nodeName=Require: Has Completed Quest
     requirement {
      class=FortControllerRequirement_HasCompletedQuest
      properties {bInverseCondition=true, questDefinition=&quest_s17_w14_legendary_q1b}
     }
    }
   }
   FortConversationRequirement_ControllerRequirement_54 {
    class=FortConversationRequirement_ControllerRequirement
    properties {
     nodeName=Require: Is Quest Active
     requirement {
      class=FortControllerRequirement_IsQuestActive
      properties {questDefinition=&quest_s17_w14_legendary_q1b}
     }
    }
   }
   FortConversationRequirement_ControllerRequirement_55 {
    class=FortConversationRequirement_ControllerRequirement
    properties {
     nodeName=Require: Has Completed Quest Objective
     requirement {
      class=FortControllerRequirement_HasCompletedQuestObjective
      properties {
       bInverseCondition=true
       objectiveBackendName=quest_s17_w14_legendary_q1b_obj9
       questDefinition=&quest_s17_w14_legendary_q1b
      }
     }
    }
   }
   FortConversationRequirement_ControllerRequirement_56 {
    class=FortConversationRequirement_ControllerRequirement
    properties {
     nodeName=Require: NPC Has Conversation Tag
     participantId {tagName=Conversation.Participant.NPC}
     requirement {class=FortControllerRequirement_NPCHasConversationEntryTag, properties {queryTag {tagName=Conversation.Entry.CavernArmored}}}
    }
   }
  ]
 }
 FortConversationTaskNode_CompleteQuestObjectiveByBackendName_15 {
  id=#43
  class=FortConversationTaskNode_CompleteQuestObjectiveByBackendName
  properties {objectiveBackendName=quest_s17_w14_legendary_q1b_obj9, questDefinition=&quest_s17_w14_legendary_q1b}
 }
 Speech_C_75 {
  id=#10
  class=Speech_C
  outputs [#44]
  properties {softTaskIcon=&T_UI_NPC_MainQuest, generalConfig {defaultMessage=Slone's done it now. That's not gonna end well.}}
  sub_nodes [
   ConversationChoiceNode_67 {
    class=ConversationChoiceNode
    properties {
     defaultChoiceDisplayText=Impending Doom
     choiceTags {gameplayTags [{tagName=Conversation.Choice.Category.MainQuestline}, {tagName=Conversation.Choice.AutoSelectIfValid}]}
    }
   }
   FortConversationRequirement_ControllerRequirement_49 {
    class=FortConversationRequirement_ControllerRequirement
    properties {
     nodeName=Require: Has Completed Quest
     requirement {
      class=FortControllerRequirement_HasCompletedQuest
      properties {bInverseCondition=true, questDefinition=&quest_s17_w14_legendary_q1b}
     }
    }
   }
   FortConversationRequirement_ControllerRequirement_50 {
    class=FortConversationRequirement_ControllerRequirement
    properties {
     nodeName=Require: Is Quest Active
     requirement {
      class=FortControllerRequirement_IsQuestActive
      properties {questDefinition=&quest_s17_w14_legendary_q1b}
     }
    }
   }
   FortConversationRequirement_ControllerRequirement_51 {
    class=FortConversationRequirement_ControllerRequirement
    properties {
     nodeName=Require: Has Completed Quest Objective
     requirement {
      class=FortControllerRequirement_HasCompletedQuestObjective
      properties {
       bInverseCondition=true
       objectiveBackendName=quest_s17_w14_legendary_q1b_obj8
       questDefinition=&quest_s17_w14_legendary_q1b
      }
     }
    }
   }
   FortConversationRequirement_ControllerRequirement_52 {
    class=FortConversationRequirement_ControllerRequirement
    properties {
     nodeName=Require: NPC Has Conversation Tag
     participantId {tagName=Conversation.Participant.NPC}
     requirement {class=FortControllerRequirement_NPCHasConversationEntryTag, properties {queryTag {tagName=Conversation.Entry.BunkerJonesy}}}
    }
   }
  ]
 }
 FortConversationTaskNode_CompleteQuestObjectiveByBackendName_14 {
  id=#44
  class=FortConversationTaskNode_CompleteQuestObjectiveByBackendName
  properties {objectiveBackendName=quest_s17_w14_legendary_q1b_obj8, questDefinition=&quest_s17_w14_legendary_q1b}
 }
 Speech_C_70 {
  id=#9
  class=Speech_C
  outputs [#45]
  properties {
   softTaskIcon=&T_UI_NPC_MainQuest
   generalConfig {defaultMessage=Uhh... Are we sure we want that big thing in the sky to end up here on the ground?}
  }
  sub_nodes [
   ConversationChoiceNode_26 {
    class=ConversationChoiceNode
    properties {
     defaultChoiceDisplayText=Impending Doom
     choiceTags {gameplayTags [{tagName=Conversation.Choice.Category.MainQuestline}, {tagName=Conversation.Choice.AutoSelectIfValid}]}
    }
   }
   FortConversationRequirement_ControllerRequirement_48 {
    class=FortConversationRequirement_ControllerRequirement
    properties {
     nodeName=Require: Has Completed Quest
     requirement {
      class=FortControllerRequirement_HasCompletedQuest
      properties {bInverseCondition=true, questDefinition=&quest_s17_w14_legendary_q1b}
     }
    }
   }
   FortConversationRequirement_ControllerRequirement_20 {
    class=FortConversationRequirement_ControllerRequirement
    properties {
     nodeName=Require: Is Quest Active
     requirement {
      class=FortControllerRequirement_IsQuestActive
      properties {questDefinition=&quest_s17_w14_legendary_q1b}
     }
    }
   }
   FortConversationRequirement_ControllerRequirement_21 {
    class=FortConversationRequirement_ControllerRequirement
    properties {
     nodeName=Require: Has Completed Quest Objective
     requirement {
      class=FortControllerRequirement_HasCompletedQuestObjective
      properties {
       bInverseCondition=true
       objectiveBackendName=quest_s17_w14_legendary_q1b_obj7
       questDefinition=&quest_s17_w14_legendary_q1b
      }
     }
    }
   }
   FortConversationRequirement_ControllerRequirement_22 {
    class=FortConversationRequirement_ControllerRequirement
    properties {
     nodeName=Require: NPC Has Conversation Tag
     participantId {tagName=Conversation.Participant.NPC}
     requirement {class=FortControllerRequirement_NPCHasConversationEntryTag, properties {queryTag {tagName=Conversation.Entry.Broccoli}}}
    }
   }
  ]
 }
 Speech_C_72 {
  id=#45
  class=Speech_C
  outputs [#46]
  properties {generalConfig {defaultMessage=You know... Where people live?}}
  sub_nodes [ConversationChoiceNode_6 {class=ConversationChoiceNode, properties {defaultChoiceDisplayText=Continue}}]
 }
 FortConversationTaskNode_CompleteQuestObjectiveByBackendName_13 {
  id=#46
  class=FortConversationTaskNode_CompleteQuestObjectiveByBackendName
  properties {objectiveBackendName=quest_s17_w14_legendary_q1b_obj7, questDefinition=&quest_s17_w14_legendary_q1b}
 }
 Speech_C_65 {
  id=#8
  class=Speech_C
  outputs [#47]
  properties {softTaskIcon=&T_UI_NPC_MainQuest, generalConfig {defaultMessage="Sounds chaotic. I'm in. "}}
  sub_nodes [
   ConversationChoiceNode_17 {
    class=ConversationChoiceNode
    properties {
     defaultChoiceDisplayText=Impending Doom
     choiceTags {gameplayTags [{tagName=Conversation.Choice.Category.MainQuestline}, {tagName=Conversation.Choice.AutoSelectIfValid}]}
    }
   }
   FortConversationRequirement_ControllerRequirement_8 {
    class=FortConversationRequirement_ControllerRequirement
    properties {
     nodeName=Require: Has Completed Quest
     requirement {
      class=FortControllerRequirement_HasCompletedQuest
      properties {bInverseCondition=true, questDefinition=&quest_s17_w14_legendary_q1b}
     }
    }
   }
   FortConversationRequirement_ControllerRequirement_9 {
    class=FortConversationRequirement_ControllerRequirement
    properties {
     nodeName=Require: Is Quest Active
     requirement {
      class=FortControllerRequirement_IsQuestActive
      properties {questDefinition=&quest_s17_w14_legendary_q1b}
     }
    }
   }
   FortConversationRequirement_ControllerRequirement_10 {
    class=FortConversationRequirement_ControllerRequirement
    properties {
     nodeName=Require: Has Completed Quest Objective
     requirement {
      class=FortControllerRequirement_HasCompletedQuestObjective
      properties {
       bInverseCondition=true
       objectiveBackendName=quest_s17_w14_legendary_q1b_obj6
       questDefinition=&quest_s17_w14_legendary_q1b
      }
     }
    }
   }
   FortConversationRequirement_ControllerRequirement_11 {
    class=FortConversationRequirement_ControllerRequirement
    properties {
     nodeName=Require: NPC Has Conversation Tag
     participantId {tagName=Conversation.Participant.NPC}
     requirement {class=FortControllerRequirement_NPCHasConversationEntryTag, properties {queryTag {tagName=Conversation.Entry.Antique}}}
    }
   }
  ]
 }
 FortConversationTaskNode_CompleteQuestObjectiveByBackendName_12 {
  id=#47
  class=FortConversationTaskNode_CompleteQuestObjectiveByBackendName
  properties {objectiveBackendName=quest_s17_w14_legendary_q1b_obj6, questDefinition=&quest_s17_w14_legendary_q1b}
 }
 Speech_C_59 {
  id=#7
  class=Speech_C
  outputs [#48]
  properties {softTaskIcon=&T_UI_NPC_MainQuest, generalConfig {defaultMessage="Is that so? Hmm. "}}
  sub_nodes [
   ConversationChoiceNode_61 {
    class=ConversationChoiceNode
    properties {
     defaultChoiceDisplayText=Impending Doom
     choiceTags {gameplayTags [{tagName=Conversation.Choice.Category.MainQuestline}, {tagName=Conversation.Choice.AutoSelectIfValid}]}
    }
   }
   FortConversationRequirement_ControllerRequirement_40 {
    class=FortConversationRequirement_ControllerRequirement
    properties {
     nodeName=Require: Has Completed Quest
     requirement {
      class=FortControllerRequirement_HasCompletedQuest
      properties {bInverseCondition=true, questDefinition=&quest_s17_w14_legendary_q1b}
     }
    }
   }
   FortConversationRequirement_ControllerRequirement_41 {
    class=FortConversationRequirement_ControllerRequirement
    properties {
     nodeName=Require: Is Quest Active
     requirement {
      class=FortControllerRequirement_IsQuestActive
      properties {questDefinition=&quest_s17_w14_legendary_q1b}
     }
    }
   }
   FortConversationRequirement_ControllerRequirement_42 {
    class=FortConversationRequirement_ControllerRequirement
    properties {
     nodeName=Require: Has Completed Quest Objective
     requirement {
      class=FortControllerRequirement_HasCompletedQuestObjective
      properties {
       bInverseCondition=true
       objectiveBackendName=quest_s17_w14_legendary_q1b_obj5
       questDefinition=&quest_s17_w14_legendary_q1b
      }
     }
    }
   }
   FortConversationRequirement_ControllerRequirement_43 {
    class=FortConversationRequirement_ControllerRequirement
    properties {
     nodeName=Require: NPC Has Conversation Tag
     participantId {tagName=Conversation.Participant.NPC}
     requirement {class=FortControllerRequirement_NPCHasConversationEntryTag, properties {queryTag {tagName=Conversation.Entry.AlienTrooper}}}
    }
   }
  ]
 }
 Speech_C_57 {
  id=#48
  class=Speech_C
  outputs [#49]
  properties {generalConfig {defaultMessage=Thank you. You will be rewarded handsomely for your betrayal.}}
  sub_nodes [ConversationChoiceNode_59 {class=ConversationChoiceNode, properties {defaultChoiceDisplayText=Continue}}]
 }
 FortConversationTaskNode_CompleteQuestObjectiveByBackendName_10 {
  id=#49
  class=FortConversationTaskNode_CompleteQuestObjectiveByBackendName
  properties {objectiveBackendName=quest_s17_w14_legendary_q1b_obj5, questDefinition=&quest_s17_w14_legendary_q1b}
 }
 Speech_C_64 {
  id=#6
  class=Speech_C
  outputs [#50]
  properties {softTaskIcon=&T_UI_NPC_MainQuest, generalConfig {defaultMessage="Ha, you mean that big ship up there?"}}
  sub_nodes [
   ConversationChoiceNode_66 {
    class=ConversationChoiceNode
    properties {
     defaultChoiceDisplayText=Impending Doom
     choiceTags {gameplayTags [{tagName=Conversation.Choice.Category.MainQuestline}, {tagName=Conversation.Choice.AutoSelectIfValid}]}
    }
   }
   FortConversationRequirement_ControllerRequirement_44 {
    class=FortConversationRequirement_ControllerRequirement
    properties {
     nodeName=Require: Has Completed Quest
     requirement {
      class=FortControllerRequirement_HasCompletedQuest
      properties {bInverseCondition=true, questDefinition=&quest_s17_w14_legendary_q1b}
     }
    }
   }
   FortConversationRequirement_ControllerRequirement_45 {
    class=FortConversationRequirement_ControllerRequirement
    properties {
     nodeName=Require: Is Quest Active
     requirement {
      class=FortControllerRequirement_IsQuestActive
      properties {questDefinition=&quest_s17_w14_legendary_q1b}
     }
    }
   }
   FortConversationRequirement_ControllerRequirement_46 {
    class=FortConversationRequirement_ControllerRequirement
    properties {
     nodeName=Require: Has Completed Quest Objective
     requirement {
      class=FortControllerRequirement_HasCompletedQuestObjective
      properties {
       bInverseCondition=true
       objectiveBackendName=quest_s17_w14_legendary_q1b_obj4
       questDefinition=&quest_s17_w14_legendary_q1b
      }
     }
    }
   }
   FortConversationRequirement_ControllerRequirement_47 {
    class=FortConversationRequirement_ControllerRequirement
    properties {
     nodeName=Require: NPC Has Conversation Tag
     participantId {tagName=Conversation.Participant.NPC}
     requirement {class=FortControllerRequirement_NPCHasConversationEntryTag, properties {queryTag {tagName=Conversation.Entry.AlienSummer}}}
    }
   }
  ]
 }
 Speech_C_62 {
  id=#50
  class=Speech_C
  outputs [#51]
  properties {generalConfig {defaultMessage="My father built that ship! No one would dare attack it! "}}
  sub_nodes [ConversationChoiceNode_64 {class=ConversationChoiceNode, properties {defaultChoiceDisplayText=Continue}}]
 }
 FortConversationTaskNode_CompleteQuestObjectiveByBackendName_11 {
  id=#51
  class=FortConversationTaskNode_CompleteQuestObjectiveByBackendName
  properties {objectiveBackendName=quest_s17_w14_legendary_q1b_obj4, questDefinition=&quest_s17_w14_legendary_q1b}
 }
 Speech_C_16 {
  id=#5
  class=Speech_C
  outputs [#52]
  properties {
   softTaskIcon=&T_UI_NPC_MainQuest
   generalConfig {defaultMessage="So Slone's ending the world, huh? Can't say I'm surprised. "}
  }
  sub_nodes [
   ConversationChoiceNode_28 {
    class=ConversationChoiceNode
    properties {
     defaultChoiceDisplayText=Impending Doom
     choiceTags {gameplayTags [{tagName=Conversation.Choice.Category.MainQuestline}, {tagName=Conversation.Choice.AutoSelectIfValid}]}
    }
   }
   FortConversationRequirement_ControllerRequirement_16 {
    class=FortConversationRequirement_ControllerRequirement
    properties {
     nodeName=Require: Has Completed Quest
     requirement {
      class=FortControllerRequirement_HasCompletedQuest
      properties {bInverseCondition=true, questDefinition=&quest_s17_w14_legendary_q1b}
     }
    }
   }
   FortConversationRequirement_ControllerRequirement_17 {
    class=FortConversationRequirement_ControllerRequirement
    properties {
     nodeName=Require: Is Quest Active
     requirement {
      class=FortControllerRequirement_IsQuestActive
      properties {questDefinition=&quest_s17_w14_legendary_q1b}
     }
    }
   }
   FortConversationRequirement_ControllerRequirement_18 {
    class=FortConversationRequirement_ControllerRequirement
    properties {
     nodeName=Require: Has Completed Quest Objective
     requirement {
      class=FortControllerRequirement_HasCompletedQuestObjective
      properties {
       bInverseCondition=true
       objectiveBackendName=quest_s17_w14_legendary_q1b_obj3
       questDefinition=&quest_s17_w14_legendary_q1b
      }
     }
    }
   }
   FortConversationRequirement_ControllerRequirement_19 {
    class=FortConversationRequirement_ControllerRequirement
    properties {
     nodeName=Require: NPC Has Conversation Tag
     participantId {tagName=Conversation.Participant.NPC}
     requirement {class=FortControllerRequirement_NPCHasConversationEntryTag, properties {queryTag {tagName=Conversation.Entry.Abstrakt}}}
    }
   }
  ]
 }
 Speech_C_18 {
  id=#52
  class=Speech_C
  outputs [#53]
  properties {generalConfig {defaultMessage="But I'm definitely a little jealous. "}}
  sub_nodes [ConversationChoiceNode_30 {class=ConversationChoiceNode, properties {defaultChoiceDisplayText=Continue}}]
 }
 FortConversationTaskNode_CompleteQuestObjectiveByBackendName_3 {
  id=#53
  class=FortConversationTaskNode_CompleteQuestObjectiveByBackendName
  properties {objectiveBackendName=quest_s17_w14_legendary_q1b_obj3, questDefinition=&quest_s17_w14_legendary_q1b}
 }
 Speech_C_11 {
  id=#4
  class=Speech_C
  outputs [#54]
  properties {
   softTaskIcon=&T_UI_NPC_MainQuest
   generalConfig {defaultMessage="\"Impending doom?\" Please. Slone doesn't stand a chance. "}
  }
  sub_nodes [
   ConversationChoiceNode_22 {
    class=ConversationChoiceNode
    properties {
     defaultChoiceDisplayText=Impending Doom
     choiceTags {gameplayTags [{tagName=Conversation.Choice.Category.MainQuestline}, {tagName=Conversation.Choice.AutoSelectIfValid}]}
    }
   }
   FortConversationRequirement_ControllerRequirement_12 {
    class=FortConversationRequirement_ControllerRequirement
    properties {
     nodeName=Require: Has Completed Quest
     requirement {
      class=FortControllerRequirement_HasCompletedQuest
      properties {bInverseCondition=true, questDefinition=&quest_s17_w14_legendary_q1b}
     }
    }
   }
   FortConversationRequirement_ControllerRequirement_13 {
    class=FortConversationRequirement_ControllerRequirement
    properties {
     nodeName=Require: Is Quest Active
     requirement {
      class=FortControllerRequirement_IsQuestActive
      properties {questDefinition=&quest_s17_w14_legendary_q1b}
     }
    }
   }
   FortConversationRequirement_ControllerRequirement_14 {
    class=FortConversationRequirement_ControllerRequirement
    properties {
     nodeName=Require: Has Completed Quest Objective
     requirement {
      class=FortControllerRequirement_HasCompletedQuestObjective
      properties {
       bInverseCondition=true
       objectiveBackendName=quest_s17_w14_legendary_q1b_obj2
       questDefinition=&quest_s17_w14_legendary_q1b
      }
     }
    }
   }
   FortConversationRequirement_ControllerRequirement_15 {
    class=FortConversationRequirement_ControllerRequirement
    properties {
     nodeName=Require: NPC Has Conversation Tag
     participantId {tagName=Conversation.Participant.NPC}
     requirement {class=FortControllerRequirement_NPCHasConversationEntryTag, properties {queryTag {tagName=Conversation.Entry.Believer}}}
    }
   }
  ]
 }
 FortConversationTaskNode_CompleteQuestObjectiveByBackendName_2 {
  id=#54
  class=FortConversationTaskNode_CompleteQuestObjectiveByBackendName
  properties {objectiveBackendName=quest_s17_w14_legendary_q1b_obj2, questDefinition=&quest_s17_w14_legendary_q1b}
 }
 Speech_C_1 {
  id=#3
  class=Speech_C
  outputs [#55]
  properties {softTaskIcon=&T_UI_NPC_MainQuest, generalConfig {defaultMessage="Whoa. Like, DOOM doom? "}}
  sub_nodes [
   ConversationChoiceNode_12 {
    class=ConversationChoiceNode
    properties {
     defaultChoiceDisplayText=Impending Doom
     choiceTags {gameplayTags [{tagName=Conversation.Choice.Category.MainQuestline}, {tagName=Conversation.Choice.AutoSelectIfValid}]}
    }
   }
   FortConversationRequirement_ControllerRequirement_4 {
    class=FortConversationRequirement_ControllerRequirement
    properties {
     nodeName=Require: Has Completed Quest
     requirement {
      class=FortControllerRequirement_HasCompletedQuest
      properties {bInverseCondition=true, questDefinition=&quest_s17_w14_legendary_q1b}
     }
    }
   }
   FortConversationRequirement_ControllerRequirement_5 {
    class=FortConversationRequirement_ControllerRequirement
    properties {
     nodeName=Require: Is Quest Active
     requirement {
      class=FortControllerRequirement_IsQuestActive
      properties {questDefinition=&quest_s17_w14_legendary_q1b}
     }
    }
   }
   FortConversationRequirement_ControllerRequirement_6 {
    class=FortConversationRequirement_ControllerRequirement
    properties {
     nodeName=Require: Has Completed Quest Objective
     requirement {
      class=FortControllerRequirement_HasCompletedQuestObjective
      properties {
       bInverseCondition=true
       objectiveBackendName=quest_s17_w14_legendary_q1b_obj0
       questDefinition=&quest_s17_w14_legendary_q1b
      }
     }
    }
   }
   FortConversationRequirement_ControllerRequirement_7 {
    class=FortConversationRequirement_ControllerRequirement
    properties {
     nodeName=Require: NPC Has Conversation Tag
     participantId {tagName=Conversation.Participant.NPC}
     requirement {class=FortControllerRequirement_NPCHasConversationEntryTag, properties {queryTag {tagName=Conversation.Entry.Dreamflower}}}
    }
   }
  ]
 }
 Speech_C_8 {
  id=#55
  class=Speech_C
  outputs [#56]
  properties {generalConfig {defaultMessage="I'll warn all the butterflies right away. "}}
  sub_nodes [ConversationChoiceNode_14 {class=ConversationChoiceNode, properties {defaultChoiceDisplayText=Continue}}]
 }
 FortConversationTaskNode_CompleteQuestObjectiveByBackendName_0 {
  id=#56
  class=FortConversationTaskNode_CompleteQuestObjectiveByBackendName
  properties {objectiveBackendName=quest_s17_w14_legendary_q1b_obj0, questDefinition=&quest_s17_w14_legendary_q1b}
 }
 Speech_C_2 {
  id=#2
  class=Speech_C
  outputs [#57]
  properties {
   softTaskIcon=&T_UI_NPC_MainQuest
   generalConfig {defaultMessage=Wow. I'd better move my cache of Meowscles Morsels.}
  }
  sub_nodes [
   ConversationChoiceNode_3 {
    class=ConversationChoiceNode
    properties {
     defaultChoiceDisplayText=Impending Doom
     choiceTags {gameplayTags [{tagName=Conversation.Choice.Category.MainQuestline}, {tagName=Conversation.Choice.AutoSelectIfValid}]}
    }
   }
   FortConversationRequirement_ControllerRequirement_0 {
    class=FortConversationRequirement_ControllerRequirement
    properties {
     nodeName=Require: Has Completed Quest
     requirement {
      class=FortControllerRequirement_HasCompletedQuest
      properties {bInverseCondition=true, questDefinition=&quest_s17_w14_legendary_q1b}
     }
    }
   }
   FortConversationRequirement_ControllerRequirement_1 {
    class=FortConversationRequirement_ControllerRequirement
    properties {
     nodeName=Require: Is Quest Active
     requirement {
      class=FortControllerRequirement_IsQuestActive
      properties {questDefinition=&quest_s17_w14_legendary_q1b}
     }
    }
   }
   FortConversationRequirement_ControllerRequirement_2 {
    class=FortConversationRequirement_ControllerRequirement
    properties {
     nodeName=Require: Has Completed Quest Objective
     requirement {
      class=FortControllerRequirement_HasCompletedQuestObjective
      properties {
       bInverseCondition=true
       objectiveBackendName=quest_s17_w14_legendary_q1b_obj1
       questDefinition=&quest_s17_w14_legendary_q1b
      }
     }
    }
   }
   FortConversationRequirement_ControllerRequirement_3 {
    class=FortConversationRequirement_ControllerRequirement
    properties {
     nodeName=Require: NPC Has Conversation Tag
     participantId {tagName=Conversation.Participant.NPC}
     requirement {class=FortControllerRequirement_NPCHasConversationEntryTag, properties {queryTag {tagName=Conversation.Entry.Faux}}}
    }
   }
  ]
 }
 FortConversationTaskNode_CompleteQuestObjectiveByBackendName_1 {
  id=#57
  class=FortConversationTaskNode_CompleteQuestObjectiveByBackendName
  properties {objectiveBackendName=quest_s17_w14_legendary_q1b_obj1, questDefinition=&quest_s17_w14_legendary_q1b}
 }
]
