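# Shared Season 17 Week 2 NPC conversation database for legendary quest quest_s17_w02_legendary_q3, about the Imagined Order.
# One entry point fans out to five auto-selected dialogue branches for Rickus, Rook, Marigold, Maven and SpecialForces, gated by NPC entry tag, active quest and unfinished objective.
# Each linear speech chain ends by completing that NPC's objective (obj0-obj4).
# The completion tasks carry an inverted turn-in check against the Week 1 quest quest_s17_w01_legendary_q1.
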ConversationDatabase /BattlepassS17/Conversations/Shared_Conversation_Week2
entry_points [{tag=Conversation.Entry.Season17.Week2, destinations [#1]}]
nodes [
 ConversationEntryPointNode_7 {
  id=#1
  class=ConversationEntryPointNode
  outputs [
   #2
   #3
   #4
   #5
   #6
  ]
  properties {entryTag {tagName=Conversation.Entry.Season17.Week2}}
 }
 Speech_C_9 {
  id=#6
  class=Speech_C
  outputs [#7]
  properties {
   softTaskIcon=&T_UI_NPC_MainQuest
   generalConfig {defaultMessage="I am <b>not cleared</> to discuss my involvement or non-involvement with the IO. "}
  }
  sub_nodes [
   ConversationChoiceNode_20 {
    class=ConversationChoiceNode
    properties {
     defaultChoiceDisplayText=Imagined Order?
     choiceTags {gameplayTags [{tagName=Conversation.Choice.Category.MainQuestline}, {tagName=Conversation.Choice.AutoSelectIfValid}]}
    }
   }
   FortConversationRequirement_ControllerRequirement_40 {
    class=FortConversationRequirement_ControllerRequirement
    properties {
     nodeName=Require: Has Completed Quest
     requirement {
      class=FortControllerRequirement_HasCompletedQuest
      properties {bInverseCondition=true, questDefinition=/BattlepassS17/Items/QuestItems/Legendary/quest_s17_w02_legendary_q3}
     }
    }
   }
   FortConversationRequirement_ControllerRequirement_41 {
    class=FortConversationRequirement_ControllerRequirement
    properties {
     nodeName=Require: Is Quest Active
     requirement {
      class=FortControllerRequirement_IsQuestActive
      properties {questDefinition=&quest_s17_w02_legendary_q3}
     }
    }
   }
   FortConversationRequirement_ControllerRequirement_42 {
    class=FortConversationRequirement_ControllerRequirement
    properties {
     nodeName=Require: Has Completed Quest Objective
     requirement {
      class=FortControllerRequirement_HasCompletedQuestObjective
      properties {
       bInverseCondition=true
       objectiveBackendName=quest_s17_w02_legendary_q3_obj4
       questDefinition=&quest_s17_w02_legendary_q3
      }
     }
    }
   }
   FortConversationRequirement_ControllerRequirement_58 {
    class=FortConversationRequirement_ControllerRequirement
    properties {
     nodeName=Require: NPC Has Conversation Tag
     participantId {tagName=Conversation.Participant.NPC}
     requirement {class=FortControllerRequirement_NPCHasConversationEntryTag, properties {queryTag {tagName=Conversation.Entry.SpecialForces}}}
    }
   }
  ]
 }
 Speech_C_10 {
  id=#7
  class=Speech_C
  outputs [#8]
  properties {generalConfig {defaultMessage="But if I were involved, I'm sure I'd be a <b>highly decorated</> field agent. "}}
  sub_nodes [ConversationChoiceNode_14 {class=ConversationChoiceNode, properties {defaultChoiceDisplayText=Continue}}]
 }
 Speech_C_14 {
  id=#8
  class=Speech_C
  outputs [#9]
  properties {generalConfig {defaultMessage="One of the <b>very best. </>"}}
  sub_nodes [ConversationChoiceNode_15 {class=ConversationChoiceNode, properties {defaultChoiceDisplayText=Continue}}]
 }
 Speech_C_15 {
  id=#9
  class=Speech_C
  outputs [#10]
  properties {generalConfig {defaultMessage="<b>If. </>"}}
  sub_nodes [ConversationChoiceNode_16 {class=ConversationChoiceNode, properties {defaultChoiceDisplayText=Continue}}]
 }
 Speech_C_16 {
  id=#10
  class=Speech_C
  outputs [#11]
  properties {generalConfig {defaultMessage="You're picking up what I'm putting down here, <b>right?</>"}}
  sub_nodes [ConversationChoiceNode_23 {class=ConversationChoiceNode, properties {defaultChoiceDisplayText=Continue}}]
 }
 FortConversationTaskNode_CompleteQuestObjectiveByBackendName_4 {
  id=#11
  class=FortConversationTaskNode_CompleteQuestObjectiveByBackendName
  properties {objectiveBackendName=quest_s17_w02_legendary_q3_obj4, questDefinition=&quest_s17_w02_legendary_q3}
  sub_nodes [FortConversationRequirement_ControllerRequirement_66 {
    class=FortConversationRequirement_ControllerRequirement
    properties {
     nodeName=Require: Has Quest To Turn In
     requirement {
      class=FortControllerRequirement_HasQuestToTurnIn
      properties {bInverseCondition=true, softQuestToTurnIn=/BattlepassS17/Items/QuestItems/Legendary/quest_s17_w01_legendary_q1}
     }
    }
   }]
 }
 Speech_C_6 {
  id=#5
  class=Speech_C
  outputs [#12]
  properties {softTaskIcon=&T_UI_NPC_MainQuest, generalConfig {defaultMessage="You're <b>interested</> in the IO?"}}
  sub_nodes [
   ConversationChoiceNode_10 {
    class=ConversationChoiceNode
    properties {
     defaultChoiceDisplayText=Imagined Order?
     choiceTags {gameplayTags [{tagName=Conversation.Choice.Category.MainQuestline}, {tagName=Conversation.Choice.AutoSelectIfValid}]}
    }
   }
   FortConversationRequirement_ControllerRequirement_19 {
    class=FortConversationRequirement_ControllerRequirement
    properties {
     nodeName=Require: Has Completed Quest
     requirement {
      class=FortControllerRequirement_HasCompletedQuest
      properties {bInverseCondition=true, questDefinition=&quest_s17_w02_legendary_q3}
     }
    }
   }
   FortConversationRequirement_ControllerRequirement_20 {
    class=FortConversationRequirement_ControllerRequirement
    properties {
     nodeName=Require: Is Quest Active
     requirement {
      class=FortControllerRequirement_IsQuestActive
      properties {questDefinition=&quest_s17_w02_legendary_q3}
     }
    }
   }
   FortConversationRequirement_ControllerRequirement_21 {
    class=FortConversationRequirement_ControllerRequirement
    properties {
     nodeName=Require: Has Completed Quest Objective
     requirement {
      class=FortControllerRequirement_HasCompletedQuestObjective
      properties {
       bInverseCondition=true
       objectiveBackendName=quest_s17_w02_legendary_q3_obj3
       questDefinition=&quest_s17_w02_legendary_q3
      }
     }
    }
   }
   FortConversationRequirement_ControllerRequirement_22 {
    class=FortConversationRequirement_ControllerRequirement
    properties {
     nodeName=Require: NPC Has Conversation Tag
     participantId {tagName=Conversation.Participant.NPC}
     requirement {class=FortControllerRequirement_NPCHasConversationEntryTag, properties {queryTag {tagName=Conversation.Entry.Maven}}}
    }
   }
  ]
 }
 Speech_C_8 {
  id=#12
  class=Speech_C
  outputs [#13]
  properties {generalConfig {defaultMessage="I should warn you that the <enemy>mortality rate</> for junior agents has gone <b>up...</>"}}
  sub_nodes [ConversationChoiceNode_13 {class=ConversationChoiceNode, properties {defaultChoiceDisplayText=Continue}}]
 }
 Speech_C_13 {
  id=#13
  class=Speech_C
  outputs [#14]
  properties {generalConfig {defaultMessage="<enemy>To 100%. </>\r\nBut if <b>you</> join, you could change that for at least a <b>little while. </>"}}
  sub_nodes [ConversationChoiceNode_11 {class=ConversationChoiceNode, properties {defaultChoiceDisplayText=Continue}}]
 }
 FortConversationTaskNode_CompleteQuestObjectiveByBackendName_3 {
  id=#14
  class=FortConversationTaskNode_CompleteQuestObjectiveByBackendName
  properties {objectiveBackendName=quest_s17_w02_legendary_q3_obj3, questDefinition=&quest_s17_w02_legendary_q3}
  sub_nodes [FortConversationRequirement_ControllerRequirement_23 {
    class=FortConversationRequirement_ControllerRequirement
    properties {
     nodeName=Require: Has Quest To Turn In
     requirement {
      class=FortControllerRequirement_HasQuestToTurnIn
      properties {bInverseCondition=true, softQuestToTurnIn=&quest_s17_w01_legendary_q1}
     }
    }
   }]
 }
 Speech_C_3 {
  id=#4
  class=Speech_C
  outputs [#15]
  properties {softTaskIcon=&T_UI_NPC_MainQuest, generalConfig {defaultMessage=I wouldn't say I support the IO...}}
  sub_nodes [
   ConversationChoiceNode_8 {
    class=ConversationChoiceNode
    properties {
     defaultChoiceDisplayText=Imagined Order?
     choiceTags {gameplayTags [{tagName=Conversation.Choice.Category.MainQuestline}, {tagName=Conversation.Choice.AutoSelectIfValid}]}
    }
   }
   FortConversationRequirement_ControllerRequirement_14 {
    class=FortConversationRequirement_ControllerRequirement
    properties {
     nodeName=Require: Has Completed Quest
     requirement {
      class=FortControllerRequirement_HasCompletedQuest
      properties {bInverseCondition=true, questDefinition=&quest_s17_w02_legendary_q3}
     }
    }
   }
   FortConversationRequirement_ControllerRequirement_15 {
    class=FortConversationRequirement_ControllerRequirement
    properties {
     nodeName=Require: Is Quest Active
     requirement {
      class=FortControllerRequirement_IsQuestActive
      properties {questDefinition=&quest_s17_w02_legendary_q3}
     }
    }
   }
   FortConversationRequirement_ControllerRequirement_16 {
    class=FortConversationRequirement_ControllerRequirement
    properties {
     nodeName=Require: Has Completed Quest Objective
     requirement {
      class=FortControllerRequirement_HasCompletedQuestObjective
      properties {
       bInverseCondition=true
       objectiveBackendName=quest_s17_w02_legendary_q3_obj2
       questDefinition=&quest_s17_w02_legendary_q3
      }
     }
    }
   }
   FortConversationRequirement_ControllerRequirement_17 {
    class=FortConversationRequirement_ControllerRequirement
    properties {
     nodeName=Require: NPC Has Conversation Tag
     participantId {tagName=Conversation.Participant.NPC}
     requirement {class=FortControllerRequirement_NPCHasConversationEntryTag, properties {queryTag {tagName=Conversation.Entry.Marigold}}}
    }
   }
  ]
 }
 Speech_C_4 {
  id=#15
  class=Speech_C
  outputs [#16]
  properties {generalConfig {defaultMessage="But I always make sure <b>they support me. </>"}}
  sub_nodes [ConversationChoiceNode_9 {class=ConversationChoiceNode, properties {defaultChoiceDisplayText=Continue}}]
 }
 FortConversationTaskNode_CompleteQuestObjectiveByBackendName_2 {
  id=#16
  class=FortConversationTaskNode_CompleteQuestObjectiveByBackendName
  properties {objectiveBackendName=quest_s17_w02_legendary_q3_obj2, questDefinition=&quest_s17_w02_legendary_q3}
  sub_nodes [FortConversationRequirement_ControllerRequirement_18 {
    class=FortConversationRequirement_ControllerRequirement
    properties {
     nodeName=Require: Has Quest To Turn In
     requirement {
      class=FortControllerRequirement_HasQuestToTurnIn
      properties {bInverseCondition=true, softQuestToTurnIn=&quest_s17_w01_legendary_q1}
     }
    }
   }]
 }
 Speech_C_12 {
  id=#3
  class=Speech_C
  outputs [#17]
  properties {softTaskIcon=&T_UI_NPC_MainQuest, generalConfig {defaultMessage="Another IO <b>loyalty check. </>"}}
  sub_nodes [
   ConversationChoiceNode_12 {
    class=ConversationChoiceNode
    properties {
     defaultChoiceDisplayText=Imagined Order?
     choiceTags {gameplayTags [{tagName=Conversation.Choice.Category.MainQuestline}, {tagName=Conversation.Choice.AutoSelectIfValid}]}
    }
   }
   FortConversationRequirement_ControllerRequirement_0 {
    class=FortConversationRequirement_ControllerRequirement
    properties {
     nodeName=Require: Has Completed Quest
     requirement {
      class=FortControllerRequirement_HasCompletedQuest
      properties {bInverseCondition=true, questDefinition=&quest_s17_w02_legendary_q3}
     }
    }
   }
   FortConversationRequirement_ControllerRequirement_6 {
    class=FortConversationRequirement_ControllerRequirement
    properties {
     nodeName=Require: Is Quest Active
     requirement {
      class=FortControllerRequirement_IsQuestActive
      properties {questDefinition=&quest_s17_w02_legendary_q3}
     }
    }
   }
   FortConversationRequirement_ControllerRequirement_7 {
    class=FortConversationRequirement_ControllerRequirement
    properties {
     nodeName=Require: Has Completed Quest Objective
     requirement {
      class=FortControllerRequirement_HasCompletedQuestObjective
      properties {
       bInverseCondition=true
       objectiveBackendName=quest_s17_w02_legendary_q3_obj1
       questDefinition=&quest_s17_w02_legendary_q3
      }
     }
    }
   }
   FortConversationRequirement_ControllerRequirement_8 {
    class=FortConversationRequirement_ControllerRequirement
    properties {
     nodeName=Require: NPC Has Conversation Tag
     participantId {tagName=Conversation.Participant.NPC}
     requirement {class=FortControllerRequirement_NPCHasConversationEntryTag, properties {queryTag {tagName=Conversation.Entry.Rook}}}
    }
   }
  ]
 }
 Speech_C_2 {
  id=#17
  class=Speech_C
  outputs [#18]
  properties {generalConfig {defaultMessage="I'll give <b>Doctor Slone</> this, \r\nshe's consistent. "}}
  sub_nodes [ConversationChoiceNode_7 {class=ConversationChoiceNode, properties {defaultChoiceDisplayText=Continue}}]
 }
 Speech_C_11 {
  id=#18
  class=Speech_C
  outputs [#19]
  properties {generalConfig {defaultMessage="As for me, I've got <b>nothing to prove. </>"}}
  sub_nodes [ConversationChoiceNode_6 {class=ConversationChoiceNode, properties {defaultChoiceDisplayText=Continue}}]
 }
 FortConversationTaskNode_CompleteQuestObjectiveByBackendName_1 {
  id=#19
  class=FortConversationTaskNode_CompleteQuestObjectiveByBackendName
  properties {objectiveBackendName=quest_s17_w02_legendary_q3_obj1, questDefinition=&quest_s17_w02_legendary_q3}
  sub_nodes [FortConversationRequirement_ControllerRequirement_13 {
    class=FortConversationRequirement_ControllerRequirement
    properties {
     nodeName=Require: Has Quest To Turn In
     requirement {
      class=FortControllerRequirement_HasQuestToTurnIn
      properties {bInverseCondition=true, softQuestToTurnIn=&quest_s17_w01_legendary_q1}
     }
    }
   }]
 }
 Speech_C_1 {
  id=#2
  class=Speech_C
  outputs [#20]
  properties {
   softTaskIcon=&T_UI_NPC_MainQuest
   generalConfig {defaultMessage="The IO is in <enemy>shambles</> and crawling with <enemy>moles,</> so if Slone offers you a job, <b>don't be too flattered. </>"}
  }
  sub_nodes [
   ConversationChoiceNode_3 {
    class=ConversationChoiceNode
    properties {
     defaultChoiceDisplayText=Imagined Order?
     choiceTags {gameplayTags [{tagName=Conversation.Choice.Category.MainQuestline}, {tagName=Conversation.Choice.AutoSelectIfValid}]}
    }
   }
   FortConversationRequirement_ControllerRequirement_1 {
    class=FortConversationRequirement_ControllerRequirement
    properties {
     nodeName=Require: Has Completed Quest
     requirement {
      class=FortControllerRequirement_HasCompletedQuest
      properties {bInverseCondition=true, questDefinition=&quest_s17_w02_legendary_q3}
     }
    }
   }
   FortConversationRequirement_ControllerRequirement_2 {
    class=FortConversationRequirement_ControllerRequirement
    properties {
     nodeName=Require: Is Quest Active
     requirement {
      class=FortControllerRequirement_IsQuestActive
      properties {questDefinition=&quest_s17_w02_legendary_q3}
     }
    }
   }
   FortConversationRequirement_ControllerRequirement_3 {
    class=FortConversationRequirement_ControllerRequirement
    properties {
     nodeName=Require: Has Completed Quest Objective
     requirement {
      class=FortControllerRequirement_HasCompletedQuestObjective
      properties {
       bInverseCondition=true
       objectiveBackendName=quest_s17_w02_legendary_q3_obj0
       questDefinition=&quest_s17_w02_legendary_q3
      }
     }
    }
   }
   FortConversationRequirement_ControllerRequirement_4 {
    class=FortConversationRequirement_ControllerRequirement
    properties {
     nodeName=Require: NPC Has Conversation Tag
     participantId {tagName=Conversation.Participant.NPC}
     requirement {class=FortControllerRequirement_NPCHasConversationEntryTag, properties {queryTag {tagName=Conversation.Entry.Rickus}}}
    }
   }
  ]
 }
 FortConversationTaskNode_CompleteQuestObjectiveByBackendName_0 {
  id=#20
  class=FortConversationTaskNode_CompleteQuestObjectiveByBackendName
  properties {objectiveBackendName=quest_s17_w02_legendary_q3_obj0, questDefinition=&quest_s17_w02_legendary_q3}
  sub_nodes [FortConversationRequirement_ControllerRequirement_5 {
    class=FortConversationRequirement_ControllerRequirement
    properties {
     nodeName=Require: Has Quest To Turn In
     requirement {
      class=FortControllerRequirement_HasQuestToTurnIn
      properties {bInverseCondition=true, softQuestToTurnIn=&quest_s17_w01_legendary_q1}
     }
    }
   }]
 }
]
