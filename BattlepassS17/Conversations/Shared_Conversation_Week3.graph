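# Shared NPC conversation database for Season 17 Week 3, giving alien-invasion dialogue for legendary quest quest_s17_w03_legendary_q1_b.
# One entry point offers three auto-selected "Beach Party!" branches, gated on NPC tag (HenchmanSummer, Faux or Believer), active quest and unfinished objective.
# Each branch chains speech lines into completion of its own objective (obj0 to obj2).
# Those completion tasks carry an inverted turn-in check referencing the Week 1 legendary quest, not Week 3.
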ConversationDatabase /BattlepassS17/Conversations/Shared_Conversation_Week3
entry_points [{tag=Conversation.Entry.Season17.Week3, destinations [#1]}]
nodes [
 ConversationEntryPointNode_7 {
  id=#1
  class=ConversationEntryPointNode
  outputs [#2, #3, #4]
  properties {entryTag {tagName=Conversation.Entry.Season17.Week3}}
 }
 Speech_C_2 {
  id=#4
  class=Speech_C
  outputs [#5]
  properties {
   softTaskIcon=&T_UI_NPC_MainQuest
   generalConfig {defaultMessage=I get that you're having a little invasion or what have you1}
  }
  sub_nodes [
   ConversationChoiceNode_8 {
    class=ConversationChoiceNode
    properties {
     defaultChoiceDisplayText="Beach Party!"
     choiceTags {gameplayTags [{tagName=Conversation.Choice.Category.MainQuestline}, {tagName=Conversation.Choice.AutoSelectIfValid}]}
    }
   }
   FortConversationRequirement_ControllerRequirement_8 {
    class=FortConversationRequirement_ControllerRequirement
    properties {
     nodeName=Require: Has Completed Quest
     requirement {
      class=FortControllerRequirement_HasCompletedQuest
      properties {bInverseCondition=true, questDefinition=/BattlepassS17/Items/QuestItems/Legendary/quest_s17_w03_legendary_q1_b}
     }
    }
   }
   FortConversationRequirement_ControllerRequirement_9 {
    class=FortConversationRequirement_ControllerRequirement
    properties {
     nodeName=Require: Is Quest Active
     requirement {
      class=FortControllerRequirement_IsQuestActive
      properties {questDefinition=&quest_s17_w03_legendary_q1_b}
     }
    }
   }
   FortConversationRequirement_ControllerRequirement_10 {
    class=FortConversationRequirement_ControllerRequirement
    properties {
     nodeName=Require: Has Completed Quest Objective
     requirement {
      class=FortControllerRequirement_HasCompletedQuestObjective
      properties {
       bInverseCondition=true
       objectiveBackendName=quest_s17_w03_legendary_q1_b_obj0
       questDefinition=&quest_s17_w03_legendary_q1_b
      }
     }
    }
   }
   FortConversationRequirement_ControllerRequirement_52 {
    class=FortConversationRequirement_ControllerRequirement
    properties {
     nodeName=Require: NPC Has Conversation Tag
     participantId {tagName=Conversation.Participant.NPC}
     requirement {class=FortControllerRequirement_NPCHasConversationEntryTag, properties {queryTag {tagName=Conversation.Entry.HenchmanSummer}}}
    }
   }
  ]
 }
 Speech_C_4 {
  id=#5
  class=Speech_C
  outputs [#6]
  properties {generalConfig {defaultMessage="But I booked this vacation months ago\r\nand I intend to relax. Aliens or no. "}}
  sub_nodes [ConversationChoiceNode_5 {class=ConversationChoiceNode, properties {defaultChoiceDisplayText=Continue}}]
 }
 Speech_C_6 {
  id=#6
  class=Speech_C
  outputs [#7]
  properties {generalConfig {defaultMessage="So what I'm sayin is, I don't care."}}
  sub_nodes [ConversationChoiceNode_24 {class=ConversationChoiceNode, properties {defaultChoiceDisplayText=Continue}}]
 }
 FortConversationTaskNode_CompleteQuestObjectiveByBackendName_1 {
  id=#7
  class=FortConversationTaskNode_CompleteQuestObjectiveByBackendName
  properties {
   objectiveBackendName=quest_s17_w03_legendary_q1_b_obj0
   questDefinition=&quest_s17_w03_legendary_q1_b
  }
  sub_nodes [FortConversationRequirement_ControllerRequirement_0 {
    class=FortConversationRequirement_ControllerRequirement
    properties {
     nodeName=Require: Has Quest To Turn In
     requirement {
      class=FortControllerRequirement_HasQuestToTurnIn
      properties {bInverseCondition=true, softQuestToTurnIn=/BattlepassS17/Items/QuestItems/Legendary/quest_s17_w01_legendary_q1}
     }
    }
   }]
 }
 Speech_C_56 {
  id=#3
  class=Speech_C
  outputs [#8]
  properties {softTaskIcon=&T_UI_NPC_MainQuest, generalConfig {defaultMessage=You won't catch me welcoming those freaks.}}
  sub_nodes [
   ConversationChoiceNode_60 {
    class=ConversationChoiceNode
    properties {
     defaultChoiceDisplayText="Beach Party!"
     choiceTags {gameplayTags [{tagName=Conversation.Choice.Category.MainQuestline}, {tagName=Conversation.Choice.AutoSelectIfValid}]}
    }
   }
   FortConversationRequirement_ControllerRequirement_77 {
    class=FortConversationRequirement_ControllerRequirement
    properties {
     nodeName=Require: Has Completed Quest
     requirement {
      class=FortControllerRequirement_HasCompletedQuest
      properties {bInverseCondition=true, questDefinition=&quest_s17_w03_legendary_q1_b}
     }
    }
   }
   FortConversationRequirement_ControllerRequirement_78 {
    class=FortConversationRequirement_ControllerRequirement
    properties {
     nodeName=Require: Is Quest Active
     requirement {
      class=FortControllerRequirement_IsQuestActive
      properties {questDefinition=&quest_s17_w03_legendary_q1_b}
     }
    }
   }
   FortConversationRequirement_ControllerRequirement_79 {
    class=FortConversationRequirement_ControllerRequirement
    properties {
     nodeName=Require: Has Completed Quest Objective
     requirement {
      class=FortControllerRequirement_HasCompletedQuestObjective
      properties {
       bInverseCondition=true
       objectiveBackendName=quest_s17_w03_legendary_q1_b_obj1
       questDefinition=&quest_s17_w03_legendary_q1_b
      }
     }
    }
   }
   FortConversationRequirement_ControllerRequirement_80 {
    class=FortConversationRequirement_ControllerRequirement
    properties {
     nodeName=Require: NPC Has Conversation Tag
     participantId {tagName=Conversation.Participant.NPC}
     requirement {class=FortControllerRequirement_NPCHasConversationEntryTag, properties {queryTag {tagName=Conversation.Entry.Faux}}}
    }
   }
  ]
 }
 Speech_C_57 {
  id=#8
  class=Speech_C
  outputs [#9]
  properties {generalConfig {defaultMessage="If I had my way we'd already be at war. "}}
  sub_nodes [ConversationChoiceNode_61 {class=ConversationChoiceNode, properties {defaultChoiceDisplayText=Continue}}]
 }
 FortConversationTaskNode_CompleteQuestObjectiveByBackendName_11 {
  id=#9
  class=FortConversationTaskNode_CompleteQuestObjectiveByBackendName
  properties {
   objectiveBackendName=quest_s17_w03_legendary_q1_b_obj1
   questDefinition=&quest_s17_w03_legendary_q1_b
  }
  sub_nodes [FortConversationRequirement_ControllerRequirement_81 {
    class=FortConversationRequirement_ControllerRequirement
    properties {
     nodeName=Require: Has Quest To Turn In
     requirement {
      class=FortControllerRequirement_HasQuestToTurnIn
      properties {bInverseCondition=true, softQuestToTurnIn=&quest_s17_w01_legendary_q1}
     }
    }
   }]
 }
 Speech_C_1 {
  id=#2
  class=Speech_C
  outputs [#10]
  properties {
   softTaskIcon=&T_UI_NPC_MainQuest
   generalConfig {defaultMessage="I've lived on this planet my whole life\r\nand there's no way these aliens are worse than us. "}
  }
  sub_nodes [
   ConversationChoiceNode_20 {
    class=ConversationChoiceNode
    properties {
     defaultChoiceDisplayText="Beach Party!"
     choiceTags {gameplayTags [{tagName=Conversation.Choice.Category.MainQuestline}, {tagName=Conversation.Choice.AutoSelectIfValid}]}
    }
   }
   FortConversationRequirement_ControllerRequirement_40 {
    class=FortConversationRequirement_ControllerRequirement
    properties {
     nodeName=Require: Has Completed Quest
     requirement {
      class=FortControllerRequirement_HasCompletedQuest
      properties {bInverseCondition=true, questDefinition=&quest_s17_w03_legendary_q1_b}
     }
    }
   }
   FortConversationRequirement_ControllerRequirement_41 {
    class=FortConversationRequirement_ControllerRequirement
    properties {
     nodeName=Require: Is Quest Active
     requirement {
      class=FortControllerRequirement_IsQuestActive
      properties {questDefinition=&quest_s17_w03_legendary_q1_b}
     }
    }
   }
   FortConversationRequirement_ControllerRequirement_42 {
    class=FortConversationRequirement_ControllerRequirement
    properties {
     nodeName=Require: Has Completed Quest Objective
     requirement {
      class=FortControllerRequirement_HasCompletedQuestObjective
      properties {
       bInverseCondition=true
       objectiveBackendName=quest_s17_w03_legendary_q1_b_obj2
       questDefinition=&quest_s17_w03_legendary_q1_b
      }
     }
    }
   }
   FortConversationRequirement_ControllerRequirement_58 {
    class=FortConversationRequirement_ControllerRequirement
    properties {
     nodeName=Require: NPC Has Conversation Tag
     participantId {tagName=Conversation.Participant.NPC}
     requirement {class=FortControllerRequirement_NPCHasConversationEntryTag, properties {queryTag {tagName=Conversation.Entry.Believer}}}
    }
   }
  ]
 }
 Speech_C_7 {
  id=#10
  class=Speech_C
  outputs [#11]
  properties {generalConfig {defaultMessage="And even if they came here to destroy us, can you blame them?"}}
  sub_nodes [ConversationChoiceNode_3 {class=ConversationChoiceNode, properties {defaultChoiceDisplayText=Continue}}]
 }
 Speech_C_0 {
  id=#11
  class=Speech_C
  outputs [#12]
  properties {generalConfig {defaultMessage=We're the worst.}}
  sub_nodes [ConversationChoiceNode_23 {class=ConversationChoiceNode, properties {defaultChoiceDisplayText=Continue}}]
 }
 FortConversationTaskNode_CompleteQuestObjectiveByBackendName_0 {
  id=#12
  class=FortConversationTaskNode_CompleteQuestObjectiveByBackendName
  properties {
   objectiveBackendName=quest_s17_w03_legendary_q1_b_obj2
   questDefinition=&quest_s17_w03_legendary_q1_b
  }
  sub_nodes [FortConversationRequirement_ControllerRequirement_66 {
    class=FortConversationRequirement_ControllerRequirement
    properties {
     nodeName=Require: Has Quest To Turn In
     requirement {
      class=FortControllerRequirement_HasQuestToTurnIn
      properties {bInverseCondition=true, softQuestToTurnIn=&quest_s17_w01_legendary_q1}
     }
    }
   }]
 }
]
