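# Season 17 Week 5 NPC conversation for legendary quest q5, where a Believer talks to a player and their parasite.
# The opening main-quest line auto-selects when the quest is active, quest and objective are incomplete, the NPC has the Believer tag, and an AvacadoEaterBird immunity tag is owned.
# A linear chain of five speeches, advanced by Continue, ends by completing objective obj0, gated by an inverted quest-turn-in check.
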ConversationDatabase /BattlepassS17/Conversations/Shared_Conversation_Week5
entry_points [{tag=Conversation.Entry.Season17.Week5, destinations [#1]}]
nodes [
 ConversationEntryPointNode_7 {
  id=#1
  class=ConversationEntryPointNode
  outputs [#2]
  properties {entryTag {tagName=Conversation.Entry.Season17.Week5}}
 }
 Speech_C_0 {
  id=#2
  class=Speech_C
  outputs [#3]
  properties {softTaskIcon=&T_UI_NPC_MainQuest, generalConfig {defaultMessage="You did it!\r\nSo... what's it like?"}}
  sub_nodes [
   ConversationChoiceNode_4 {
    class=ConversationChoiceNode
    properties {
     defaultChoiceDisplayText="[ph] Greetings Earthling!"
     choiceTags {gameplayTags [{tagName=Conversation.Choice.Category.MainQuestline}, {tagName=Conversation.Choice.AutoSelectIfValid}]}
    }
   }
   FortConversationRequirement_ControllerRequirement_9 {
    class=FortConversationRequirement_ControllerRequirement
    properties {
     nodeName=Require: Has Completed Quest
     requirement {
      class=FortControllerRequirement_HasCompletedQuest
      properties {bInverseCondition=true, questDefinition=/BattlepassS17/Items/QuestItems/Legendary/quest_s17_w05_legendary_q5}
     }
    }
   }
   FortConversationRequirement_ControllerRequirement_10 {
    class=FortConversationRequirement_ControllerRequirement
    properties {
     nodeName=Require: Is Quest Active
     requirement {
      class=FortControllerRequirement_IsQuestActive
      properties {questDefinition=&quest_s17_w05_legendary_q5}
     }
    }
   }
   FortConversationRequirement_ControllerRequirement_11 {
    class=FortConversationRequirement_ControllerRequirement
    properties {
     nodeName=Require: Has Completed Quest Objective
     requirement {
      class=FortControllerRequirement_HasCompletedQuestObjective
      properties {
       bInverseCondition=true
       objectiveBackendName=quest_s17_w05_legendary_q5_obj0
       questDefinition=&quest_s17_w05_legendary_q5
      }
     }
    }
   }
   FortConversationRequirement_ControllerRequirement_12 {
    class=FortConversationRequirement_ControllerRequirement
    properties {
     nodeName=Require: NPC Has Conversation Tag
     participantId {tagName=Conversation.Participant.NPC}
     requirement {class=FortControllerRequirement_NPCHasConversationEntryTag, properties {queryTag {tagName=Conversation.Entry.Believer}}}
    }
   }
   FortConversationRequirement_ControllerRequirement_25 {
    class=FortConversationRequirement_ControllerRequirement
    properties {
     nodeName=Require: Has GameplayTag
     requirement {class=FortControllerRequirement_HasOwnedTag, properties {ownedTag {tagName=Granted.Athena.EnvItem.AvacadoEaterBird.ImmunityFromBirds}}}
    }
   }
  ]
 }
 Speech_C_7 {
  id=#3
  class=Speech_C
  outputs [#4]
  properties {generalConfig {defaultMessage="Hello? Hello?! Blink if you can hear me."}}
  sub_nodes [ConversationChoiceNode_5 {class=ConversationChoiceNode, properties {defaultChoiceDisplayText=Continue}}]
 }
 Speech_C_17 {
  id=#4
  class=Speech_C
  outputs [#5]
  properties {generalConfig {defaultMessage=I guess maybe I should just talk to the parasite...}}
  sub_nodes [ConversationChoiceNode_3 {class=ConversationChoiceNode, properties {defaultChoiceDisplayText=Continue}}]
 }
 Speech_C_1 {
  id=#5
  class=Speech_C
  outputs [#6]
  properties {generalConfig {defaultMessage="WELCOME TO OUR HOME!"}}
  sub_nodes [ConversationChoiceNode_6 {class=ConversationChoiceNode, properties {defaultChoiceDisplayText=Continue}}]
 }
 Speech_C_2 {
  id=#6
  class=Speech_C
  outputs [#7]
  properties {generalConfig {defaultMessage="IF YOU WANT TO DESTROY IT, \r\nTHAT'S FINE.\r\nJUST BRING ME WITH YOU."}}
  sub_nodes [ConversationChoiceNode_17 {class=ConversationChoiceNode, properties {defaultChoiceDisplayText=Continue}}]
 }
 FortConversationTaskNode_CompleteQuestObjectiveByBackendName_5 {
  id=#7
  class=FortConversationTaskNode_CompleteQuestObjectiveByBackendName
  properties {objectiveBackendName=quest_s17_w05_legendary_q5_obj0, questDefinition=&quest_s17_w05_legendary_q5}
  sub_nodes [FortConversationRequirement_ControllerRequirement_24 {
    class=FortConversationRequirement_ControllerRequirement
    properties {
     nodeName=Require: Has Quest To Turn In
     requirement {
      class=FortControllerRequirement_HasQuestToTurnIn
      properties {bInverseCondition=true, softQuestToTurnIn=&quest_s17_w05_legendary_q5}
     }
    }
   }]
 }
]
