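# Conversation database for the Season 17 Week 8 Brawny NPC dialogue, where the player interrogates an apparent infiltrator for legendary quest q5.
# The auto-select Interrogate choice requires the quest active and uncompleted, objective obj15 unfinished, and the NPC carrying the BrawnySmasher entry tag.
# Answering "Um, no..." completes five quest objectives through chained task nodes.
# The player then drops the subject, returning to start, or chooses Reveal Infiltrator, triggering the disguise reveal.
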ConversationDatabase /BattlepassS17/Conversations/Shared_Conversation_Week8_Brawny
entry_points [{tag=Conversation.Entry.Season17.Brawny, destinations [#1]}]
nodes [
 ConversationEntryPointNode_7 {
  id=#1
  class=ConversationEntryPointNode
  outputs [#2]
  properties {entryTag {tagName=Conversation.Entry.Season17.Brawny}}
 }
 Speech_C_0 {
  id=#2
  class=Speech_C
  outputs [#3, #4]
  properties {
   softTaskIcon=&T_UI_NPC_MainQuest
   generalConfig {defaultMessage="If it isn't my <b>best human friend! </>Remember me?"}
  }
  sub_nodes [
   ConversationChoiceNode_4 {
    class=ConversationChoiceNode
    properties {
     defaultChoiceDisplayText=Interrogate
     choiceTags {gameplayTags [{tagName=Conversation.Choice.Category.MainQuestline}, {tagName=Conversation.Choice.AutoSelectIfValid}]}
    }
   }
   FortConversationRequirement_ControllerRequirement_10 {
    class=FortConversationRequirement_ControllerRequirement
    properties {
     nodeName=Require: Is Quest Active
     requirement {
      class=FortControllerRequirement_IsQuestActive
      properties {questDefinition=/BattlepassS17/Items/QuestItems/Legendary/quest_s17_w08_legendary_q5}
     }
    }
   }
   FortConversationRequirement_ControllerRequirement_12 {
    class=FortConversationRequirement_ControllerRequirement
    properties {
     nodeName=Require: NPC Has Conversation Tag
     participantId {tagName=Conversation.Participant.NPC}
     requirement {class=FortControllerRequirement_NPCHasConversationEntryTag, properties {queryTag {tagName=Conversation.Entry.BrawnySmasher}}}
    }
   }
   FortConversationRequirement_ControllerRequirement_9 {
    class=FortConversationRequirement_ControllerRequirement
    properties {
     nodeName=Require: Has Completed Quest
     requirement {
      class=FortControllerRequirement_HasCompletedQuest
      properties {bInverseCondition=true, questDefinition=&quest_s17_w08_legendary_q5}
     }
    }
   }
   FortConversationRequirement_ControllerRequirement_0 {
    class=FortConversationRequirement_ControllerRequirement
    properties {
     nodeName=Require: Has Completed Quest Objective
     requirement {
      class=FortControllerRequirement_HasCompletedQuestObjective
      properties {
       bInverseCondition=true
       objectiveBackendName=quest_s17_w08_legendary_q5_obj15
       questDefinition=&quest_s17_w08_legendary_q5
      }
     }
    }
   }
  ]
 }
 FortConversationTaskNode_CompleteQuestObjectiveByBackendName_0 {
  id=#4
  class=FortConversationTaskNode_CompleteQuestObjectiveByBackendName
  outputs [#5]
  properties {objectiveBackendName=quest_s17_w08_legendary_q5_obj15, questDefinition=&quest_s17_w08_legendary_q5}
  sub_nodes [ConversationChoiceNode_10 {class=ConversationChoiceNode, properties {defaultChoiceDisplayText="Um, no..."}}]
 }
 FortConversationTaskNode_CompleteQuestObjectiveByBackendName_1 {
  id=#5
  class=FortConversationTaskNode_CompleteQuestObjectiveByBackendName
  outputs [#6]
  properties {objectiveBackendName=quest_s17_w08_legendary_q5_obj0, questDefinition=&quest_s17_w08_legendary_q5}
 }
 FortConversationTaskNode_CompleteQuestObjectiveByBackendName_2 {
  id=#6
  class=FortConversationTaskNode_CompleteQuestObjectiveByBackendName
  outputs [#7]
  properties {objectiveBackendName=quest_s17_w08_legendary_q5_obj1, questDefinition=&quest_s17_w08_legendary_q5}
 }
 FortConversationTaskNode_CompleteQuestObjectiveByBackendName_3 {
  id=#7
  class=FortConversationTaskNode_CompleteQuestObjectiveByBackendName
  outputs [#8]
  properties {objectiveBackendName=quest_s17_w08_legendary_q5_obj2, questDefinition=&quest_s17_w08_legendary_q5}
 }
 FortConversationTaskNode_CompleteQuestObjectiveByBackendName_4 {
  id=#8
  class=FortConversationTaskNode_CompleteQuestObjectiveByBackendName
  outputs [#9]
  properties {objectiveBackendName=quest_s17_w08_legendary_q5_obj3, questDefinition=&quest_s17_w08_legendary_q5}
 }
 Speech_C_7 {
  id=#9
  class=Speech_C
  outputs [#10, #11]
  properties {
   softTaskIcon=&T_UI_NPC_question
   generalConfig {defaultMessage="<enemy>Drop the act.</> \r\nYou <b>really</> don't remember me?"}
  }
 }
 Speech_C_3 {
  id=#11
  class=Speech_C
  outputs [#12]
  properties {
   softTaskIcon=/Game/Athena/HUD/Interaction/T_UI_NPC_exclamation
   generalConfig {defaultMessage="Them? You mean <b>us?</>\r\nAh! You don't remember...\r\n<enemy>You will.</>"}
  }
  sub_nodes [ConversationChoiceNode_12 {class=ConversationChoiceNode, properties {defaultChoiceDisplayText=Reveal Infiltrator}}]
 }
 RevealDisguise_C_0 {
  id=#12
  class=RevealDisguise_C
  outputs []
  properties {softTaskIcon=&T_UI_NPC_Intel2}
 }
 Speech_C_1 {
  id=#10
  class=Speech_C
  outputs [#13]
  properties {generalConfig {defaultMessage="Give it time. \r\n<b>Your directive</> from <enemy>Kymera</> will come back to you."}}
  sub_nodes [ConversationChoiceNode_3 {class=ConversationChoiceNode, properties {defaultChoiceDisplayText=Drop the subject}}]
 }
 ReturnToStart_C_0 {id=#13, class=ReturnToStart_C}
 Back_C_0 {
  id=#3
  class=Back_C
  sub_nodes [ConversationChoiceNode_9 {class=ConversationChoiceNode, properties {defaultChoiceDisplayText=BACK}}]
 }
]
