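# Shared NPC conversation for the Season 17 Week 8 legendary quest (q5), where players interrogate NPCs to find an alien impostor.
# The entry fans out to auto-selectable "Interrogate" tasks, each gated by the NPC's conversation tag and completing that NPC's quest objective.
# Each task requires the quest active, with its own objective and obj15 not yet done.
# All tasks share one speech node that picks its line by completed-objective count, then returns to start.
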ConversationDatabase /BattlepassS17/Conversations/Shared_Conversation_Week8_Impostor
entry_points [{tag=Conversation.Entry.Season17.Week8.Impostor, destinations [#1]}]
nodes [
 ConversationEntryPointNode_7 {
  id=#1
  class=ConversationEntryPointNode
  outputs [
   #2
   #3
   #4
   #5
   #6
   #7
   #8
   #9
   #10
   #11
   #12
   #13
  ]
  properties {entryTag {tagName=Conversation.Entry.Season17.Week8.Impostor}}
 }
 FortConversationTaskNode_CompleteQuestObjectiveByBackendName_14 {
  id=#13
  class=FortConversationTaskNode_CompleteQuestObjectiveByBackendName
  outputs [#14]
  properties {
   objectiveBackendName=quest_s17_w08_legendary_q5_obj14
   questDefinition=/BattlepassS17/Items/QuestItems/Legendary/quest_s17_w08_legendary_q5
   softTaskIcon=/Game/Athena/HUD/Interaction/T_UI_NPC_Interrobang
  }
  sub_nodes [
   FortConversationRequirement_ControllerRequirement_17 {
    class=FortConversationRequirement_ControllerRequirement
    properties {
     nodeName=Require: Has Completed Quest
     requirement {
      class=FortControllerRequirement_HasCompletedQuest
      properties {bInverseCondition=true, questDefinition=&quest_s17_w08_legendary_q5}
     }
    }
   }
   FortConversationRequirement_ControllerRequirement_18 {
    class=FortConversationRequirement_ControllerRequirement
    properties {
     nodeName=Require: Is Quest Active
     requirement {
      class=FortControllerRequirement_IsQuestActive
      properties {questDefinition=&quest_s17_w08_legendary_q5}
     }
    }
   }
   FortConversationRequirement_ControllerRequirement_19 {
    class=FortConversationRequirement_ControllerRequirement
    properties {
     nodeName=Require: Has Completed Quest Objective
     requirement {
      class=FortControllerRequirement_HasCompletedQuestObjective
      properties {
       bInverseCondition=true
       objectiveBackendName=quest_s17_w08_legendary_q5_obj14
       questDefinition=&quest_s17_w08_legendary_q5
      }
     }
    }
   }
   FortConversationRequirement_ControllerRequirement_20 {
    class=FortConversationRequirement_ControllerRequirement
    properties {
     nodeName=Require: NPC Has Conversation Tag
     participantId {tagName=Conversation.Participant.NPC}
     requirement {class=FortControllerRequirement_NPCHasConversationEntryTag, properties {queryTag {tagName=Conversation.Entry.AlienSummer}}}
    }
   }
   ConversationChoiceNode_6 {
    class=ConversationChoiceNode
    properties {
     defaultChoiceDisplayText=Interrogate
     choiceTags {gameplayTags [{tagName=Conversation.Choice.Category.MainQuestline}, {tagName=Conversation.Choice.AutoSelectIfValid}]}
    }
   }
   FortConversationRequirement_ControllerRequirement_83 {
    class=FortConversationRequirement_ControllerRequirement
    properties {
     nodeName=Require: Has Completed Quest Objective
     requirement {
      class=FortControllerRequirement_HasCompletedQuestObjective
      properties {
       bInverseCondition=true
       objectiveBackendName=quest_s17_w08_legendary_q5_obj15
       questDefinition=&quest_s17_w08_legendary_q5
      }
     }
    }
   }
  ]
 }
 Speech_C_12 {
  id=#14
  class=Speech_C
  outputs [#15]
  properties {
   softTaskIcon=&T_UI_NPC_MainQuest
   generalConfig {
    defaultMessage="It's <b>not me...</>\r\nBut what if it's <enemy>you? </>"
    contextualMessages [
     {
      message="The aliens could impersonate any one of us… <b>except me.</>"
      contextRequirements [{requirement {
         class=FortControllerRequirement_HasCompletedNumQuestObjectives
         properties {questDefinition=&quest_s17_w08_legendary_q5, requiredCount {requirement=Equal, quantity {value=1}}}
        }}]
      priority {value=0, curve {curveTable=null, rowName=None}}
     }
     {
      message="You'll find the Infiltrator soon enough.\r\nI heard they only eat <b>cat food. </>\r\nHard to hide that habit."
      contextRequirements [{requirement {
         class=FortControllerRequirement_HasCompletedNumQuestObjectives
         properties {questDefinition=&quest_s17_w08_legendary_q5, requiredCount {requirement=Equal, quantity {value=2}}}
        }}]
      priority {value=0, curve {curveTable=null, rowName=None}}
     }
     {
      message="Does it look like I'm blinking with <enemy>vertical eyelids? </>\r\nAsk someone else."
      contextRequirements [{requirement {
         class=FortControllerRequirement_HasCompletedNumQuestObjectives
         properties {questDefinition=&quest_s17_w08_legendary_q5, requiredCount {requirement=Equal, quantity {value=3}}}
        }}]
      priority {value=0, curve {curveTable=null, rowName=None}}
     }
     {
      message="I heard the alien Infiltrator smells just like <enemy>wet leather.</>\r\nAnd I smell like... <b>me. </>"
      contextRequirements [{requirement {
         class=FortControllerRequirement_HasCompletedNumQuestObjectives
         properties {questDefinition=&quest_s17_w08_legendary_q5, requiredCount {requirement=Equal, quantity {value=4}}}
        }}]
      priority {value=0, curve {curveTable=null, rowName=None}}
     }
     {
      message="It's <b>not me...</>\r\nBut what if it's <enemy>you? </>"
      contextRequirements [{requirement {
         class=FortControllerRequirement_HasCompletedNumQuestObjectives
         properties {questDefinition=&quest_s17_w08_legendary_q5, requiredCount {requirement=Equal, quantity {value=5}}}
        }}]
      priority {value=0, curve {curveTable=null, rowName=None}}
     }
    ]
   }
  }
 }
 ReturnToStart_C_0 {id=#15, class=ReturnToStart_C}
 FortConversationTaskNode_CompleteQuestObjectiveByBackendName_13 {
  id=#12
  class=FortConversationTaskNode_CompleteQuestObjectiveByBackendName
  outputs [#14]
  properties {
   objectiveBackendName=quest_s17_w08_legendary_q5_obj13
   questDefinition=&quest_s17_w08_legendary_q5
   softTaskIcon=&T_UI_NPC_Interrobang
  }
  sub_nodes [
   FortConversationRequirement_ControllerRequirement_61 {
    class=FortConversationRequirement_ControllerRequirement
    properties {
     nodeName=Require: Has Completed Quest
     requirement {
      class=FortControllerRequirement_HasCompletedQuest
      properties {bInverseCondition=true, questDefinition=&quest_s17_w08_legendary_q5}
     }
    }
   }
   FortConversationRequirement_ControllerRequirement_62 {
    class=FortConversationRequirement_ControllerRequirement
    properties {
     nodeName=Require: Is Quest Active
     requirement {
      class=FortControllerRequirement_IsQuestActive
      properties {questDefinition=&quest_s17_w08_legendary_q5}
     }
    }
   }
   FortConversationRequirement_ControllerRequirement_63 {
    class=FortConversationRequirement_ControllerRequirement
    properties {
     nodeName=Require: Has Completed Quest Objective
     requirement {
      class=FortControllerRequirement_HasCompletedQuestObjective
      properties {
       bInverseCondition=true
       objectiveBackendName=quest_s17_w08_legendary_q5_obj13
       questDefinition=&quest_s17_w08_legendary_q5
      }
     }
    }
   }
   FortConversationRequirement_ControllerRequirement_64 {
    class=FortConversationRequirement_ControllerRequirement
    properties {
     nodeName=Require: NPC Has Conversation Tag
     participantId {tagName=Conversation.Participant.NPC}
     requirement {class=FortControllerRequirement_NPCHasConversationEntryTag, properties {queryTag {tagName=Conversation.Entry.BunkerJonesy}}}
    }
   }
   ConversationChoiceNode_17 {
    class=ConversationChoiceNode
    properties {
     defaultChoiceDisplayText=Interrogate
     choiceTags {gameplayTags [{tagName=Conversation.Choice.Category.MainQuestline}, {tagName=Conversation.Choice.AutoSelectIfValid}]}
    }
   }
   FortConversationRequirement_ControllerRequirement_82 {
    class=FortConversationRequirement_ControllerRequirement
    properties {
     nodeName=Require: Has Completed Quest Objective
     requirement {
      class=FortControllerRequirement_HasCompletedQuestObjective
      properties {
       bInverseCondition=true
       objectiveBackendName=quest_s17_w08_legendary_q5_obj15
       questDefinition=&quest_s17_w08_legendary_q5
      }
     }
    }
   }
  ]
 }
 FortConversationTaskNode_CompleteQuestObjectiveByBackendName_11 {
  id=#11
  class=FortConversationTaskNode_CompleteQuestObjectiveByBackendName
  outputs [#14]
  properties {
   objectiveBackendName=quest_s17_w08_legendary_q5_obj11
   questDefinition=&quest_s17_w08_legendary_q5
   softTaskIcon=&T_UI_NPC_Interrobang
  }
  sub_nodes [
   FortConversationRequirement_ControllerRequirement_53 {
    class=FortConversationRequirement_ControllerRequirement
    properties {
     nodeName=Require: Has Completed Quest
     requirement {
      class=FortControllerRequirement_HasCompletedQuest
      properties {bInverseCondition=true, questDefinition=&quest_s17_w08_legendary_q5}
     }
    }
   }
   FortConversationRequirement_ControllerRequirement_54 {
    class=FortConversationRequirement_ControllerRequirement
    properties {
     nodeName=Require: Is Quest Active
     requirement {
      class=FortControllerRequirement_IsQuestActive
      properties {questDefinition=&quest_s17_w08_legendary_q5}
     }
    }
   }
   FortConversationRequirement_ControllerRequirement_55 {
    class=FortConversationRequirement_ControllerRequirement
    properties {
     nodeName=Require: Has Completed Quest Objective
     requirement {
      class=FortControllerRequirement_HasCompletedQuestObjective
      properties {
       bInverseCondition=true
       objectiveBackendName=quest_s17_w08_legendary_q5_obj11
       questDefinition=&quest_s17_w08_legendary_q5
      }
     }
    }
   }
   FortConversationRequirement_ControllerRequirement_56 {
    class=FortConversationRequirement_ControllerRequirement
    properties {
     nodeName=Require: NPC Has Conversation Tag
     participantId {tagName=Conversation.Participant.NPC}
     requirement {class=FortControllerRequirement_NPCHasConversationEntryTag, properties {queryTag {tagName=Conversation.Entry.Dreamflower}}}
    }
   }
   ConversationChoiceNode_15 {
    class=ConversationChoiceNode
    properties {
     defaultChoiceDisplayText=Interrogate
     choiceTags {gameplayTags [{tagName=Conversation.Choice.Category.MainQuestline}, {tagName=Conversation.Choice.AutoSelectIfValid}]}
    }
   }
   FortConversationRequirement_ControllerRequirement_81 {
    class=FortConversationRequirement_ControllerRequirement
    properties {
     nodeName=Require: Has Completed Quest Objective
     requirement {
      class=FortControllerRequirement_HasCompletedQuestObjective
      properties {
       bInverseCondition=true
       objectiveBackendName=quest_s17_w08_legendary_q5_obj15
       questDefinition=&quest_s17_w08_legendary_q5
      }
     }
    }
   }
  ]
 }
 FortConversationTaskNode_CompleteQuestObjectiveByBackendName_10 {
  id=#10
  class=FortConversationTaskNode_CompleteQuestObjectiveByBackendName
  outputs [#14]
  properties {
   objectiveBackendName=quest_s17_w08_legendary_q5_obj10
   questDefinition=&quest_s17_w08_legendary_q5
   softTaskIcon=&T_UI_NPC_Interrobang
  }
  sub_nodes [
   FortConversationRequirement_ControllerRequirement_49 {
    class=FortConversationRequirement_ControllerRequirement
    properties {
     nodeName=Require: Has Completed Quest
     requirement {
      class=FortControllerRequirement_HasCompletedQuest
      properties {bInverseCondition=true, questDefinition=&quest_s17_w08_legendary_q5}
     }
    }
   }
   FortConversationRequirement_ControllerRequirement_50 {
    class=FortConversationRequirement_ControllerRequirement
    properties {
     nodeName=Require: Is Quest Active
     requirement {
      class=FortControllerRequirement_IsQuestActive
      properties {questDefinition=&quest_s17_w08_legendary_q5}
     }
    }
   }
   FortConversationRequirement_ControllerRequirement_51 {
    class=FortConversationRequirement_ControllerRequirement
    properties {
     nodeName=Require: Has Completed Quest Objective
     requirement {
      class=FortControllerRequirement_HasCompletedQuestObjective
      properties {
       bInverseCondition=true
       objectiveBackendName=quest_s17_w08_legendary_q5_obj10
       questDefinition=&quest_s17_w08_legendary_q5
      }
     }
    }
   }
   FortConversationRequirement_ControllerRequirement_52 {
    class=FortConversationRequirement_ControllerRequirement
    properties {
     nodeName=Require: NPC Has Conversation Tag
     participantId {tagName=Conversation.Participant.NPC}
     requirement {class=FortControllerRequirement_NPCHasConversationEntryTag, properties {queryTag {tagName=Conversation.Entry.Abstrakt}}}
    }
   }
   ConversationChoiceNode_14 {
    class=ConversationChoiceNode
    properties {
     defaultChoiceDisplayText=Interrogate
     choiceTags {gameplayTags [{tagName=Conversation.Choice.Category.MainQuestline}, {tagName=Conversation.Choice.AutoSelectIfValid}]}
    }
   }
   FortConversationRequirement_ControllerRequirement_80 {
    class=FortConversationRequirement_ControllerRequirement
    properties {
     nodeName=Require: Has Completed Quest Objective
     requirement {
      class=FortControllerRequirement_HasCompletedQuestObjective
      properties {
       bInverseCondition=true
       objectiveBackendName=quest_s17_w08_legendary_q5_obj15
       questDefinition=&quest_s17_w08_legendary_q5
      }
     }
    }
   }
  ]
 }
 FortConversationTaskNode_CompleteQuestObjectiveByBackendName_9 {
  id=#9
  class=FortConversationTaskNode_CompleteQuestObjectiveByBackendName
  outputs [#14]
  properties {
   objectiveBackendName=quest_s17_w08_legendary_q5_obj9
   questDefinition=&quest_s17_w08_legendary_q5
   softTaskIcon=&T_UI_NPC_Interrobang
  }
  sub_nodes [
   FortConversationRequirement_ControllerRequirement_45 {
    class=FortConversationRequirement_ControllerRequirement
    properties {
     nodeName=Require: Has Completed Quest
     requirement {
      class=FortControllerRequirement_HasCompletedQuest
      properties {bInverseCondition=true, questDefinition=&quest_s17_w08_legendary_q5}
     }
    }
   }
   FortConversationRequirement_ControllerRequirement_46 {
    class=FortConversationRequirement_ControllerRequirement
    properties {
     nodeName=Require: Is Quest Active
     requirement {
      class=FortControllerRequirement_IsQuestActive
      properties {questDefinition=&quest_s17_w08_legendary_q5}
     }
    }
   }
   FortConversationRequirement_ControllerRequirement_47 {
    class=FortConversationRequirement_ControllerRequirement
    properties {
     nodeName=Require: Has Completed Quest Objective
     requirement {
      class=FortControllerRequirement_HasCompletedQuestObjective
      properties {
       bInverseCondition=true
       objectiveBackendName=quest_s17_w08_legendary_q5_obj9
       questDefinition=&quest_s17_w08_legendary_q5
      }
     }
    }
   }
   FortConversationRequirement_ControllerRequirement_48 {
    class=FortConversationRequirement_ControllerRequirement
    properties {
     nodeName=Require: NPC Has Conversation Tag
     participantId {tagName=Conversation.Participant.NPC}
     requirement {class=FortControllerRequirement_NPCHasConversationEntryTag, properties {queryTag {tagName=Conversation.Entry.SpecialForces}}}
    }
   }
   ConversationChoiceNode_13 {
    class=ConversationChoiceNode
    properties {
     defaultChoiceDisplayText=Interrogate
     choiceTags {gameplayTags [{tagName=Conversation.Choice.Category.MainQuestline}, {tagName=Conversation.Choice.AutoSelectIfValid}]}
    }
   }
   FortConversationRequirement_ControllerRequirement_79 {
    class=FortConversationRequirement_ControllerRequirement
    properties {
     nodeName=Require: Has Completed Quest Objective
     requirement {
      class=FortControllerRequirement_HasCompletedQuestObjective
      properties {
       bInverseCondition=true
       objectiveBackendName=quest_s17_w08_legendary_q5_obj15
       questDefinition=&quest_s17_w08_legendary_q5
      }
     }
    }
   }
  ]
 }
 FortConversationTaskNode_CompleteQuestObjectiveByBackendName_8 {
  id=#8
  class=FortConversationTaskNode_CompleteQuestObjectiveByBackendName
  outputs [#14]
  properties {
   objectiveBackendName=quest_s17_w08_legendary_q5_obj8
   questDefinition=&quest_s17_w08_legendary_q5
   softTaskIcon=&T_UI_NPC_Interrobang
  }
  sub_nodes [
   FortConversationRequirement_ControllerRequirement_41 {
    class=FortConversationRequirement_ControllerRequirement
    properties {
     nodeName=Require: Has Completed Quest
     requirement {
      class=FortControllerRequirement_HasCompletedQuest
      properties {bInverseCondition=true, questDefinition=&quest_s17_w08_legendary_q5}
     }
    }
   }
   FortConversationRequirement_ControllerRequirement_42 {
    class=FortConversationRequirement_ControllerRequirement
    properties {
     nodeName=Require: Is Quest Active
     requirement {
      class=FortControllerRequirement_IsQuestActive
      properties {questDefinition=&quest_s17_w08_legendary_q5}
     }
    }
   }
   FortConversationRequirement_ControllerRequirement_43 {
    class=FortConversationRequirement_ControllerRequirement
    properties {
     nodeName=Require: Has Completed Quest Objective
     requirement {
      class=FortControllerRequirement_HasCompletedQuestObjective
      properties {
       bInverseCondition=true
       objectiveBackendName=quest_s17_w08_legendary_q5_obj8
       questDefinition=&quest_s17_w08_legendary_q5
      }
     }
    }
   }
   FortConversationRequirement_ControllerRequirement_44 {
    class=FortConversationRequirement_ControllerRequirement
    properties {
     nodeName=Require: NPC Has Conversation Tag
     participantId {tagName=Conversation.Participant.NPC}
     requirement {class=FortControllerRequirement_NPCHasConversationEntryTag, properties {queryTag {tagName=Conversation.Entry.Maven}}}
    }
   }
   ConversationChoiceNode_12 {
    class=ConversationChoiceNode
    properties {
     defaultChoiceDisplayText=Interrogate
     choiceTags {gameplayTags [{tagName=Conversation.Choice.Category.MainQuestline}, {tagName=Conversation.Choice.AutoSelectIfValid}]}
    }
   }
   FortConversationRequirement_ControllerRequirement_78 {
    class=FortConversationRequirement_ControllerRequirement
    properties {
     nodeName=Require: Has Completed Quest Objective
     requirement {
      class=FortControllerRequirement_HasCompletedQuestObjective
      properties {
       bInverseCondition=true
       objectiveBackendName=quest_s17_w08_legendary_q5_obj15
       questDefinition=&quest_s17_w08_legendary_q5
      }
     }
    }
   }
  ]
 }
 FortConversationTaskNode_CompleteQuestObjectiveByBackendName_7 {
  id=#7
  class=FortConversationTaskNode_CompleteQuestObjectiveByBackendName
  outputs [#14]
  properties {
   objectiveBackendName=quest_s17_w08_legendary_q5_obj7
   questDefinition=&quest_s17_w08_legendary_q5
   softTaskIcon=&T_UI_NPC_Interrobang
  }
  sub_nodes [
   FortConversationRequirement_ControllerRequirement_37 {
    class=FortConversationRequirement_ControllerRequirement
    properties {
     nodeName=Require: Has Completed Quest
     requirement {
      class=FortControllerRequirement_HasCompletedQuest
      properties {bInverseCondition=true, questDefinition=&quest_s17_w08_legendary_q5}
     }
    }
   }
   FortConversationRequirement_ControllerRequirement_38 {
    class=FortConversationRequirement_ControllerRequirement
    properties {
     nodeName=Require: Is Quest Active
     requirement {
      class=FortControllerRequirement_IsQuestActive
      properties {questDefinition=&quest_s17_w08_legendary_q5}
     }
    }
   }
   FortConversationRequirement_ControllerRequirement_39 {
    class=FortConversationRequirement_ControllerRequirement
    properties {
     nodeName=Require: Has Completed Quest Objective
     requirement {
      class=FortControllerRequirement_HasCompletedQuestObjective
      properties {
       bInverseCondition=true
       objectiveBackendName=quest_s17_w08_legendary_q5_obj7
       questDefinition=&quest_s17_w08_legendary_q5
      }
     }
    }
   }
   FortConversationRequirement_ControllerRequirement_40 {
    class=FortConversationRequirement_ControllerRequirement
    properties {
     nodeName=Require: NPC Has Conversation Tag
     participantId {tagName=Conversation.Participant.NPC}
     requirement {class=FortControllerRequirement_NPCHasConversationEntryTag, properties {queryTag {tagName=Conversation.Entry.Marigold}}}
    }
   }
   ConversationChoiceNode_11 {
    class=ConversationChoiceNode
    properties {
     defaultChoiceDisplayText=Interrogate
     choiceTags {gameplayTags [{tagName=Conversation.Choice.Category.MainQuestline}, {tagName=Conversation.Choice.AutoSelectIfValid}]}
    }
   }
   FortConversationRequirement_ControllerRequirement_77 {
    class=FortConversationRequirement_ControllerRequirement
    properties {
     nodeName=Require: Has Completed Quest Objective
     requirement {
      class=FortControllerRequirement_HasCompletedQuestObjective
      properties {
       bInverseCondition=true
       objectiveBackendName=quest_s17_w08_legendary_q5_obj15
       questDefinition=&quest_s17_w08_legendary_q5
      }
     }
    }
   }
  ]
 }
 FortConversationTaskNode_CompleteQuestObjectiveByBackendName_6 {
  id=#6
  class=FortConversationTaskNode_CompleteQuestObjectiveByBackendName
  outputs [#14]
  properties {
   objectiveBackendName=quest_s17_w08_legendary_q5_obj6
   questDefinition=&quest_s17_w08_legendary_q5
   softTaskIcon=&T_UI_NPC_Interrobang
  }
  sub_nodes [
   FortConversationRequirement_ControllerRequirement_33 {
    class=FortConversationRequirement_ControllerRequirement
    properties {
     nodeName=Require: Has Completed Quest
     requirement {
      class=FortControllerRequirement_HasCompletedQuest
      properties {bInverseCondition=true, questDefinition=&quest_s17_w08_legendary_q5}
     }
    }
   }
   FortConversationRequirement_ControllerRequirement_34 {
    class=FortConversationRequirement_ControllerRequirement
    properties {
     nodeName=Require: Is Quest Active
     requirement {
      class=FortControllerRequirement_IsQuestActive
      properties {questDefinition=&quest_s17_w08_legendary_q5}
     }
    }
   }
   FortConversationRequirement_ControllerRequirement_35 {
    class=FortConversationRequirement_ControllerRequirement
    properties {
     nodeName=Require: Has Completed Quest Objective
     requirement {
      class=FortControllerRequirement_HasCompletedQuestObjective
      properties {
       bInverseCondition=true
       objectiveBackendName=quest_s17_w08_legendary_q5_obj6
       questDefinition=&quest_s17_w08_legendary_q5
      }
     }
    }
   }
   FortConversationRequirement_ControllerRequirement_36 {
    class=FortConversationRequirement_ControllerRequirement
    properties {
     nodeName=Require: NPC Has Conversation Tag
     participantId {tagName=Conversation.Participant.NPC}
     requirement {class=FortControllerRequirement_NPCHasConversationEntryTag, properties {queryTag {tagName=Conversation.Entry.Rook}}}
    }
   }
   ConversationChoiceNode_10 {
    class=ConversationChoiceNode
    properties {
     defaultChoiceDisplayText=Interrogate
     choiceTags {gameplayTags [{tagName=Conversation.Choice.Category.MainQuestline}, {tagName=Conversation.Choice.AutoSelectIfValid}]}
    }
   }
   FortConversationRequirement_ControllerRequirement_76 {
    class=FortConversationRequirement_ControllerRequirement
    properties {
     nodeName=Require: Has Completed Quest Objective
     requirement {
      class=FortControllerRequirement_HasCompletedQuestObjective
      properties {
       bInverseCondition=true
       objectiveBackendName=quest_s17_w08_legendary_q5_obj15
       questDefinition=&quest_s17_w08_legendary_q5
      }
     }
    }
   }
  ]
 }
 FortConversationTaskNode_CompleteQuestObjectiveByBackendName_5 {
  id=#5
  class=FortConversationTaskNode_CompleteQuestObjectiveByBackendName
  outputs [#14]
  properties {
   objectiveBackendName=quest_s17_w08_legendary_q5_obj5
   questDefinition=&quest_s17_w08_legendary_q5
   softTaskIcon=&T_UI_NPC_Interrobang
  }
  sub_nodes [
   FortConversationRequirement_ControllerRequirement_29 {
    class=FortConversationRequirement_ControllerRequirement
    properties {
     nodeName=Require: Has Completed Quest
     requirement {
      class=FortControllerRequirement_HasCompletedQuest
      properties {bInverseCondition=true, questDefinition=&quest_s17_w08_legendary_q5}
     }
    }
   }
   FortConversationRequirement_ControllerRequirement_30 {
    class=FortConversationRequirement_ControllerRequirement
    properties {
     nodeName=Require: Is Quest Active
     requirement {
      class=FortControllerRequirement_IsQuestActive
      properties {questDefinition=&quest_s17_w08_legendary_q5}
     }
    }
   }
   FortConversationRequirement_ControllerRequirement_31 {
    class=FortConversationRequirement_ControllerRequirement
    properties {
     nodeName=Require: Has Completed Quest Objective
     requirement {
      class=FortControllerRequirement_HasCompletedQuestObjective
      properties {
       bInverseCondition=true
       objectiveBackendName=quest_s17_w08_legendary_q5_obj5
       questDefinition=&quest_s17_w08_legendary_q5
      }
     }
    }
   }
   FortConversationRequirement_ControllerRequirement_32 {
    class=FortConversationRequirement_ControllerRequirement
    properties {
     nodeName=Require: NPC Has Conversation Tag
     participantId {tagName=Conversation.Participant.NPC}
     requirement {class=FortControllerRequirement_NPCHasConversationEntryTag, properties {queryTag {tagName=Conversation.Entry.Hayseed}}}
    }
   }
   ConversationChoiceNode_9 {
    class=ConversationChoiceNode
    properties {
     defaultChoiceDisplayText=Interrogate
     choiceTags {gameplayTags [{tagName=Conversation.Choice.Category.MainQuestline}, {tagName=Conversation.Choice.AutoSelectIfValid}]}
    }
   }
   FortConversationRequirement_ControllerRequirement_75 {
    class=FortConversationRequirement_ControllerRequirement
    properties {
     nodeName=Require: Has Completed Quest Objective
     requirement {
      class=FortControllerRequirement_HasCompletedQuestObjective
      properties {
       bInverseCondition=true
       objectiveBackendName=quest_s17_w08_legendary_q5_obj15
       questDefinition=&quest_s17_w08_legendary_q5
      }
     }
    }
   }
  ]
 }
 FortConversationTaskNode_CompleteQuestObjectiveByBackendName_4 {
  id=#4
  class=FortConversationTaskNode_CompleteQuestObjectiveByBackendName
  outputs [#14]
  properties {
   objectiveBackendName=quest_s17_w08_legendary_q5_obj4
   questDefinition=&quest_s17_w08_legendary_q5
   softTaskIcon=&T_UI_NPC_Interrobang
  }
  sub_nodes [
   FortConversationRequirement_ControllerRequirement_25 {
    class=FortConversationRequirement_ControllerRequirement
    properties {
     nodeName=Require: Has Completed Quest
     requirement {
      class=FortControllerRequirement_HasCompletedQuest
      properties {bInverseCondition=true, questDefinition=&quest_s17_w08_legendary_q5}
     }
    }
   }
   FortConversationRequirement_ControllerRequirement_26 {
    class=FortConversationRequirement_ControllerRequirement
    properties {
     nodeName=Require: Is Quest Active
     requirement {
      class=FortControllerRequirement_IsQuestActive
      properties {questDefinition=&quest_s17_w08_legendary_q5}
     }
    }
   }
   FortConversationRequirement_ControllerRequirement_27 {
    class=FortConversationRequirement_ControllerRequirement
    properties {
     nodeName=Require: Has Completed Quest Objective
     requirement {
      class=FortControllerRequirement_HasCompletedQuestObjective
      properties {
       bInverseCondition=true
       objectiveBackendName=quest_s17_w08_legendary_q5_obj4
       questDefinition=&quest_s17_w08_legendary_q5
      }
     }
    }
   }
   FortConversationRequirement_ControllerRequirement_28 {
    class=FortConversationRequirement_ControllerRequirement
    properties {
     nodeName=Require: NPC Has Conversation Tag
     participantId {tagName=Conversation.Participant.NPC}
     requirement {class=FortControllerRequirement_NPCHasConversationEntryTag, properties {queryTag {tagName=Conversation.Entry.SwampStalker}}}
    }
   }
   ConversationChoiceNode_8 {
    class=ConversationChoiceNode
    properties {
     defaultChoiceDisplayText=Interrogate
     choiceTags {gameplayTags [{tagName=Conversation.Choice.Category.MainQuestline}, {tagName=Conversation.Choice.AutoSelectIfValid}]}
    }
   }
   FortConversationRequirement_ControllerRequirement_74 {
    class=FortConversationRequirement_ControllerRequirement
    properties {
     nodeName=Require: Has Completed Quest Objective
     requirement {
      class=FortControllerRequirement_HasCompletedQuestObjective
      properties {
       bInverseCondition=true
       objectiveBackendName=quest_s17_w08_legendary_q5_obj15
       questDefinition=&quest_s17_w08_legendary_q5
      }
     }
    }
   }
  ]
 }
 FortConversationTaskNode_CompleteQuestObjectiveByBackendName_1 {
  id=#3
  class=FortConversationTaskNode_CompleteQuestObjectiveByBackendName
  outputs [#14]
  properties {
   objectiveBackendName=quest_s17_w08_legendary_q5_obj1
   questDefinition=&quest_s17_w08_legendary_q5
   softTaskIcon=&T_UI_NPC_Interrobang
  }
  sub_nodes [
   FortConversationRequirement_ControllerRequirement_13 {
    class=FortConversationRequirement_ControllerRequirement
    properties {
     nodeName=Require: Has Completed Quest
     requirement {
      class=FortControllerRequirement_HasCompletedQuest
      properties {bInverseCondition=true, questDefinition=&quest_s17_w08_legendary_q5}
     }
    }
   }
   FortConversationRequirement_ControllerRequirement_14 {
    class=FortConversationRequirement_ControllerRequirement
    properties {
     nodeName=Require: Is Quest Active
     requirement {
      class=FortControllerRequirement_IsQuestActive
      properties {questDefinition=&quest_s17_w08_legendary_q5}
     }
    }
   }
   FortConversationRequirement_ControllerRequirement_15 {
    class=FortConversationRequirement_ControllerRequirement
    properties {
     nodeName=Require: Has Completed Quest Objective
     requirement {
      class=FortControllerRequirement_HasCompletedQuestObjective
      properties {
       bInverseCondition=true
       objectiveBackendName=quest_s17_w08_legendary_q5_obj1
       questDefinition=&quest_s17_w08_legendary_q5
      }
     }
    }
   }
   FortConversationRequirement_ControllerRequirement_16 {
    class=FortConversationRequirement_ControllerRequirement
    properties {
     nodeName=Require: NPC Has Conversation Tag
     participantId {tagName=Conversation.Participant.NPC}
     requirement {class=FortControllerRequirement_NPCHasConversationEntryTag, properties {queryTag {tagName=Conversation.Entry.Rickus}}}
    }
   }
   ConversationChoiceNode_5 {
    class=ConversationChoiceNode
    properties {
     defaultChoiceDisplayText=Interrogate
     choiceTags {gameplayTags [{tagName=Conversation.Choice.Category.MainQuestline}, {tagName=Conversation.Choice.AutoSelectIfValid}]}
    }
   }
   FortConversationRequirement_ControllerRequirement_71 {
    class=FortConversationRequirement_ControllerRequirement
    properties {
     nodeName=Require: Has Completed Quest Objective
     requirement {
      class=FortControllerRequirement_HasCompletedQuestObjective
      properties {
       bInverseCondition=true
       objectiveBackendName=quest_s17_w08_legendary_q5_obj15
       questDefinition=&quest_s17_w08_legendary_q5
      }
     }
    }
   }
  ]
 }
 FortConversationTaskNode_CompleteQuestObjectiveByBackendName_0 {
  id=#2
  class=FortConversationTaskNode_CompleteQuestObjectiveByBackendName
  outputs [#14]
  properties {
   objectiveBackendName=quest_s17_w08_legendary_q5_obj0
   questDefinition=&quest_s17_w08_legendary_q5
   softTaskIcon=&T_UI_NPC_Interrobang
  }
  sub_nodes [
   FortConversationRequirement_ControllerRequirement_5 {
    class=FortConversationRequirement_ControllerRequirement
    properties {
     nodeName=Require: Has Completed Quest
     requirement {
      class=FortControllerRequirement_HasCompletedQuest
      properties {bInverseCondition=true, questDefinition=&quest_s17_w08_legendary_q5}
     }
    }
   }
   FortConversationRequirement_ControllerRequirement_6 {
    class=FortConversationRequirement_ControllerRequirement
    properties {
     nodeName=Require: Is Quest Active
     requirement {
      class=FortControllerRequirement_IsQuestActive
      properties {questDefinition=&quest_s17_w08_legendary_q5}
     }
    }
   }
   FortConversationRequirement_ControllerRequirement_7 {
    class=FortConversationRequirement_ControllerRequirement
    properties {
     nodeName=Require: Has Completed Quest Objective
     requirement {
      class=FortControllerRequirement_HasCompletedQuestObjective
      properties {
       bInverseCondition=true
       objectiveBackendName=quest_s17_w08_legendary_q5_obj0
       questDefinition=&quest_s17_w08_legendary_q5
      }
     }
    }
   }
   FortConversationRequirement_ControllerRequirement_8 {
    class=FortConversationRequirement_ControllerRequirement
    properties {
     nodeName=Require: NPC Has Conversation Tag
     participantId {tagName=Conversation.Participant.NPC}
     requirement {class=FortControllerRequirement_NPCHasConversationEntryTag, properties {queryTag {tagName=Conversation.Entry.Believer}}}
    }
   }
   ConversationChoiceNode_3 {
    class=ConversationChoiceNode
    properties {
     defaultChoiceDisplayText=Interrogate
     choiceTags {gameplayTags [{tagName=Conversation.Choice.Category.MainQuestline}, {tagName=Conversation.Choice.AutoSelectIfValid}]}
    }
   }
   FortConversationRequirement_ControllerRequirement_70 {
    class=FortConversationRequirement_ControllerRequirement
    properties {
     nodeName=Require: Has Completed Quest Objective
     requirement {
      class=FortControllerRequirement_HasCompletedQuestObjective
      properties {
       bInverseCondition=true
       objectiveBackendName=quest_s17_w08_legendary_q5_obj15
       questDefinition=&quest_s17_w08_legendary_q5
      }
     }
    }
   }
  ]
 }
]
